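; Creative device volume that eliminates or applies damage-over-time to players, bots, creatures and vehicles inside, per its options.
; Enter events filter by type, team, class and play mode, then force-kill or apply a SetByCaller GameplayEffect, with huge damage substituting for vehicle kills.
; Stored effect handles allow removal on exit.
; Team, class and vehicle occupancy changes re-evaluate affected targets, while resizing rescales trigger and zone particles.
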(event EventInitializeDevice
  ;[0.9244 0.06801 1] START ME UP!
  (|CallParentFunction)
  (Utilities|FlowControl|Sequence
    (:then_0
      (CallFunction|UpdateVisibilityComponentsToUserSettings))
    (:then_1
      (Visibility|UpdateVisiblity)
      (Visibility|SetVisiblityDuringPhase (Variables|Default|GetVisibleInGameComponent) "NewEnumerator1"))))

(event EventUpdateVolumeSize
  ;[0.5705 1 0.7202] Override parent here to handle VFX-specific logic
  (|CallParentFunction)
  (Resizing|UpdateVolumeSizeParticles))

(event EventPawnEnteredTriggerVolume
  (params (Pawn :type Pawn%20Object%20Reference) (PlayerController :type Player%20Controller%20Object%20Reference))
  (bind _returnvalue (Pawn|GetController Pawn))
  ;[0.2776 1 0.43] Filter through what kind of pawn has entered the volume and apply the appropriate Effect
  (if (Utilities|IsValid PlayerController :!class /Script/Engine.KismetSystemLibrary)
    (bind _valid_controller (CallFunction|CheckPlayerTeamandClassValidity Pawn))
    (if (Math|Boolean|ANDBoolean _valid_controller (Variables|DamageOptions|GetbDamagePlayers) (CallFunction|CanPassinCreateMode))
      (Default|TriggerEffect Pawn)
      (Utilities|IsValid _returnvalue
        (:Is%20Valid
          (GameplayEvent|BroadcastEvent (Variables|UserOptions|Events|GetOn%20Player%20Entering%20Zone) 0 (Utilities|Struct|MakeDamagePlayerEnteredInstigator _returnvalue)))))
    (elif (Variables|DamageOptions|GetbDamageCreatures)
      (bind _asfort_aipawn (Utilities|Casting|CastToFortAIPawn Pawn))
      (Default|TriggerEffect _asfort_aipawn))))

(event EventNon-PawnEnteredTriggerVolume
  (params (Actor :type Actor%20Object%20Reference))
  ;[0.2776 1 0.43] Filter through what kind of pawn has entered the volume and apply the appropriate Effect
  (if (and (Variables|DamageOptions|GetbDamageVehicles) (CallFunction|CanPassinCreateMode))
    (bind _asfort_athena_vehicle (Utilities|Casting|CastToFortAthenaVehicle Actor))
    (CallFunction|BindVehicleEvents _asfort_athena_vehicle)
    (Default|TriggerEffect 0 _asfort_athena_vehicle)))

(event Custom|Player%20Exited%20Vehicle
  (params (Vehicle :type Fort%20Vehicle%20Interface%20Interface%20%28by%20ref%29) (PlayerPawn :type Fort%20Pawn%20Object%20Reference) (SeatIndex :type Integer))
  ;[0.2776 1 0.43] Filter through what kind of pawn has entered the volume and apply the appropriate Effect
  ;;[0.271 0.222 1 0.321] When a player enters or exits a vehicle, re-trigger the effects on the vehicle to filter out if it's still affected
  (if (< (Utilities|Array|Length (Vehicle|GetAllPassengers Vehicle)) 1)
    (CallFunction|Re-TriggerEffectsOnVehicle Vehicle)))

(event Custom|Player%20Entered%20Vehicle
  (params (Vehicle :type Fort%20Vehicle%20Interface%20Interface%20%28by%20ref%29) (PlayerPawn :type Fort%20Pawn%20Object%20Reference) (SeatIndex :type Integer))
  ;[0.2776 1 0.43] Filter through what kind of pawn has entered the volume and apply the appropriate Effect
  ;;[0.271 0.222 1 0.321] When a player enters or exits a vehicle, re-trigger the effects on the vehicle to filter out if it's still affected
  (if (== (Utilities|Array|Length (Vehicle|GetAllPassengers Vehicle)) 1)
    (CallFunction|Re-TriggerEffectsOnVehicle Vehicle)))

(event Custom|UpdateSelectedTeam
  (params (Instigator :type Controller%20Object%20Reference))
  ;[0.6869 1 0.2143] Update the selected Team or Class when receiving from channel transmissions
  ;;[0.3222 1 0.9701] Check the Authority, and Game Progress
  (Networking|SwitchHasAuthority
    (:Authority
      (bind _currentminigame (Default|GetCurrentMinigame)
        (:Is%20Valid
          (if (Minigame|IsinProgress _currentminigame)
            ;;[1 0.3279 0.2682] Get the Team and Class of the player
            (bind _asfort_player_controller_athena (Utilities|Casting|CastToFortPlayerControllerAthena Instigator))
            (Utilities|IsValid _asfort_player_controller_athena
              (:Is%20Valid
                ;;[0.9653 0.3944 1] Update the team and class to affect
                (UserOptions|CustomInitializeTeams (Knobs|Team|ReshiftTeamIndexValue (Math|Conversions|ToInteger(Byte) (Team|GetActorPvPTeamIndex _asfort_player_controller_athena))))
                (Registration|RevalidateOverlappingPawns)))))))))

(event Custom|UpdateSelectedClass
  (params (Instigator :type Controller%20Object%20Reference))
  ;[0.6869 1 0.2143] Update the selected Team or Class when receiving from channel transmissions
  ;;[0.3222 1 0.9701] Check the Authority, and Game Progress
  (Networking|SwitchHasAuthority
    (:Authority
      (bind _currentminigame (Default|GetCurrentMinigame)
        (:Is%20Valid
          (Utilities|IsValid (Minigame|GetFNEClassComponent _currentminigame)
            (:Is%20Valid
              (if (Minigame|IsinProgress _currentminigame)
                ;;[1 0.3279 0.2682] Get the Team and Class of the player
                (bind _asfort_player_controller_athena (Utilities|Casting|CastToFortPlayerControllerAthena Instigator))
                (bind _asfort_player_state (Utilities|Casting|CastToFortPlayerState (Variables|/Script/Engine.Controller|GetPlayerState _asfort_player_controller_athena)))
                ;;[0.9653 0.3944 1] Update the team and class to affect
                (UserOptions|CustomInitializeClass (Math|Conversions|ToInteger(Byte) (Minigame|GetClassSlotforPlayer (Minigame|GetFNEClassComponent _currentminigame) _asfort_player_state false)))
                (Registration|RevalidateOverlappingPawns)))))))))

(event EventPlayerBotEnteredTriggerVolume
  (params (Pawn :type Pawn%20Object%20Reference) (AI%20Bot%20Controller :type Fort%20Athena%20AIBot%20Controller%20Object%20Reference))
  ;[0.2776 1 0.43] Filter through what kind of pawn has entered the volume and apply the appropriate Effect
  ;;[0.643 0.6823 0.1546 0.5] When Player Bot Enters the Trigger Volume, Trigger the effect upon Valid Team
  (if (Utilities|IsValid AI%20Bot%20Controller :!class /Script/Engine.KismetSystemLibrary)
    (if (Math|Boolean|ANDBoolean (Variables|DamageOptions|GetbDamagePlayerBots) (CallFunction|CanPassinCreateMode) (StateChecks|IsValidTeam (Variables|Default|GetCreativeUseRestrictionComponent) Pawn))
      (Default|TriggerEffect Pawn))))

(event EventOnPlayerChangedTeam
  (params (PlayerState :type Fort%20Player%20State%20Object%20Reference) (TeamId :type Byte) (OldTeamId :type Byte) (bTeamChangedWithoutRespawn :type Boolean))
  ;[0.2785 1 0.9402 0.5] Re-fresh application on the damage volume for a player who's changed class or teams
  (bind (_outpawn _pawnreturned) (Class|AbilityFunctions|GetPawnfromPlayerStateZone PlayerState))
  (CallFunction|ReprocessOverlappingPlayersPlayers _outpawn _pawnreturned))

(event MinigameStartedDelegate(FortMinigameProgress)
  ;[0.2619 1 0.8901 0.702] Bind and Unbind events for player team and class states changing
  (Networking|SwitchHasAuthority
    (:Authority
      (bind _currentminigame (Default|GetCurrentMinigame)
        (:Is%20Valid
          (Minigame|BindEventtoOnPlayerClassApplied _currentminigame (EventDispatchers|CreateEvent :!event OnPlayerClassApplied))
          (Minigame|BindEventtoPlayerChangedTeamDelegate _currentminigame (EventDispatchers|CreateEvent :!event OnPlayerChangedTeam)))))))

(event EventPawnExitedTriggerVolume
  (params (Pawn :type Pawn%20Object%20Reference) (PlayerController :type Player%20Controller%20Object%20Reference))
  (bind _returnvalue (Pawn|GetController Pawn))
  ;[1 0.1927 0.1927] End the Effect on pawns leaving the volume
  (if (Utilities|IsValid PlayerController :!class /Script/Engine.KismetSystemLibrary)
    (Default|EndEffect 0 Pawn)
    (Utilities|IsValid _returnvalue
      (:Is%20Valid
        (GameplayEvent|BroadcastEvent (Variables|UserOptions|Events|GetOn%20Player%20Exiting%20Zone) 0 (Utilities|Struct|MakeDamagePlayerExitedInstigator _returnvalue))))
    (elif (Variables|DamageOptions|GetbDamageCreatures)
      (Default|EndEffect 0 Pawn))))

(event EventOnPlayerClassApplied
  (params (PlayerState :type Fort%20Player%20State%20Object%20Reference) (ClassSlotIndex :type Byte))
  ;[0.2785 1 0.9402 0.5] Re-fresh application on the damage volume for a player who's changed class or teams
  (bind (_outpawn _pawnreturned) (Class|AbilityFunctions|GetPawnfromPlayerStateZone PlayerState))
  (CallFunction|ReprocessOverlappingPlayersPlayers _outpawn _pawnreturned))

(event MinigameEndedDelegate(FortMinigameProgress)
  ;[0.2619 1 0.8901 0.702] Bind and Unbind events for player team and class states changing
  (Networking|SwitchHasAuthority
    (:Authority
      (bind _currentminigame (Default|GetCurrentMinigame)
        (:Is%20Valid
          (Minigame|UnbindEventfromOnPlayerClassApplied _currentminigame (EventDispatchers|CreateEvent :!event OnPlayerClassApplied))
          (Minigame|UnbindEventfromPlayerChangedTeamDelegate _currentminigame (EventDispatchers|CreateEvent :!event OnPlayerChangedTeam)))))))

(event EventNon-PawnExitedTriggerVolume
  (params (Actor :type Actor%20Object%20Reference))
  ;[1 0.1927 0.1927] End the Effect on pawns leaving the volume
  (if (Variables|DamageOptions|GetbDamageVehicles)
    (bind _asfort_athena_vehicle (Utilities|Casting|CastToFortAthenaVehicle Actor))
    (CallFunction|UnbindVehicleEvents _asfort_athena_vehicle)
    (Default|EndEffect _asfort_athena_vehicle)))

(event EventPlayerBotExitedTriggerVolume
  (params (Pawn :type Pawn%20Object%20Reference) (AI%20Bot%20Controller :type Fort%20Athena%20AIBot%20Controller%20Object%20Reference))
  ;[1 0.1927 0.1927] End the Effect on pawns leaving the volume
  (if (Utilities|IsValid AI%20Bot%20Controller :!class /Script/Engine.KismetSystemLibrary)
    (if (Variables|DamageOptions|GetbDamagePlayerBots)
      (Default|EndEffect 0 Pawn))))

(fn AddHandleToHoldingArray
  (params (ActiveGameplayEffectHandle :type Active%20Gameplay%20Effect%20Handle%20Structure) (InputPin :type Ability%20System%20Component%20Object%20Reference))
  ;[0.2257 0.8599 1] Add the player's ability system component and the active gameplay and put them in a struct so we can retain knowledge of the handle of the GE we just applied to the player
  (bind _output_get (Variables|Default|SetActiveGameplayEffectHandle ActiveGameplayEffectHandle))
  (bind _output_get_2 (Variables|SetGE%20Handle_ActiveGameplayEffectHandle_7_C6FF36454DF9DCBF6904BD96974FFF0B (Utilities|Struct|MakeAbilityHandleAndSystemHold _output_get InputPin)))
  (Utilities|Array|AddUnique (Variables|Default|GetAbility%20Handle%20Array) _output_get_2))

(fn ApplyGE
  (params (Actor :type Actor%20Object%20Reference) (Magnitude :type Integer))
  (bind _returnvalue (Ability|GetAbilitySystemComponent Actor))
  ;[1 0.9023 0.1292] Using the targeted object, create a Gameplay Effect Spec
  (bind _gameplay_effect (CallFunction|GetVolumeGE))
  ;[0.439 1 0.2192] Apply the Game Effect with the Spec
  (bind _returnvalue_2 (Ability|GameplayEffect|AssignTagSetbyCallerMagnitude (GameplayEffects|MakeOutgoingSpec _returnvalue _gameplay_effect (Math|Conversions|ToFloat(Integer) (Variables|DamageOptions|GetZoneDamagePeriodLevel)) (GameplayEffects|MakeEffectContext _returnvalue)) "(TagName=\"SetByCaller.AbilityDamage\")" (Math|Conversions|ToFloat(Integer) (Variables|GetMagnitude))))
  (bind _returnvalue_3 (GameplayEffects|ApplyGameplayEffectSpecToTarget _returnvalue _returnvalue_2 _returnvalue))
  (CallFunction|AddHandletoHoldingArray _returnvalue_3 _returnvalue))

(fn BindVehicleEvents
  (params (FortAthenaVehicle :type Fort%20Athena%20Vehicle%20Object%20Reference))
  ;[0.505 1 0.2499] Bind events when players enter or exit this vehicle
  (CallFunction|UnbindVehicleEvents FortAthenaVehicle)
  (Vehicle|BindEventtoOnPawnExitedVehicleSeat FortAthenaVehicle (EventDispatchers|CreateEvent :!event Player%20Exited%20Vehicle))
  (Vehicle|BindEventtoOnPawnEnteredVehicleSeat FortAthenaVehicle (EventDispatchers|CreateEvent :!event Player%20Entered%20Vehicle)))

(fn CanPassInCreateMode
  ;[1 0.9202 0.2854] Check to see if we can pass during Create mode when playing in creative
  (return (or (CallFunction|IsinPlayMode) (and (not (CallFunction|IsinPlayMode)) (Variables|DamageOptions|GetbActiveInEditor)))))

(fn CheckVehicleValidity
  (params (Object :type Object%20Reference))
  ;[1 0.8039 0.2999] Should this vehicle be damaged at all?
  (return (or (and (Variables|DamageOptions|GetbDamageVehicles) (not (CallFunction|ReturnIsVehicleEmpty Object))) (and (Variables|DamageOptions|GetbDamageVehicles) (Variables|DamageOptions|GetbDamageEmptyVehicles)))))

(fn CheckPlayerTeamAndClassValidity
  (params (Actor :type Actor%20Object%20Reference))
  ;[0.4111 1 0.2042] Check whether or not the controller is valid based on its team and class compared to what is being filtered
  (bind (_valid_controller _fortplayerpawn) (DEPRECATED|ActorHasValidController (Variables|Default|GetCreativeUseRestrictionComponent) Actor true true))
  (return _valid_controller _fortplayerpawn))

(fn CreateBaseVisibilityComponentArray
  (return 0))

(fn EndEffect
  (params (Vehicle :type Actor%20Object%20Reference) (Pawn :type Pawn%20Object%20Reference))
  (bind _returnvalue (Utilities|IsValid Pawn :!class /Script/Engine.KismetSystemLibrary))
  (bind _returnvalue_2 (Ability|GetAbilitySystemComponent (select _returnvalue Pawn Vehicle)))
  (bind _returnvalue_3 (select _returnvalue Pawn Vehicle))
  ;[1 0.0625 0.0625] Remove a SINGLE stack of the relevant GE upon leaving the volume
  (Utilities|IsValid _returnvalue_3
    (:Is%20Valid
      (bind _active_gameplay_effect_handle (CallFunction|ReturnMatchingGameplayEffectHandle _returnvalue_2))
      (GameplayEffects|RemoveActiveGameplayEffect _returnvalue_2 _active_gameplay_effect_handle -1))))

(fn GetVolumeGE
  (bind _bknob_enablevfx (Variables|DeviceOptions|GetbKnob_EnableVFX))
  ;[1 0.5233 0.4534] Lets filter through our options, and set the correct Volume GE to use down the line
  (if (Variables|DamageOptions|GetbShieldDamage)
    (if _bknob_enablevfx
      (return (Variables|GameplayEffects|GetGE%20DOT%20VFX))
      (else
        (return (Variables|GameplayEffects|GetGE%20DOT%20No%20VFX))))
    (elif _bknob_enablevfx
      (return (Variables|GameplayEffects|GetGE%20True%20DOT%20VFX))
      (else
        (return (Variables|GameplayEffects|GetGE%20True%20DOT%20No%20VFX))))))

(fn GetVisibilityComponents
  (return (Utilities|Array|MakeArray (Variables|Default|GetCreativeEditOnlyMesh))))

(fn IsTargetableByLinearMover
  (return true))

(fn Re-TriggerEffectsonVehicle
  (params (Object :type Fort%20Vehicle%20Interface%20Interface))
  ;[0.2308 0.1369 1] Remove then re-apply the effects on the vehicle in the event that it becomes occupied or vacant
  (bind _asfort_athena_vehicle (Utilities|Casting|CastToFortAthenaVehicle Object))
  (Default|EndEffect _asfort_athena_vehicle)
  (CallFunction|Non-PawnEnteredTriggerVolume _asfort_athena_vehicle))

(fn ReprocessOverlappingPlayersPlayers
  (params (Player%20Pawn :type Fort%20Player%20Pawn%20Object%20Reference) (Pawn%20Returned :type Boolean))
  (bind _playercontroller (Default|GetPlayerControllerfromActor Player%20Pawn))
  ;[0.2458 1 0.2784 0.5] If we have a player pawn, and are overlaping it, remove the applied effect and re-apply it in order to filter out unqualified players
  (Utilities|IsValid Player%20Pawn
    (:Is%20Valid
      (Utilities|IsValid _playercontroller
        (:Is%20Valid
          (if (and Pawn%20Returned (Collision|IsOverlappingActor self Player%20Pawn))
            (CallFunction|PawnExitedTriggerVolume Player%20Pawn _playercontroller)
            (CallFunction|PawnEnteredTriggerVolume Player%20Pawn _playercontroller)))))))

(fn ReturnMatchingGameplayEffectHandle
  (params (Ability%20System%20Component :type Ability%20System%20Component%20Object%20Reference))
  ;[0.3118 1 0.9701] Retrieve the Gameplay handle used to identify the gameplay effect.
  (bind (_array_element _array_index) (Utilities|Array|ForEachLoopwithBreak (Variables|Default|GetAbility%20Handle%20Array))
    (:LoopBody
      (if (Utilities|Equal(Object) Ability%20System%20Component (.AbilitySystemComponent_8_57E9F25741437CC55361B58F9DB743E0 (Utilities|Struct|BreakAbilityHandleandSystemHold _array_element)))
        (Variables|Default|SetActiveGameplayEffectHandle (Utilities|Struct|BreakAbilityHandleandSystemHold _array_element))
        (Utilities|Array|RemoveIndex (Variables|Default|GetAbility%20Handle%20Array) _array_index)))
    (:Completed
      _))
  (return (Variables|Default|GetActiveGameplayEffectHandle)))

(fn ReturnIsVehicleEmpty
  (params (Object :type Object%20Reference))
  ;[0.8591 1 0.3345] Return whether or not this vehicle is empty
  (bind _asfort_athena_vehicle (Utilities|Casting|CastToFortAthenaVehicle Object))
  (return (<= (Utilities|Array|Length (Vehicle|GetAllPassengers _asfort_athena_vehicle)) 0)))

(fn SendTriggerMessage
  (params (Trigger%20Message%20Component :type Fort%20Gameplay%20Trigger%20Message%20Component%20Object%20Reference) (Pawn :type Pawn%20Object%20Reference))
  (bind _returnvalue (Pawn|GetController Pawn))
  (Utilities|IsValid _returnvalue
    (:Is%20Valid
      (Class|FortGameplayTriggerMessageComponent|SendTriggerMessage Trigger%20Message%20Component _returnvalue))))

(fn ShouldInvisiblePropsBeAllowed
  (return true))

(fn ShouldShowInGame
  (params (DeviceOptionVisualInGame :type Boolean))
  (bind _returnvalue (Minigame|GetMinigameforActor self))
  (Utilities|IsValid _returnvalue
    (:Is%20Valid
      (LocalVariables|SetIsPlayMode (Minigame|IsinPlayMode _returnvalue :!class /Script/FortniteGame.FortMinigame)))
    (:Is%20Not%20Valid
      _))
  (return (select (LocalVariables|GetIsPlayMode) (|CallParentFunction (Variables|GetDeviceOptionVisualInGame)) (Variables|GetDeviceOptionVisualInGame))))

(fn TriggerEffect
  (params (Pawn :type Pawn%20Object%20Reference) (Vehicle :type Actor%20Object%20Reference))
  ;[0.1639 0.174 1] Filter out if we're affecting players or vehicles
  (if (Utilities|IsValid Pawn :!class /Script/Engine.KismetSystemLibrary)
    ;[1 0.06152 0] If the options are set to eliminate, force the pawn to die
    (switch int (Variables|DamageOptions|GetZoneDamageType)
      (:0
        (bind _asfort_pawn (Utilities|Casting|CastToFortPawn Pawn))
        (FortPawn|ForceKill _asfort_pawn "(TagName=\"Gameplay.Effect.InstantDeath\")" 0 self))
      (:1
        ;[0.3017 1 0.2404] Apply the GE
        (CallFunction|ApplyGE Pawn (Variables|DamageOptions|GetZoneDamage))))
    (else
      (if (CallFunction|CheckVehicleValidity Vehicle)
        (switch int (Variables|DamageOptions|GetZoneDamageType)
          (:0
            (CallFunction|ApplyGE Vehicle 9999999) ; At this time no function exists to force vehicles to die, so we're going to do this the dirty way and apply more damage then health ever
          )
          (:1
            (CallFunction|ApplyGE Vehicle (Variables|DamageOptions|GetZoneDamage))))))))

(fn UnbindVehicleEvents
  (params (FortAthenaVehicle :type Fort%20Athena%20Vehicle%20Object%20Reference))
  ;[1 0.6892 0.2661] Unbind all events where players enter or leave the specific vehicle
  (Vehicle|UnbindEventfromOnPawnEnteredVehicleSeat FortAthenaVehicle (EventDispatchers|CreateEvent :!event Player%20Entered%20Vehicle))
  (Vehicle|UnbindEventfromOnPawnExitedVehicleSeat FortAthenaVehicle (EventDispatchers|CreateEvent :!event Player%20Exited%20Vehicle)))

(fn UpdateVisibilityComponentstoUserSettings
  (bind (_zone_components _base_plate_components) (Visibility|UpdateVisualComponents))
  ;[0.4546 1 0.379] Set our visibility components to use later
  (bind _zonevisibilitycomponent (Variables|Default|GetZoneVisibilityComponent)
    (:then
      (Visibility|SetVisibilityComponents _zonevisibilitycomponent _zone_components true))
    (:else
      _))
  (bind _basevisibilitycomponent (Variables|Default|GetBaseVisibilityComponent)
    (:then
      (Visibility|SetVisibilityComponents _basevisibilitycomponent _base_plate_components true))
    (:else)))

(fn UpdateVisualComponents
  ;[1 0.8606 0.217 0.864] Create an array of all components we need to affect visually
  (return (Utilities|Array|MakeArray (Variables|Default|GetP_EliminationZone_Edges) (Variables|Default|GetZoneMesh)) (Utilities|Array|MakeArray (Variables|Debug|GetStaticMeshComponent))))

(fn UpdateVolumeSizeParticles
  ;[1 0.9828 0.5683] Particle Man, Particle Man
  (Components|Activation|Deactivate (Variables|Default|GetP_EliminationZone_Edges))
  (bind _output_get (Variables|Default|SetNewScale (Math|Vector|MakeVector (Variables|Default|GetZoneXScale) (Variables|Default|GetZoneYScale) (Variables|Default|GetZoneZScale))))
  (Transformation|SetRelativeScale3D (Variables|Default|GetP_EliminationZone_Edges) _output_get)
  (Variables|/Script/CascadeRuntime.ParticleSystemComponent|SetInstanceParameters (Resizing|VFXParameters) (Variables|Default|GetP_EliminationZone_Edges))
  (Effects|Components|ParticleSystem|SetTemplate (Variables|Default|GetP_EliminationZone_Edges) (select (== (Variables|DamageOptions|GetZoneDamageType) 0) (Variables|VisualBounds|GetP_ElimZone_DirectDamage) (Variables|VisualBounds|GetP_ElimZone_DOT)))
  (Rendering|SetBoundsScale (Variables|Default|GetP_EliminationZone_Edges) (Math|Float|Max(Float) (Variables|Default|GetZoneXScale) (Variables|Default|GetZoneYScale) (Variables|Default|GetZoneZScale) (Variables|VisualBounds|GetParticlesMaxDrawDistance)))
  (Transformation|SetRelativeScale3D (Variables|Default|GetTrigger) (Variables|Default|GetNewScale))
  (bind _location (Resizing|GetTriggerLocation))
  (Transformation|SetRelativeLocation (Variables|Default|GetTrigger) _location false true))

(fn ConstructionScript
  (|CallParentFunction))

(fn VFXParameters
  (bind (_returnvalue_x _returnvalue_y _returnvalue_z) (Math|Vector|BreakVector (* (Math|Vector|MakeVector 512.0 512.0 384.0) (Math|Vector|MakeVector (Variables|Default|GetZoneXScale) (Variables|Default|GetZoneYScale) (Variables|Default|GetZoneZScale)))))
  (bind (_zoneedgessize_x _zoneedgessize_y _zoneedgessize_z) (Math|Vector|BreakVector (Variables|GetZoneEdgesSize_X)))
  (return (Utilities|Array|MakeArray (Utilities|Struct|MakeParticleSysParam "BlocksVolume" "PSPT_VectorRand" 0.0 0.0 (Variables|Default|GetBlocksVolumeHigh) (Variables|Default|GetBlocksVolumeLow)) (Utilities|Struct|MakeParticleSysParam "BlocksSpawnRateScale" "PSPT_Scalar" (Utilities|ClampMaxFloat (/ (Math|Float|float*float _returnvalue_x _returnvalue_y _returnvalue_z) 100663296.0) 1e+01) 0.0 "0.000000,0.000000,0.000000" (0.0 0.0 0.0) "(B=0,G=0,R=0,A=0)" 0 0) (Utilities|Struct|MakeParticleSysParam "MeshSize" "PSPT_Vector" 0.0 0.0 (Math|Vector|MakeVector _zoneedgessize_x _zoneedgessize_y (/ _zoneedgessize_z (Variables|Default|GetZoneZScale))) "0.000000,0.000000,0.000000" "(B=0,G=0,R=0,A=0)" 0 0) (Utilities|Struct|MakeParticleSysParam "BlocksSize" "PSPT_Vector" 0.0 0.0 (Math|Vector|MakeVector (Utilities|Reciprocal (Variables|Default|GetZoneXScale)) (Utilities|Reciprocal (Variables|Default|GetZoneXScale))) "0.000000,0.000000,0.000000" "(B=0,G=0,R=0,A=0)" 0 0) (Utilities|Struct|MakeParticleSysParam "BlocksVelocity" "PSPT_Vector" 0.0 0.0 (Math|Vector|MakeVector 0.0 0.0 (* -2e+01 (Utilities|Reciprocal (Variables|Default|GetZoneZScale)))) "0.000000,0.000000,0.000000" "(B=0,G=0,R=0,A=0)" 0 0))))
