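; Creative damage volume device that damages or eliminates players, bots, creatures, vehicles and physics props inside a configurable zone.
; Enter events filter by team, class and options, then apply a set-by-caller damage-over-time GameplayEffect or force-kill pawns; exit removes it using stored handles.
; Team or class changes, vehicle occupancy and damage overrides re-trigger effects.
; Zone visuals use an async-loaded box or cylinder particle system scaled to the volume.
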(event EventInitializeDevice
  ;[0 1 0.09904 0.2] Initialization
  (|CallParentFunction)
  (CallFunction|UpdateVolume))

(event EventOnWorldReady
  ;[0 1 0.09904 0.2] Initialization
  (|CallParentFunction)
  (CallFunction|UpdateCreativeVolumeFrameonWorldReady))

(event ChildFunctions|EventPlayModeChanged
  (params (Minigame :type Fort%20Minigame%20Object%20Reference) (bIsInPlayMode :type Boolean))
  (|CallParentFunction Minigame bIsInPlayMode)
  (CallFunction|UpdateVolume))

(event EventUpdateVolumeSize
  ;[0.5705 1 0.7202] Override parent here to handle VFX-specific logic
  (|CallParentFunction)
  (CallFunction|UpdateVolume))

(event EventPawnEnteredTriggerVolume
  (params (Pawn :type Pawn%20Object%20Reference) (PlayerController :type Controller%20Object%20Reference))
  (bind _returnvalue (Pawn|GetController Pawn))
  ;[0.2776 1 0.43] Filter through what kind of pawn has entered the volume and apply the appropriate Effect
  (if (Utilities|IsValid PlayerController :!class /Script/Engine.KismetSystemLibrary)
    (bind _valid_controller (CallFunction|CheckPlayerTeamandClassValidity Pawn))
    (if (Math|Boolean|ANDBoolean _valid_controller (Variables|UserOptions|GetAffectsPlayers) (CallFunction|CanPassinCreateMode))
      (Default|TriggerEffect 0 Pawn 0 0)
      (Utilities|IsValid _returnvalue
        (:Is%20Valid
          (GameplayEvent|BroadcastEvent (Variables|UserOptions|Events|GetOn%20Player%20Entering%20Zone) 0 (Utilities|Struct|MakeDamagePlayerEnteredInstigator _returnvalue)))))
    (elif (Variables|UserOptions|GetAffectsCreatures)
      (bind _asfort_aipawn (Utilities|Casting|CastToFortAIPawn Pawn))
      (bind (_isvalidcontroller _fortaipawn) (CreativeAI|TeamRestriction|CheckAITeamValidity _asfort_aipawn (Variables|UserOptions|GetSelectedTeam) (Variables|UserOptions|GetInvertTeamSelection)))
      (if _isvalidcontroller
        (Default|TriggerEffect 0 _fortaipawn 0 0)
        (Utilities|IsValid _returnvalue
          (:Is%20Valid
            _)))))
  (GameplayEvent|BroadcastEvent (Variables|UserOptions|Events|GetOn%20Agent%20Entering%20Zone) 0 (Utilities|Struct|MakeDamageAgentEnteredInstigator _returnvalue)))

(event EventNon-PawnEnteredTriggerVolume
  (params (Actor :type Actor%20Object%20Reference))
  ;[0.2776 1 0.43] Filter through what kind of pawn has entered the volume and apply the appropriate Effect
  (if (and (Variables|UserOptions|GetAffectsVehicles) (CallFunction|CanPassinCreateMode))
    (bind _asfort_athena_vehicle (Utilities|Casting|CastToFortAthenaVehicle Actor))
    (CallFunction|BindVehicleEvents _asfort_athena_vehicle)
    (Default|TriggerEffect 0 0 _asfort_athena_vehicle 0)))

(event Custom|Player%20Exited%20Vehicle
  (params (Vehicle :type Fort%20Vehicle%20Interface%20Interface%20%28by%20ref%29) (PlayerPawn :type Fort%20Pawn%20Object%20Reference) (SeatIndex :type Integer))
  ;[0.2776 1 0.43] Filter through what kind of pawn has entered the volume and apply the appropriate Effect
  ;;[0.271 0.222 1 0.321] When a player enters or exits a vehicle, re-trigger the effects on the vehicle to filter out if it's still affected
  (if (< (Utilities|Array|Length (Vehicle|GetAllPassengers Vehicle)) 1)
    (CallFunction|Re-TriggerEffectsOnVehicle Vehicle)))

(event Custom|Player%20Entered%20Vehicle
  (params (Vehicle :type Fort%20Vehicle%20Interface%20Interface%20%28by%20ref%29) (PlayerPawn :type Fort%20Pawn%20Object%20Reference) (SeatIndex :type Integer))
  ;[0.2776 1 0.43] Filter through what kind of pawn has entered the volume and apply the appropriate Effect
  ;;[0.271 0.222 1 0.321] When a player enters or exits a vehicle, re-trigger the effects on the vehicle to filter out if it's still affected
  (if (== (Utilities|Array|Length (Vehicle|GetAllPassengers Vehicle)) 1)
    (CallFunction|Re-TriggerEffectsOnVehicle Vehicle)))

(event Custom|UpdateSelectedTeam
  (params (Instigator :type Controller%20Object%20Reference))
  (bind _controller (CreativeAI|Controller|GetParticipantControllerfromActor Instigator false))
  ;[0.6869 1 0.2143] Update the selected Team or Class when receiving from channel transmissions
  ;;[0.3222 1 0.9701] Check the Authority, and Game Progress
  (Networking|SwitchHasAuthority
    (:Authority
      (bind _currentminigame (Default|GetCurrentMinigame)
        (:Is%20Valid
          (if (Minigame|IsinProgress _currentminigame)
            ;;[1 0.3279 0.2682] Get the Team and Class of the player
            (Utilities|IsValid _controller
              (:Is%20Valid
                ;;[0.9653 0.3944 1] Update the team and class to affect
                (UserOptions|CustomInitializeTeams (Knobs|Team|ReshiftTeamIndexValue (Math|Conversions|ToInteger(Byte) (Team|GetActorPvPTeamIndex _controller))))
                (Registration|RevalidateOverlappingPawns)))))))))

(event Custom|UpdateSelectedClass
  (params (Instigator :type Controller%20Object%20Reference))
  (bind _controller (CreativeAI|Controller|GetParticipantControllerfromActor Instigator false))
  ;[0.6869 1 0.2143] Update the selected Team or Class when receiving from channel transmissions
  ;;[0.3222 1 0.9701] Check the Authority, and Game Progress
  (Networking|SwitchHasAuthority
    (:Authority
      (bind _currentminigame (Default|GetCurrentMinigame)
        (:Is%20Valid
          (Utilities|IsValid (Minigame|GetFNEClassComponent _currentminigame)
            (:Is%20Valid
              (if (Minigame|IsinProgress _currentminigame)
                ;;[1 0.3279 0.2682] Get the Team and Class of the player
                (Utilities|IsValid _controller
                  (:Is%20Valid
                    (bind _asfort_player_state (Utilities|Casting|CastToFortPlayerState (Variables|/Script/Engine.Controller|GetPlayerState _controller)))
                    ;;[0.9653 0.3944 1] Update the team and class to affect
                    (UserOptions|CustomInitializeClass (Math|Conversions|ToInteger(Byte) (Minigame|GetClassSlotforPlayer (Minigame|GetFNEClassComponent _currentminigame) _asfort_player_state false)))
                    (Registration|RevalidateOverlappingPawns)))))))))))

(event EventPlayerBotEnteredTriggerVolume
  (params (Pawn :type Pawn%20Object%20Reference) (AI%20Bot%20Controller :type Fort%20Athena%20AIBot%20Controller%20Object%20Reference))
  ;[0.2776 1 0.43] Filter through what kind of pawn has entered the volume and apply the appropriate Effect
  ;;[0.643 0.6823 0.1546 0.5] When Player Bot Enters the Trigger Volume, Trigger the effect upon Valid Team
  (if (Utilities|IsValid AI%20Bot%20Controller :!class /Script/Engine.KismetSystemLibrary)
    (if (Math|Boolean|ANDBoolean (CallFunction|ShouldBotBeAffected Pawn) (CallFunction|CanPassinCreateMode) (StateChecks|IsValidTeam (Variables|Default|GetCreativeUseRestrictionComponent) Pawn))
      (Default|TriggerEffect 0 Pawn 0 0)
      (GameplayEvent|BroadcastEvent (Variables|UserOptions|Events|GetOn%20Agent%20Entering%20Zone) 0 (Utilities|Struct|MakeDamageAgentEnteredInstigator AI%20Bot%20Controller)))))

(event EventOnPlayerChangedTeam
  (params (PlayerState :type Fort%20Player%20State%20Object%20Reference) (TeamId :type Byte) (OldTeamId :type Byte) (bTeamChangedWithoutRespawn :type Boolean))
  ;[0.2785 1 0.9402 0.5] Re-fresh application on the damage volume for a player who's changed class or teams
  (bind (_outpawn _pawnreturned) (Class|AbilityFunctions|GetPawnfromPlayerStateZone PlayerState))
  (CallFunction|ReprocessOverlappingPlayersPlayers _outpawn _pawnreturned))

(event MinigameStartedDelegate(FortMinigameProgress)
  ;[0.2619 1 0.8901 0.702] Bind and Unbind events for player team and class states changing
  (Networking|SwitchHasAuthority
    (:Authority
      (bind _currentminigame (Default|GetCurrentMinigame)
        (:Is%20Valid
          (Minigame|BindEventtoOnPlayerClassApplied _currentminigame (EventDispatchers|CreateEvent :!event OnPlayerClassApplied))
          (Minigame|BindEventtoPlayerChangedTeamDelegate _currentminigame (EventDispatchers|CreateEvent :!event OnPlayerChangedTeam)))))))

(event EventPhysicsObjectEnteredTriggerVolume
  (params (Actor :type Actor%20Object%20Reference))
  ;[0.2776 1 0.43] Filter through what kind of pawn has entered the volume and apply the appropriate Effect
  (Default|TriggerEffect 0 0 0 Actor)
  (bind _asbuilding_prop (Utilities|Casting|CastToBuildingProp Actor))
  (GameplayEvent|BroadcastEvent (Variables|UserOptions|Events|GetOnPhysicsPropEnter) 0 (Utilities|Struct|MakeEventVolumeDevicePhysicsEnter _asbuilding_prop)))

(event EventOnPlayerClassApplied
  (params (PlayerState :type Fort%20Player%20State%20Object%20Reference) (ClassSlotIndex :type Byte))
  ;[0.2785 1 0.9402 0.5] Re-fresh application on the damage volume for a player who's changed class or teams
  (bind (_outpawn _pawnreturned) (Class|AbilityFunctions|GetPawnfromPlayerStateZone PlayerState))
  (CallFunction|ReprocessOverlappingPlayersPlayers _outpawn _pawnreturned))

(event MinigameEndedDelegate(FortMinigameProgress)
  ;[0.2619 1 0.8901 0.702] Bind and Unbind events for player team and class states changing
  (Variables|Verse|SetDamageOverride -1)
  (Networking|SwitchHasAuthority
    (:Authority
      (bind _currentminigame (Default|GetCurrentMinigame)
        (:Is%20Valid
          (Minigame|UnbindEventfromOnPlayerClassApplied _currentminigame (EventDispatchers|CreateEvent :!event OnPlayerClassApplied))
          (Minigame|UnbindEventfromPlayerChangedTeamDelegate _currentminigame (EventDispatchers|CreateEvent :!event OnPlayerChangedTeam)))))))

(event EventPawnExitedTriggerVolume
  (params (Pawn :type Pawn%20Object%20Reference) (PlayerController :type Controller%20Object%20Reference))
  (bind _returnvalue (Pawn|GetController Pawn))
  ;[1 0.1927 0.1927] End the Effect on pawns leaving the volume
  (if (Utilities|IsValid PlayerController :!class /Script/Engine.KismetSystemLibrary)
    (Default|EndEffect 0 Pawn)
    (Utilities|IsValid _returnvalue
      (:Is%20Valid
        (GameplayEvent|BroadcastEvent (Variables|UserOptions|Events|GetOn%20Player%20Exiting%20Zone) 0 (Utilities|Struct|MakeDamagePlayerExitedInstigator _returnvalue))))
    (elif (Variables|UserOptions|GetAffectsCreatures)
      (Default|EndEffect 0 Pawn)
      (Utilities|IsValid _returnvalue
        (:Is%20Valid
          _))))
  (GameplayEvent|BroadcastEvent (Variables|UserOptions|Events|GetOn%20Agent%20Exiting%20Zone) 0 (Utilities|Struct|MakeDamageAgentExitedInstigator _returnvalue)))

(event EventNon-PawnExitedTriggerVolume
  (params (Actor :type Actor%20Object%20Reference))
  ;[1 0.1927 0.1927] End the Effect on pawns leaving the volume
  (if (Variables|UserOptions|GetAffectsVehicles)
    (bind _asfort_athena_vehicle (Utilities|Casting|CastToFortAthenaVehicle Actor))
    (CallFunction|UnbindVehicleEvents _asfort_athena_vehicle)
    (Default|EndEffect _asfort_athena_vehicle)))

(event EventPlayerBotExitedTriggerVolume
  (params (Pawn :type Pawn%20Object%20Reference) (AI%20Bot%20Controller :type Fort%20Athena%20AIBot%20Controller%20Object%20Reference))
  ;[1 0.1927 0.1927] End the Effect on pawns leaving the volume
  (if (Utilities|IsValid AI%20Bot%20Controller :!class /Script/Engine.KismetSystemLibrary)
    (if (CallFunction|ShouldBotBeAffected Pawn)
      (Default|EndEffect 0 Pawn)
      (GameplayEvent|BroadcastEvent (Variables|UserOptions|Events|GetOn%20Agent%20Exiting%20Zone) 0 (Utilities|Struct|MakeDamageAgentExitedInstigator AI%20Bot%20Controller)))))

(event EventPhysicsObjectExitedTriggerVolume
  (params (Actor :type Actor%20Object%20Reference))
  ;[1 0.1927 0.1927] End the Effect on pawns leaving the volume
  (Default|EndEffect Actor)
  (bind _asbuilding_prop (Utilities|Casting|CastToBuildingProp Actor))
  (GameplayEvent|BroadcastEvent (Variables|UserOptions|Events|GetOnPhysicsPropExit) 0 (Utilities|Struct|MakeEventVolumeDevicePhysicsExit _asbuilding_prop)))

(event Custom|Update%20Volume
  (bind (_returnvalue_x _returnvalue_y _returnvalue_z) (Math|Vector|BreakVector (CallFunction|GetVolumeScale)))
  (bind _p_eliminationzone_edges (Variables|Default|GetP_EliminationZone_Edges))
  ;[1 0.9828 0.5683] Update the volume. For the Damage Volume, this is primarily updating the particle.
  (Components|Activation|Deactivate _p_eliminationzone_edges)
  (if (and (or (Variables|UserOptions|GetZoneVisibleDuringGame) (not (CallFunction|IsinPlayMode))) (not (Building|IsPreviewBuildingActor)))
    (Transformation|SetRelativeScale3D _p_eliminationzone_edges (CallFunction|GetVolumeScale))
    (Variables|/Script/CascadeRuntime.ParticleSystemComponent|SetInstanceParameters (Resizing|VFXParameters) (Variables|Default|GetP_EliminationZone_Edges))
    (bind _loaded_object (Creative|Loading|AsyncLoadEditorSafe (select (Utilities|Enum|Equal(Enum) (Variables|UserOptions|GetZone%20Shape) "NewEnumerator4") (Variables|VisualBounds|GetP_ElimZone_Cylinder) (Variables|VisualBounds|GetP_ElimZone_Box)))
      (:Loaded
        (bind _asparticle_system (Utilities|Casting|CastToParticleSystem _loaded_object))
        (Effects|Components|ParticleSystem|SetTemplate (Variables|Default|GetP_EliminationZone_Edges) _asparticle_system)
        (Rendering|SetBoundsScale (Variables|Default|GetP_EliminationZone_Edges) (Math|Float|Max(Float) _returnvalue_x _returnvalue_y _returnvalue_z (Variables|VisualBounds|GetParticlesMaxDrawDistance)))))
    (else
      (Effects|Components|ParticleSystem|SetTemplate _p_eliminationzone_edges 0) ; Clear template to free asset ref..
    )))

(fn AddHandleToHoldingArray
  (params (ActiveGameplayEffectHandle :type Active%20Gameplay%20Effect%20Handle%20Structure) (InputPin :type Ability%20System%20Component%20Object%20Reference))
  ;[0.2257 0.8599 1] Add the player's ability system component and the active gameplay and put them in a struct so we can retain knowledge of the handle of the GE we just applied to the player
  (bind _output_get (Variables|Default|SetActiveGameplayEffectHandle ActiveGameplayEffectHandle))
  (bind _output_get_2 (Variables|SetGE%20Handle_ActiveGameplayEffectHandle_7_C6FF36454DF9DCBF6904BD96974FFF0B (Utilities|Struct|MakeAbilityHandleAndSystemHold _output_get InputPin)))
  (Utilities|Array|AddUnique (Variables|Default|GetAbility%20Handle%20Array) _output_get_2))

(fn ApplyGE
  (params (Actor :type Actor%20Object%20Reference) (Magnitude :type Integer))
  (bind _returnvalue (Ability|GetAbilitySystemComponent Actor))
  ;[1 0.9023 0.1292] Using the targeted object, create a Gameplay Effect Spec
  (bind _gameplay_effect (CallFunction|GetVolumeGE))
  ;[0.439 1 0.2192] Apply the Game Effect with the Spec
  (bind _returnvalue_2 (Ability|GameplayEffect|AddAssetTag (GameplayEffects|MakeOutgoingSpec _returnvalue _gameplay_effect (Utilities|Map|Find (Variables|Default|GetDamage%20Tickrate%20Map) (Variables|UserOptions|GetDamage%20Tick%20Rate)) (GameplayEffects|MakeEffectContext _returnvalue)) (Variables|Default|GetHideIndicatorDamageTag)))
  (bind _returnvalue_3 (Ability|GameplayEffect|AddAssetTag _returnvalue_2 (Variables|Default|GetVolumeDamageTag)))
  (bind _returnvalue_4 (Ability|GameplayEffect|AssignTagSetbyCallerMagnitude _returnvalue_3 "(TagName=\"SetByCaller.AbilityDamage\")" (Math|Conversions|ToFloat(Integer) (Variables|GetMagnitude))))
  (bind _returnvalue_5 (GameplayEffects|ApplyGameplayEffectSpecToTarget _returnvalue _returnvalue_4 _returnvalue))
  (CallFunction|AddHandletoHoldingArray _returnvalue_5 _returnvalue))

(fn BindVehicleEvents
  (params (FortAthenaVehicle :type Fort%20Athena%20Vehicle%20Object%20Reference))
  ;[0.505 1 0.2499] Bind events when players enter or exit this vehicle
  (CallFunction|UnbindVehicleEvents FortAthenaVehicle)
  (Vehicle|BindEventtoOnPawnExitedVehicleSeat FortAthenaVehicle (EventDispatchers|CreateEvent :!event Player%20Exited%20Vehicle))
  (Vehicle|BindEventtoOnPawnEnteredVehicleSeat FortAthenaVehicle (EventDispatchers|CreateEvent :!event Player%20Entered%20Vehicle)))

(fn CanPassInCreateMode
  ;[1 0.9202 0.2854] Check to see if we can pass during Create mode when playing in creative
  (return (or (CallFunction|IsinPlayMode) (and (not (CallFunction|IsinPlayMode)) (Variables|DamageOptions|GetbActiveInEditor)))))

(fn CheckVehiclePassengerTeamValidity
  (params (Vehicle :type Actor%20Object%20Reference))
  (bind _asfort_vehicle_interface (Utilities|Casting|CastToFortVehicleInterface Vehicle)
    (:then
      (bind _output_get (LocalVariables|SetCachedPassengers (Vehicle|GetAllPassengers _asfort_vehicle_interface)))
      (bind _array_element (Utilities|Array|ForEachLoopwithBreak _output_get)
        (:LoopBody
          (bind _output_get_2 (LocalVariables|SetResult (CreativeUserRestrictions|CanParticipantActorInteract (Variables|Default|GetCreativeUseRestrictionComponent) _array_element true true false)))
          (if _output_get_2
            _))
        (:Completed
          _))
      (return (LocalVariables|GetResult)))
    (:CastFailed
      (return false))))

(fn CheckVehicleValidity
  (params (Object :type Object%20Reference))
  ;[1 0.8039 0.2999] Should this vehicle be damaged at all?
  (return (and (Utilities|IsValid Object :!class /Script/Engine.KismetSystemLibrary) (or (and (Variables|UserOptions|GetAffectsVehicles) (not (CallFunction|ReturnIsVehicleEmpty Object))) (and (Variables|UserOptions|GetAffectsVehicles) (Variables|UserOptions|GetAffectsUnmannedVehicles))))))

(fn CheckPlayerTeamAndClassValidity
  (params (Actor :type Actor%20Object%20Reference))
  ;[0.4111 1 0.2042] Check whether or not the controller is valid based on its team and class compared to what is being filtered
  (bind (_valid_controller _fortplayerpawn) (CreativeUserRestrictions|ActorHasValidParticipantController (Variables|Default|GetCreativeUseRestrictionComponent) Actor true true true))
  (return _valid_controller _fortplayerpawn))

(fn EndEffect
  (params (VehicleOrPhysicsObject :type Actor%20Object%20Reference) (Pawn :type Pawn%20Object%20Reference))
  (bind _returnvalue (Utilities|IsValid Pawn :!class /Script/Engine.KismetSystemLibrary))
  (bind _returnvalue_2 (Ability|GetAbilitySystemComponent (select _returnvalue Pawn VehicleOrPhysicsObject)))
  (bind _returnvalue_3 (select _returnvalue Pawn VehicleOrPhysicsObject))
  ;[1 0.0625 0.0625] Remove a SINGLE stack of the relevant GE upon leaving the volume
  (Utilities|IsValid _returnvalue_3
    (:Is%20Valid
      (bind _active_gameplay_effect_handle (CallFunction|ReturnMatchingGameplayEffectHandle _returnvalue_2))
      (GameplayEffects|RemoveActiveGameplayEffect _returnvalue_2 _active_gameplay_effect_handle -1))))

(fn GetAllVisualComponents
  (return (Utilities|Array|MakeArray (Variables|Debug|GetStaticMeshComponent)) (Utilities|Array|MakeArray (Variables|Default|GetFNE_VolumeOverlapComponent) (Variables|Default|GetP_EliminationZone_Edges))))

(fn GetVolumeGE
  (bind _enablevfx (Variables|UserOptions|GetEnableVFX))
  ;[1 0.5233 0.4534] Lets filter through our options, and set the correct Volume GE to use down the line
  (if (Variables|UserOptions|GetAffectsShields)
    (if _enablevfx
      (return (Variables|GameplayEffects|GetGE%20DOT%20VFX))
      (else
        (return (Variables|GameplayEffects|GetGE%20DOT%20No%20VFX))))
    (elif _enablevfx
      (return (Variables|GameplayEffects|GetGE%20True%20DOT%20VFX))
      (else
        (return (Variables|GameplayEffects|GetGE%20True%20DOT%20No%20VFX))))))

(fn GetZoneShape
  (return (Variables|UserOptions|GetZone%20Shape)))

(fn GetCollisionComponent
  (return (Variables|Default|GetFNE_VolumeOverlapComponent)))

(fn GetDamage
  (return (Utilities|GetCurrentDamage)))

(fn GetSelectedClassInternal
  ;[1 1 1 1 16] Return the required class for automated testing purposes (if class is 255 return 0, else return class + 1)
  (return (select (== (Variables|/Game/Creative/Devices/Common/Components/Creative_UseRestriction_Component.Creative_UseRestriction_Component_C|GetRequiredClass (Variables|Default|GetCreativeUseRestrictionComponent)) 255) 0 (+ (Variables|/Game/Creative/Devices/Common/Components/Creative_UseRestriction_Component.Creative_UseRestriction_Component_C|GetRequiredClass (Variables|Default|GetCreativeUseRestrictionComponent)) 1))))

(fn GetSelectedTeamInternal
  ;[1 1 1] Return the required team for automated testing purposes
  (return (+ (Variables|/Game/Creative/Devices/Common/Components/Creative_UseRestriction_Component.Creative_UseRestriction_Component_C|GetRequiredTeam (Variables|Default|GetCreativeUseRestrictionComponent)) 1)))

(fn IsTargetableByLinearMover
  (return true))

(fn MigrateLegacyUserOptions
  (bind (_p_elimzone_cylinder _p_elimzone_box _zonedamageperiodlevel) (Class|GetClassDefaults "/CRD_VolumetricRegion/BP_Device_DamageVolume_V2.BP_Device_DamageVolume_V2_C"))
  (bind (_p_elimzone_cylinder_2 _p_elimzone_box_2 _zone_shape _shape_edit_mesh_and_mats _hideindicatordamagetag _on_agent_entering_zone _on_agent_exiting_zone _volumedamagetag _onphysicspropenter _onphysicspropexit _enabledonphase) (Class|GetClassDefaults "/CRD_VolumetricRegion/BP_Device_DamageVolume_V2.BP_Device_DamageVolume_V2_C"))
  (bind _zonedamageperiodlevel_2 (Variables|Legacy|GetZoneDamagePeriodLevel))
  (|CallParentFunction)
  (Utilities|FlowControl|Sequence
    (:then_0
      (Utilities|FlowControl|Sequence
        (:then_0
          (DeviceConversion|MigrateLegacyEnabledIndex (Variables|UserOptions|GetEnabledOnPhase) _enabledonphase))
        (:then_1
          (if (!= _zonedamageperiodlevel_2 _zonedamageperiodlevel)
            (switch int _zonedamageperiodlevel_2
              (:0
                (Variables|UserOptions|SetDamage%20Tick%20Rate "NewEnumerator0"))
              (:1
                (Variables|UserOptions|SetDamage%20Tick%20Rate "NewEnumerator1"))
              (:2
                (Variables|UserOptions|SetDamage%20Tick%20Rate "NewEnumerator2"))
              (:3
                (Variables|UserOptions|SetDamage%20Tick%20Rate "NewEnumerator4"))
              (:4
                (Variables|UserOptions|SetDamage%20Tick%20Rate "NewEnumerator5"))
              (:5
                (Variables|UserOptions|SetDamage%20Tick%20Rate "NewEnumerator6"))
              (:6
                (Variables|UserOptions|SetDamage%20Tick%20Rate "NewEnumerator7"))
              (:7
                (Variables|UserOptions|SetDamage%20Tick%20Rate "NewEnumerator8")))
            (Variables|Legacy|SetZoneDamagePeriodLevel _zonedamageperiodlevel))))
      (LocalVariables|SetOptionsMigrated true))
    (:then_1
      (if (LocalVariables|GetOptionsMigrated)
        (OptionData|RefreshPropertyOverrides (Variables|Default|GetToyOptionsComponent))))))

(fn Re-TriggerEffectsonVehicle
  (params (Object :type Fort%20Vehicle%20Interface%20Interface))
  ;[0.2308 0.1369 1] Remove then re-apply the effects on the vehicle in the event that it becomes occupied or vacant
  (bind _asfort_athena_vehicle (Utilities|Casting|CastToFortAthenaVehicle Object))
  (Default|EndEffect _asfort_athena_vehicle)
  (CallFunction|Non-PawnEnteredTriggerVolume _asfort_athena_vehicle))

(fn ReprocessOverlappingPlayersPlayers
  (params (Player%20Pawn :type Fort%20Player%20Pawn%20Object%20Reference) (Pawn%20Returned :type Boolean))
  (bind _playercontroller (Default|GetPlayerControllerfromActor Player%20Pawn))
  ;[0.2458 1 0.2784 0.5] If we have a player pawn, and are overlaping it, remove the applied effect and re-apply it in order to filter out unqualified players
  (Utilities|IsValid Player%20Pawn
    (:Is%20Valid
      (Utilities|IsValid _playercontroller
        (:Is%20Valid
          (if (and Pawn%20Returned (Collision|IsOverlappingActor self Player%20Pawn))
            (CallFunction|PawnExitedTriggerVolume Player%20Pawn _playercontroller)
            (CallFunction|PawnEnteredTriggerVolume Player%20Pawn _playercontroller)))))))

(fn ReturnMatchingGameplayEffectHandle
  (params (Ability%20System%20Component :type Ability%20System%20Component%20Object%20Reference))
  ;[0.3118 1 0.9701] Retrieve the Gameplay handle used to identify the gameplay effect.
  (bind (_array_element _array_index) (Utilities|Array|ForEachLoopwithBreak (Variables|Default|GetAbility%20Handle%20Array))
    (:LoopBody
      (if (Utilities|Equal(Object) Ability%20System%20Component (.AbilitySystemComponent_8_57E9F25741437CC55361B58F9DB743E0 (Utilities|Struct|BreakAbilityHandleandSystemHold _array_element)))
        (Variables|Default|SetActiveGameplayEffectHandle (Utilities|Struct|BreakAbilityHandleandSystemHold _array_element))
        (Utilities|Array|RemoveIndex (Variables|Default|GetAbility%20Handle%20Array) _array_index)))
    (:Completed
      _))
  (return (Variables|Default|GetActiveGameplayEffectHandle)))

(fn ReturnIsVehicleEmpty
  (params (Object :type Object%20Reference))
  ;[0.8591 1 0.3345] Return whether or not this vehicle is empty
  (bind _asfort_athena_vehicle (Utilities|Casting|CastToFortAthenaVehicle Object))
  (return (<= (Utilities|Array|Length (Vehicle|GetAllPassengers _asfort_athena_vehicle)) 0)))

(fn SendTriggerMessage
  (params (Trigger%20Message%20Component :type Fort%20Gameplay%20Trigger%20Message%20Component%20Object%20Reference) (Pawn :type Pawn%20Object%20Reference))
  (bind _returnvalue (Pawn|GetController Pawn))
  (Utilities|IsValid _returnvalue
    (:Is%20Valid
      (Class|FortGameplayTriggerMessageComponent|SendTriggerMessage Trigger%20Message%20Component _returnvalue))))

(fn SetDamage
  (params (NewDamage :type Integer))
  (Networking|SwitchHasAuthority
    (:Authority
      (Variables|Verse|SetDamageOverride (Math|Integer|Clamp(Integer) NewDamage 1 500))
      (Utilities|Array|Clear (LocalVariables|GetActorsToProcess))
      ;[1 1 1] Reset GEs for all currently affected actors. Note: if there was a way to just update the magnitude directly.. that would be preferable?
      (bind _array_element (Utilities|Array|ForEachLoopwithBreak (Variables|Default|GetAbility%20Handle%20Array))
        (:LoopBody
          (Utilities|IsValid (.AbilitySystemComponent_8_57E9F25741437CC55361B58F9DB743E0 (Utilities|Struct|BreakAbilityHandleandSystemHold _array_element))
            (:Is%20Valid
              (Utilities|Array|AddUnique (LocalVariables|GetActorsToProcess) _array_element))))
        (:Completed
          (bind _array_element_2 (Utilities|Array|ForEachLoopwithBreak (LocalVariables|GetActorsToProcess))
            (:LoopBody
              (Utilities|IsValid (Components|GetOwner (.AbilitySystemComponent_8_57E9F25741437CC55361B58F9DB743E0 (Utilities|Struct|BreakAbilityHandleandSystemHold _array_element_2)))
                (:Is%20Valid
                  (if (Utilities|DoesObjectImplementInterface (Components|GetOwner (.AbilitySystemComponent_8_57E9F25741437CC55361B58F9DB743E0 (Utilities|Struct|BreakAbilityHandleandSystemHold _array_element_2))) "/Script/FortniteGame.FortVehicleInterface")
                    (Default|EndEffect (Components|GetOwner (.AbilitySystemComponent_8_57E9F25741437CC55361B58F9DB743E0 (Utilities|Struct|BreakAbilityHandleandSystemHold _array_element_2))))
                    (CallFunction|Non-PawnEnteredTriggerVolume (Components|GetOwner (.AbilitySystemComponent_8_57E9F25741437CC55361B58F9DB743E0 (Utilities|Struct|BreakAbilityHandleandSystemHold _array_element_2))))
                    (else
                      (bind _asplayer_state (Utilities|Casting|CastToPlayerState (Components|GetOwner (.AbilitySystemComponent_8_57E9F25741437CC55361B58F9DB743E0 (Utilities|Struct|BreakAbilityHandleandSystemHold _array_element_2))))
                        (:then
                          (Utilities|IsValid _asplayer_state
                            (:Is%20Valid
                              (Utilities|IsValid (PlayerState|GetPlayerController _asplayer_state)
                                (:Is%20Valid
                                  (Utilities|IsValid (Pawn|GetControlledPawn (PlayerState|GetPlayerController _asplayer_state))
                                    (:Is%20Valid
                                      (Default|EndEffect 0 (Pawn|GetControlledPawn (PlayerState|GetPlayerController _asplayer_state)))
                                      (CallFunction|PawnEnteredTriggerVolume (Pawn|GetControlledPawn (PlayerState|GetPlayerController _asplayer_state)) (PlayerState|GetPlayerController _asplayer_state)))))))))
                        (:CastFailed
                          (bind _asfort_aipawn (Utilities|Casting|CastToFortAIPawn (Components|GetOwner (.AbilitySystemComponent_8_57E9F25741437CC55361B58F9DB743E0 (Utilities|Struct|BreakAbilityHandleandSystemHold _array_element_2)))))
                          (Default|EndEffect 0 _asfort_aipawn)
                          (CallFunction|PawnEnteredTriggerVolume _asfort_aipawn))))))))))))))

(fn ShouldBotbeAffected
  (params (BotPawn :type Pawn%20Object%20Reference))
  (bind _asfort_player_state_athena (Utilities|Casting|CastToFortPlayerStateAthena (Variables|/Script/Engine.Pawn|GetPlayerState BotPawn))
    (:then
      (return (or (and (Variables|UserOptions|GetAffectsPlayers) (Class|FortPlayerStateAthena|IsAnAthenaGameParticipant _asfort_player_state_athena)) (and (not (Class|FortPlayerStateAthena|IsAnAthenaGameParticipant _asfort_player_state_athena)) (Variables|UserOptions|GetAffectsGuards)))))
    (:CastFailed
      (return false))))

(fn ShouldInvisiblePropsBeAllowed
  (return true))

(fn TriggerEffect
  (params (Pawn :type Pawn%20Object%20Reference) (Vehicle :type Actor%20Object%20Reference) (PhysicsObject :type Actor%20Object%20Reference))
  ;[0.1639 0.174 1] Filter out if we're affecting players or vehicles
  (if (Utilities|IsValid Pawn :!class /Script/Engine.KismetSystemLibrary)
    ;[1 0.06152 0] If the options are set to eliminate, force the pawn to die
    (switch Utilities|FlowControl|Switch|SwitchonEDamageVolumeDamageType (Variables|UserOptions|GetDamageType)
      (:NewEnumerator1
        (bind _asfort_pawn (Utilities|Casting|CastToFortPawn Pawn))
        (FortPawn|ForceKill _asfort_pawn "(TagName=\"Gameplay.Effect.InstantDeath\")" 0 self))
      (:NewEnumerator0
        ;[0.3017 1 0.2404] Apply the GE
        (CallFunction|ApplyGE Pawn (Utilities|GetCurrentDamage))))
    (elif (CallFunction|CheckVehicleValidity Vehicle)
      (if (CallFunction|ReturnIsVehicleEmpty Vehicle)
        _
        (elif (CallFunction|CheckVehiclePassengerTeamValidity Vehicle)
          _))
      (switch Utilities|FlowControl|Switch|SwitchonEDamageVolumeDamageType (Variables|UserOptions|GetDamageType)
        (:NewEnumerator1
          (CallFunction|ApplyGE Vehicle 9999999) ; At this time no function exists to force vehicles to die, so we're going to do this the dirty way and apply more damage then health ever
        )
        (:NewEnumerator0
          (CallFunction|ApplyGE Vehicle (Utilities|GetCurrentDamage))))
      (else
        (bind _returnvalue (FortPhysicsSimulationLibrary|IsPhysicsObjectActiveinSimulation PhysicsObject))
        (if (and _returnvalue (Fortnite|IsPhysicallySimulatedIsland))
          (switch Utilities|FlowControl|Switch|SwitchonEDamageVolumeDamageType (Variables|UserOptions|GetDamageType)
            (:NewEnumerator1
              (CallFunction|ApplyGE PhysicsObject 9999999) ; At this time no function exists to force vehicles to die, so we're going to do this the dirty way and apply more damage then health ever
            )
            (:NewEnumerator0
              (CallFunction|ApplyGE PhysicsObject (Utilities|GetCurrentDamage)))))))))

(fn UnbindVehicleEvents
  (params (FortAthenaVehicle :type Fort%20Athena%20Vehicle%20Object%20Reference))
  ;[1 0.6892 0.2661] Unbind all events where players enter or leave the specific vehicle
  (Vehicle|UnbindEventfromOnPawnEnteredVehicleSeat FortAthenaVehicle (EventDispatchers|CreateEvent :!event Player%20Entered%20Vehicle))
  (Vehicle|UnbindEventfromOnPawnExitedVehicleSeat FortAthenaVehicle (EventDispatchers|CreateEvent :!event Player%20Exited%20Vehicle)))

(fn UpdateCreativeVolumeFrameOnWorldReady
  ;[1 0.8085 0.09327] Plug in our values that can't be directly referenced but can be made transient
  (Variables|/Game/Creative/Devices/Common/Components/CreativeVolumeFrame.CreativeVolumeFrame_C|SetShapeEditMeshAndMats (Variables|Default|GetShape%20Edit%20Mesh%20And%20Mats) (Variables|Default|GetCreativeVolumeFrame)))

(fn ConstructionScript
  (|CallParentFunction))

(fn VFXParameters
  (bind _value (Utilities|Map|Find (Variables|Default|GetZoneEdgesSize) (Variables|UserOptions|GetZoneShape)))
  (bind (_value_x _value_y _value_z) (Math|Vector|BreakVector _value))
  (bind (_returnvalue_x _returnvalue_y _returnvalue_z) (Math|Vector|BreakVector (* (Math|Vector|MakeVector 512.0 512.0 384.0) (Math|Vector|MakeVector (Variables|UserOptions|GetZoneWidth) (Variables|UserOptions|GetZoneDepth) (Variables|UserOptions|GetZoneHeight)))))
  (return (Utilities|Array|MakeArray (Utilities|Struct|MakeParticleSysParam "BlocksVolume" "PSPT_VectorRand" 0.0 0.0 (Variables|Default|GetBlocksVolumeHigh) (Variables|Default|GetBlocksVolumeLow)) (Utilities|Struct|MakeParticleSysParam "BlocksSpawnRateScale" "PSPT_Scalar" (Utilities|ClampMaxFloat (/ (Math|Float|float*float _returnvalue_x _returnvalue_y _returnvalue_z) 100663296.0) 1e+01) 0.0 "0.000000,0.000000,0.000000" (0.0 0.0 0.0) "(B=0,G=0,R=0,A=0)" 0 0) (Utilities|Struct|MakeParticleSysParam "MeshSize" "PSPT_Vector" 0.0 0.0 (Math|Vector|MakeVector _value_x _value_y (/ _value_z (Variables|UserOptions|GetZoneHeight))) "0.000000,0.000000,0.000000" "(B=0,G=0,R=0,A=0)" 0 0) (Utilities|Struct|MakeParticleSysParam "BlocksSize" "PSPT_Vector" 0.0 0.0 (Math|Vector|MakeVector (Utilities|Reciprocal (Variables|UserOptions|GetZoneWidth)) (Utilities|Reciprocal (Variables|UserOptions|GetZoneWidth))) "0.000000,0.000000,0.000000" "(B=0,G=0,R=0,A=0)" 0 0) (Utilities|Struct|MakeParticleSysParam "BlocksVelocity" "PSPT_Vector" 0.0 0.0 (Math|Vector|MakeVector 0.0 0.0 (* -2e+01 (Utilities|Reciprocal (Variables|UserOptions|GetZoneHeight)))) "0.000000,0.000000,0.000000" "(B=0,G=0,R=0,A=0)" 0 0))))
