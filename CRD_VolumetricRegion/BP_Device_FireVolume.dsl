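; Fire Volume Creative device: a resizable zone that ignites or extinguishes objects and grass inside it through Curie fire and water interactions.
; Ignite clears water handles, then requests a throttled fire start on overlapping actors, tracking handles in a map.
; Extinguish ends fire and applies water to block re-ignition from nearby flames.
; Minigame states register the volume with the minigame Curie component, while enabled state and options drive spark visibility and material.
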(event EventUpdateVolumeSize
  (bind _output_get (Variables|Default|SetNewScale (Math|Vector|MakeVector (Variables|Default|GetZoneXScale) (Variables|Default|GetZoneYScale) (Variables|Default|GetZoneZScale))))
  (Transformation|SetRelativeScale3D (Variables|Default|GetPlaySpaceVolume) _output_get)
  (bind _location (Resizing|GetTriggerLocation))
  (Transformation|SetRelativeLocation (Variables|Default|GetTrigger) _location false true)
  (CallFunction|ResizeFrame)
  (VFX|UpdateVolumeSizeParticles))

(event EventBeginPlay
  (|CallParentFunction)
  (Transformation|SetWorldScale3D (Variables|Default|GetInteract%20Mesh) "1.000000,1.000000,1.000000"))

(event EventInitializeDevice
  ;[0.1244 0.1589 1] Initialize Device
  ; Apply Fire Ignition Settings
  (|CallParentFunction)
  (Rendering|SetVisibility (Variables|Default|GetPlaySpaceVolume) (not (Building|IsPreviewBuildingActor)) false)
  (CallFunction|InitializeInteractMesh)
  (Development|SetHiddeninGame (Variables|Default|GetPlaySpaceVolume) (Building|IsPreviewBuildingActor) false)
  (CallFunction|HandleSparks))

(event Custom|RemoveWaterForDisable
  (bind _waterinteracthandles (Variables|Default|GetWaterInteractHandles))
  ;[0.6278 1 0.176] Ignite When Received
  ; 1. End any curie water interaction
  ; 2. Begin fire interaction with any overlapping actors
  ; 3. Ignite Grass in Bounds
  ; 4. Send transmit ignite trigger message
  ;;[0.08627 0.5882 1] End Water Interactions
  (bind _keys (Utilities|Map|Keys _waterinteracthandles))
  (for _array_element _keys
    (if (.ReturnValue (Utilities|Map|Find _waterinteracthandles _array_element))
      (Curie|ExecuteEndWaterInteract _array_element (Utilities|Map|Find _waterinteracthandles _array_element))
      (Utilities|Map|Remove _waterinteracthandles _array_element))))

(event Custom|Ignite%20Event
  (params (Instigator :type Controller%20Object%20Reference))
  (bind _waterinteracthandles (Variables|Default|GetWaterInteractHandles))
  ;[0.6278 1 0.176] Ignite When Received
  ; 1. End any curie water interaction
  ; 2. Begin fire interaction with any overlapping actors
  ; 3. Ignite Grass in Bounds
  ; 4. Send transmit ignite trigger message
  (if (Math|Boolean|ANDBoolean (CallFunction|IsinPlayMode) (Utilities|Enum|NotEqual(Enum) (Variables|FireOptions|GetAllowObjectsToIgnite) "NewEnumerator2") (Class|CreativeEnabledComponent|IsEnabled (Variables|Default|GetEnabledComponent)))
    (Utilities|FlowControl|Sequence
      (:then_0
        ;;[0.08627 0.5882 1] End Water Interactions
        (bind _keys (Utilities|Map|Keys _waterinteracthandles))
        (for _array_element _keys
          (if (.ReturnValue (Utilities|Map|Find _waterinteracthandles _array_element))
            (Curie|ExecuteEndWaterInteract _array_element (Utilities|Map|Find _waterinteracthandles _array_element))
            (Utilities|Map|Remove _waterinteracthandles _array_element))))
      (:then_1
        ;;[0.7412 0.3059 1] Start fire interaction with overlapping actors and props
        ;;;[1 1 1] The curie code requires the instigator to have the AbilitySystem component
        (Utilities|IsValid Instigator
          (:Is%20Valid
            (Utilities|IsValid (Ability|GetAbilitySystemComponent Instigator)
              (:Is%20Valid
                (Variables|Default|SetInstigatorForFire Instigator))
              (:Is%20Not%20Valid
                (Variables|Default|SetInstigatorForFire self)))
            (if (not (Variables|Default|GetWaitingForThrottle))
              (CallFunction|BeginFireandAddFireHandleThrottled)))
          (:Is%20Not%20Valid
            _))))))

(event Custom|Completed%20Throttle
  (params (Succeeded :type Boolean))
  ;[0.6278 1 0.176] Ignite When Received
  ; 1. End any curie water interaction
  ; 2. Begin fire interaction with any overlapping actors
  ; 3. Ignite Grass in Bounds
  ; 4. Send transmit ignite trigger message
  ;;[0.7728 0.08274 1] Start Fire Throttled
  (CallFunction|BeginFireAndAddFireHandle)
  (Variables|Default|SetWaitingForThrottle false))

(event Custom|End%20Fire%20Interactions
  (bind _fireinteracthandles (Variables|Default|GetFireInteractHandles))
  ;[1 0.6238 0.0603] Extinguish When Received
  ; 1. End any fire interactions
  ; 2. Apply curie water interaction (so objects don't get re-ignited by any surrounding landscape fires when they should be extinguished)
  ; 3. Send extinguish trigger message
  ;;[0.6202 0.2603 1] End fire interactions
  (bind _keys (Utilities|Map|Keys _fireinteracthandles))
  (for _array_element _keys
    (if (.ReturnValue (Utilities|Map|Find _fireinteracthandles _array_element))
      (Curie|ExecuteEndFireInteract _array_element (Utilities|Map|Find _fireinteracthandles _array_element))
      (Utilities|Map|Remove _fireinteracthandles _array_element))))

(event Custom|Extinguish%20Event
  (params (Instigator :type Controller%20Object%20Reference))
  (bind _fireinteracthandles (Variables|Default|GetFireInteractHandles))
  ;[1 0.6238 0.0603] Extinguish When Received
  ; 1. End any fire interactions
  ; 2. Apply curie water interaction (so objects don't get re-ignited by any surrounding landscape fires when they should be extinguished)
  ; 3. Send extinguish trigger message
  (if (and (CallFunction|IsinPlayMode) (Class|CreativeEnabledComponent|IsEnabled (Variables|Default|GetEnabledComponent)))
    (Utilities|FlowControl|Sequence
      (:then_0
        ;;[0.6202 0.2603 1] End fire interactions
        (bind _keys (Utilities|Map|Keys _fireinteracthandles))
        (for _array_element _keys
          (if (.ReturnValue (Utilities|Map|Find _fireinteracthandles _array_element))
            (Curie|ExecuteEndFireInteract _array_element (Utilities|Map|Find _fireinteracthandles _array_element))
            (Utilities|Map|Remove _fireinteracthandles _array_element))))
      (:then_1
        ;;[0 0.2366 1] Begin water interaction for any overlapping actors
        (Utilities|FlowControl|Sequence
          (:then_0
            (CallFunction|RemoveWaterforDisable))
          (:then_1
            (bind (_overlappedactors _returnvalue) (CallFunction|GetOverlappingActors))
            (if _returnvalue
              (CallFunction|BeginWaterAndAddWaterHandle _overlappedactors Instigator)))))
      (:then_2
        ;;[0 0.6384 1] Extinguish Grass In Bounds
        (bind _gameplayvolume (CallFunction|GetGameplayVolume))
        (Utilities|IsValid _gameplayvolume
          (:Is%20Valid
            (Curie|ExtinguishGrassinBounds (Math|Box|MakeBox (- (Collision|GetActorBounds _gameplayvolume false false) (Math|Vector|MakeVector (* (.x (.BoxExtent (Collision|GetActorBounds _gameplayvolume false false))) (Variables|Default|GetExtinguishFireMultiplier)) (* (.y (.BoxExtent (Collision|GetActorBounds _gameplayvolume false false))) (Variables|Default|GetExtinguishFireMultiplier)) (.z (.BoxExtent (Collision|GetActorBounds _gameplayvolume false false))))) (+ (Collision|GetActorBounds _gameplayvolume false false) (Math|Vector|MakeVector (* (.x (.BoxExtent (Collision|GetActorBounds _gameplayvolume false false))) (Variables|Default|GetExtinguishFireMultiplier)) (* (.y (.BoxExtent (Collision|GetActorBounds _gameplayvolume false false))) (Variables|Default|GetExtinguishFireMultiplier)) (.z (.BoxExtent (Collision|GetActorBounds _gameplayvolume false false)))))))))))))

(event Custom|ApplyWaterCurie
  ;[1 0.6238 0.0603] Extinguish When Received
  ; 1. End any fire interactions
  ; 2. Apply curie water interaction (so objects don't get re-ignited by any surrounding landscape fires when they should be extinguished)
  ; 3. Send extinguish trigger message
  ;;[0 0.2366 1] Begin water interaction for any overlapping actors
  (Utilities|FlowControl|Sequence
    (:then_0
      (CallFunction|RemoveWaterforDisable))
    (:then_1
      (bind (_overlappedactors _returnvalue) (CallFunction|GetOverlappingActors))
      (if _returnvalue
        (CallFunction|BeginWaterAndAddWaterHandle _overlappedactors (stale Custom|Extinguish%20Event Instigator))))))

(event Custom|CollectObjectsInVolume
  ;[0.116 1 0.3511] Find Objects In Volume - Get the objects within the volume, check if they're a building prop, and if so, add to Objects In Volume Array.
  (Networking|SwitchHasAuthority
    (:Authority
      ;;[1 0.9506 0.4375] Should these props be considered invulnerable? Save the result prior to entering the loop.
      (Variables|Default|SetbPropsInvulnerable (CallFunction|ArePropsInvulnerable))
      ;;[1 0.8273 0] Don't clear ObjectsInVolume until you are sure the objects have been returned to their default state. Once collision has been disabled they will be undetectable and updating tracked actors won't return anything.
      (bind _trackedactors (CallFunction|UpdateAndGetTrackedActors))
      (for _array_element _trackedactors
        (bind _asbuilding_static_mesh (Utilities|Casting|CastToBuildingSMActor _array_element))
        (if (not (GameplayTags|HasMatchingGameplayTag _asbuilding_static_mesh "(TagName=\"Asset.Creative.Device\")"))
          (Utilities|Array|Add (Variables|Default|GetPropObjectsInVolume) _asbuilding_static_mesh))))))

(event MinigameStartedDelegate(FortMinigameProgress)
  (Class|CreativeEnabledComponent|SetEnabled (Variables|Default|GetEnabledComponent) (Variables|UserOptions|GetEnabledOnStart)))

(event MinigameEndedDelegate(FortMinigameProgress)
  (Class|CreativeEnabledComponent|SetEnabled (Variables|Default|GetEnabledComponent) true)
  (Development|SetHiddeninGame (Variables|Default|GetPlaySpaceVolume) false false)
  (CallFunction|HandleSparks)
  (Networking|SwitchHasAuthority
    (:Authority
      (CallFunction|RemoveWaterforDisable))))

(event OnEnabledStateChanged(EnabledComponent)
  (params (Enabled :type Boolean))
  (if (CallFunction|IsinPlayMode)
    (CallFunction|SetEnabledStateVisibility Enabled)
    (Networking|SwitchHasAuthority
      (:Authority
        (if (and (Utilities|Enum|Equal(Enum) (Variables|FireOptions|GetAllowObjectsToIgnite) "NewEnumerator2") Enabled)
          (CallFunction|ApplyWaterCurie)
          (else
            (CallFunction|RemoveWaterforDisable)))))))

(event MinigameStateChangedDelegate(FortMinigameProgress)
  (params (Minigame :type Fort%20Minigame%20Object%20Reference) (NewMinigameState :type EFortMinigameState%20Enum :default "PreGame"))
  (bind _self self)
  (bind _returnvalue (Minigame|GetMinigame _self :!class /Script/FortniteGame.MinigameBlueprintLibrary))
  (bind _self_2 self)
  (bind _returnvalue_2 (Minigame|GetMinigame _self_2 :!class /Script/FortniteGame.MinigameBlueprintLibrary))
  (bind _curiecomponent (Variables|/Script/FortniteGame.FortMinigame|GetCurieComponent _returnvalue))
  (bind _curiecomponent_2 (Variables|/Script/FortniteGame.FortMinigame|GetCurieComponent _returnvalue_2))
  ;[0.01424 0.276 0.05245] Process Minigame State Changed
  (switch Utilities|FlowControl|Switch|SwitchonEFortMinigameState NewMinigameState
    (:Setup
      (Networking|SwitchHasAuthority
        (:Authority
          (Utilities|IsValid _returnvalue_2
            (:Is%20Valid
              (bind _gameplayvolume (CallFunction|GetGameplayVolume))
              (Utilities|IsValid _gameplayvolume
                (:Is%20Valid
                  (Utilities|IsValid _curiecomponent_2
                    (:Is%20Valid
                      (Class|FortMinigameCurieComponent|AddtoFireVolumesforMinigame _curiecomponent_2 _gameplayvolume (Utilities|Enum|BytetoEnumEDeviceFireVolumeFireOption (Math|Conversions|ToByte(Integer) (Math|Conversions|ToInteger(Byte) (Variables|FireOptions|GetAllowObjectsToIgnite))))))))))))))
    (:WaitingForCameras
      (Development|SetHiddeninGame (Variables|Default|GetPlaySpaceVolume) true false)
      (Class|CreativeEnabledComponent|SetEnabled (Variables|Default|GetEnabledComponent) false))
    (:PostGameTimeDilation
      (Networking|SwitchHasAuthority
        (:Authority
          (Utilities|IsValid _returnvalue
            (:Is%20Valid
              (bind _gameplayvolume_2 (CallFunction|GetGameplayVolume))
              (Utilities|IsValid _gameplayvolume_2
                (:Is%20Valid
                  (Utilities|IsValid _curiecomponent
                    (:Is%20Valid
                      (Class|FortMinigameCurieComponent|RemovefromFireVolumesforMinigame _curiecomponent _gameplayvolume_2))))))))))))

(fn ArePropsInvulnerable
  ;[0 1 0.04802 0.5] If PropHealth is equal to PropHealthInvulnerable, it means the prop is invulnerable and no damage should be applied. We should not set bInvulnerable here, becuase doing so will not allow resources drop.
  (return (== (Variables|FireOptions|GetPropHealth) (Variables|Default|GetPropHealthInvulnerable))))

(fn BeginFireandAddFireHandle
  ;[0.07181 0.3179 1] Begin Fire and add the fire handle to a map to extinguish later
  (Utilities|IsValid (Variables|Default|GetInstigatorForFire)
    (:Is%20Valid
      (Utilities|FlowControl|Sequence
        (:then_0
          (CallFunction|EndFireInteractions))
        (:then_1
          (bind (_overlappedactors _returnvalue) (CallFunction|GetOverlappingActors))
          (if _returnvalue
            (bind _array_element (Utilities|Array|ForEachLoopwithBreak _overlappedactors)
              (:LoopBody
                (Utilities|IsValid _array_element
                  (:Is%20Valid
                    (bind _returnvalue_2 (Curie|ExecuteBeginFireInteract :Instigator (Variables|Default|GetInstigatorForFire) :EffectCauser _array_element :Target _array_element :Magnitude 1.0 :PropagationFuel (Variables|Default|GetPropagation%20Fuel) :LandscapeIgnitionRadius (Variables|Default|GetLandscapIgnitionRadius)))
                    (Utilities|Map|Add (Variables|Default|GetFireInteractHandles) _array_element _returnvalue_2)
                    ;;[0.12 0.9103 1] If we have exceeded our Max Objects On Fire then Break
                    (if (<= (Utilities|Map|Length (Variables|Default|GetFireInteractHandles)) (Variables|Default|GetMaxObjectsOnFire))
                      _
                      (else
                        _))))))))
        (:then_2
          ;;[0 0.5 0.4872] Ignite Grass in Bounds.
          (bind _gameplayvolume (CallFunction|GetGameplayVolume))
          (Utilities|IsValid _gameplayvolume
            (:Is%20Valid
              (Curie|IgniteGrassinBounds (Variables|Default|GetInstigatorForFire) 0 (Math|Box|MakeBox (- (Collision|GetActorBounds _gameplayvolume false false) (.BoxExtent (Collision|GetActorBounds _gameplayvolume false false))) (+ (Collision|GetActorBounds _gameplayvolume false false) (.BoxExtent (Collision|GetActorBounds _gameplayvolume false false)))) -1))))))))

(fn BeginFireAndAddFireHandleThrottled
  ;[0.4337 1 0.2604] Use the Creative Throttle Manager to throttle the Begin Fire and Add Fire Handle
  (bind _returnvalue (Class|CreativeBlueprintLibrary|GetCreativeGeneralThrottleManager))
  (Utilities|IsValid _returnvalue
    (:Is%20Valid
      (Variables|Default|SetWaitingForThrottle true)
      (Class|FortCreativeGeneralThrottleManager|RequestThrottleObject _returnvalue self (Math|Conversions|ToInteger(Byte) (Utilities|Enum|LiteralenumEGeneralThrottleBuckets "NewEnumerator2")) (EventDispatchers|CreateEvent :!event Completed%20Throttle)))))

(fn BeginWaterandAddWaterHandle
  (params (Actor :type Array%20of%20Actor%20Object%20References) (Instigator :type Actor%20Object%20Reference))
  ;[0.07181 0.3179 1] Begin Water and add the water handle to a map in order to end interaction if fire is desired again
  (for _array_element Actor
    (Utilities|IsValid _array_element
      (:Is%20Valid
        (bind _returnvalue (Curie|ExecuteBeginWaterInteract :Instigator Instigator :EffectCauser _array_element :Target _array_element :Magnitude 1.0 :LandscapeExtinguishRadius (Variables|Default|GetLandscapIgnitionRadius)))
        (Utilities|Map|Add (Variables|Default|GetWaterInteractHandles) _array_element _returnvalue)))))

(fn GetandCacheActorTracker
  ;[0 1 0.1098 0.5] Get the Actor Tracker component from the Play Space and cache. Early return if already cached.
  (bind _cachedactortracker (Variables|Default|GetCachedActorTracker)
    (:then
      (return _cachedactortracker))
    (:else
      (bind _gameplayvolume (CallFunction|GetGameplayVolume))
      (Utilities|IsValid _gameplayvolume
        (:Is%20Valid
          (bind _asfire_volume_playspace (Utilities|Casting|CastToFireVolumePlayspace (Class|GameplayVolume|GetPlayspace _gameplayvolume))
            (:then
              (bind _playspacecomponent_spatialactortracker (Variables|/CRD_VolumetricRegion/SetupAssets/TID/FireVolume/FireVolumePlayspace.FireVolumePlayspace_C|GetPlayspaceComponent_SpatialActorTracker _asfire_volume_playspace)
                (:then
                  (bind _output_get (Variables|Default|SetCachedActorTracker _playspacecomponent_spatialactortracker))
                  (return _output_get))
                (:else
                  _)))
            (:CastFailed
              _))
          (return 0))
        (:Is%20Not%20Valid
          _)))))

(fn GetOverlappingActors
  ;[0.3979 1 0.6978] Get Overlapping Actors Based On Requirement Types
  (bind _gameplayvolume (CallFunction|GetGameplayVolume))
  (Utilities|IsValid _gameplayvolume
    (:Is%20Valid
      (bind (_outactors _returnvalue) (Collision|BoxOverlapActors (Collision|GetActorBounds _gameplayvolume false false) (.BoxExtent (Collision|GetActorBounds _gameplayvolume false false)) (Variables|Default|GetMaterialsObjectTypes) (Variables|Default|GetMaterialsClassFilter) (Utilities|Array|MakeArray self) :!class /Script/Engine.KismetSystemLibrary))
      (return _outactors _returnvalue))))

(fn GetTrackedActors
  ;[0 1 0.2751] Get the tracked actors from the Actor Tracker, convert the result to an array and return same.
  (bind _playspacecomponent_spatialactortracker (CallFunction|GetAndCacheActorTracker))
  (Utilities|IsValid _playspacecomponent_spatialactortracker
    (:Is%20Valid
      (bind _result (Utilities|Set|ToArray (Class|PlayspaceComponentSpatialActorTracker|GetTrackedActors _playspacecomponent_spatialactortracker)))
      (return _result))
    (:Is%20Not%20Valid
      (return 0))))

(fn GetCollisionStaticMesh
  (return "0, 0, 0" (Variables|/Script/Engine.StaticMeshComponent|GetStaticMesh (Variables|Default|GetInteract%20Mesh))))

(fn GetGameplayVolume
  ;[0 1 0.1724] Getter for the Gameplay Volume.
  (bind _cachedgameplayvolume (Variables|Default|GetCachedGameplayVolume)
    (:then
      (return _cachedgameplayvolume))
    (:else
      (bind _playspacevolume (Variables|Default|GetPlaySpaceVolume)
        (:then
          (bind _asfire_volume_gameplay_volume (Utilities|Casting|CastToFireVolumeGameplayVolume (Variables|/Script/Engine.ChildActorComponent|GetChildActor _playspacevolume))
            (:then
              (bind _output_get (Variables|Default|SetCachedGameplayVolume _asfire_volume_gameplay_volume))
              (return _output_get))
            (:CastFailed
              _)))
        (:else
          _))
      (return 0))))

(fn GetMeshComponents
  (return (Utilities|Array|MakeArray (Variables|Default|GetInteract%20Mesh) (Variables|Default|GetCreativeEditOnlyMesh))))

(fn GetVisualComponents
  (return (Utilities|Array|MakeArray (Variables|Default|GetInteract%20Mesh) (Variables|Default|GetPlaySpaceVolume) (Variables|Default|GetCreativeEditOnlyMesh)) (Utilities|Array|MakeArray (Variables|Default|GetNS_FireVolume_Effect) (Variables|Default|GetZoneMesh))))

(fn HandleSparks
  (Development|SetHiddeninGame (Variables|Default|GetNS_FireVolume_Effect) (Building|IsPreviewBuildingActor) false)
  (bind _playspacevolume (Variables|Default|GetPlaySpaceVolume)
    (:then
      (bind _asfire_volume_gameplay_volume (Utilities|Casting|CastToFireVolumeGameplayVolume (Variables|/Script/Engine.ChildActorComponent|GetChildActor _playspacevolume)))
      (bind _overlapmesh (Variables|/CRD_VolumetricRegion/SetupAssets/TID/FireVolume/FireVolumeGameplayVolume.FireVolumeGameplayVolume_C|GetOverlapMesh _asfire_volume_gameplay_volume)
        (:then
          (Rendering|Material|SetMaterial _overlapmesh 0 (select (Utilities|Enum|Equal(Enum) (Variables|FireOptions|GetAllowObjectsToIgnite) "NewEnumerator2") (Variables|/CRD_VolumetricRegion/SetupAssets/TID/FireVolume/FireVolumeGameplayVolume.FireVolumeGameplayVolume_C|GetFireInactive _asfire_volume_gameplay_volume) (Variables|/CRD_VolumetricRegion/SetupAssets/TID/FireVolume/FireVolumeGameplayVolume.FireVolumeGameplayVolume_C|GetFireActive _asfire_volume_gameplay_volume))))
        (:else)))
    (:else)))

(fn IgnoreWorldGridScaleOverride
  (return false))

(fn InitializeInteractMesh
  (bind _interact_mesh (Variables|Default|GetInteract%20Mesh))
  (Transformation|SetWorldScale3D _interact_mesh "1.000000,1.000000,1.000000")
  (if (not (Editor|InEditor))
    (Class|CreativeEditOnlyMeshComponent|ReloadMeshandMaterials _interact_mesh)))

(fn SetEnabledStateVisibility
  (params (Enabled :type Boolean))
  (Development|SetHiddeninGame (Variables|Default|GetNS_FireVolume_Effect) (not (and Enabled (Variables|DeviceOptions|GetZoneVisibleInGame))) false))

(fn UpdateandGetTrackedActors
  ;[0 1 0.1376] Get the Actor Tracker from the playspace volume and refresh the tracked actors utilizing the VolumetricRegionNativeComponent. Then get and return the tracked actors.
  (bind _playspacecomponent_spatialactortracker (CallFunction|GetAndCacheActorTracker))
  (Utilities|IsValid _playspacecomponent_spatialactortracker
    (:Is%20Valid
      (bind _volumetricregionnative (Variables|Default|GetVolumetricRegionNative)
        (:then
          (Class|VolumetricRegionNativeComponent|RefreshTrackedActors _volumetricregionnative _playspacecomponent_spatialactortracker)
          (bind _result (CallFunction|GetTrackedActors))
          (return _result))
        (:else)))
    (:Is%20Not%20Valid
      (return 0))))

(fn UpdateVolumeSizeParticles
  (bind _zonezscale (Variables|Default|GetZoneZScale))
  (bind _zoneyscale (Variables|Default|GetZoneYScale))
  (bind _zonexscale (Variables|Default|GetZoneXScale))
  (bind _returnvalue (Math|Vector|MakeVector _zonexscale _zoneyscale _zonezscale))
  (bind _ns_firevolume_effect (Variables|Default|GetNS_FireVolume_Effect))
  ;[1 0.9828 0.5683] Particle Man, Particle Man
  (Components|Activation|Deactivate (Variables|Default|GetNS_FireVolume_Effect))
  (bind _output_get (Variables|Default|SetNewScale _returnvalue))
  (Transformation|SetRelativeScale3D _ns_firevolume_effect _output_get)
  (Effects|Components|ParticleSystem|SetFloatParameter _ns_firevolume_effect "Spawnrate" (Math|Vector|VectorLength _returnvalue))
  (Rendering|SetBoundsScale (Variables|Default|GetNS_FireVolume_Effect) (Math|Float|Max(Float) (Variables|Default|GetZoneXScale) (Variables|Default|GetZoneYScale) (Variables|Default|GetZoneZScale) 7.2e+03))
  (bind _location (Resizing|GetTriggerLocation))
  (Transformation|SetRelativeLocation (Variables|Default|GetTrigger) _location false true)
  (Effects|Components|ParticleSystem|SetFloatParameter (Variables|Default|GetNS_FireVolume_Effect) "Extinguish" (Math|Float|SelectFloat 0.0 1.0 (Utilities|Enum|NotEqual(Enum) (Variables|FireOptions|GetAllowObjectsToIgnite) "NewEnumerator2")))
  (Components|Activation|Activate (Variables|Default|GetNS_FireVolume_Effect) false))

(fn ConstructionScript
  (|CallParentFunction))

(fn VFXParameters
  (bind (_returnvalue_x _returnvalue_y _returnvalue_z) (Math|Vector|BreakVector (* (Math|Vector|MakeVector 512.0 512.0 384.0) (Math|Vector|MakeVector (Variables|Default|GetZoneXScale) (Variables|Default|GetZoneYScale) (Variables|Default|GetZoneZScale)))))
  (bind (_zoneedgessize_x _zoneedgessize_y _zoneedgessize_z) (Math|Vector|BreakVector (Variables|GetZoneEdgesSize_X)))
  (return (Utilities|Array|MakeArray (Utilities|Struct|MakeParticleSysParam "BlocksVolume" "PSPT_VectorRand" 0.0 0.0 (Variables|Default|GetBlocksVolumeHigh) (Variables|Default|GetBlocksVolumeLow)) (Utilities|Struct|MakeParticleSysParam "BlocksSpawnRateScale" "PSPT_Scalar" (Utilities|ClampMaxFloat (/ (Math|Float|float*float _returnvalue_x _returnvalue_y _returnvalue_z) 100663296.0) 1e+01) 0.0 "0.000000,0.000000,0.000000" (0.0 0.0 0.0) "(B=0,G=0,R=0,A=0)" 0 0) (Utilities|Struct|MakeParticleSysParam "MeshSize" "PSPT_Vector" 0.0 0.0 (Math|Vector|MakeVector _zoneedgessize_x _zoneedgessize_y (/ _zoneedgessize_z (Variables|Default|GetZoneZScale))) "0.000000,0.000000,0.000000" "(B=0,G=0,R=0,A=0)" 0 0) (Utilities|Struct|MakeParticleSysParam "BlocksSize" "PSPT_Vector" 0.0 0.0 (Math|Vector|MakeVector (Utilities|Reciprocal (Variables|Default|GetZoneXScale)) (Utilities|Reciprocal (Variables|Default|GetZoneXScale))) "0.000000,0.000000,0.000000" "(B=0,G=0,R=0,A=0)" 0 0) (Utilities|Struct|MakeParticleSysParam "BlocksVelocity" "PSPT_Vector" 0.0 0.0 (Math|Vector|MakeVector 0.0 0.0 (* -2e+01 (Utilities|Reciprocal (Variables|Default|GetZoneZScale)))) "0.000000,0.000000,0.000000" "(B=0,G=0,R=0,A=0)" 0 0))))
