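; Creative Mutator Volume device that changes rules for players, creatures, bots and vehicles inside its zone, such as building, weapons, jumping, speed and gravity.
; Overlaps pass team, class and type filters, then apply option-selected gameplay effects, with AI-specific speed replacements.
; Stored handles and affected-actor lists allow removal on exit or game end, and re-evaluation on team, class or vehicle occupancy changes.
; It also broadcasts enter, exit and emote events and overrides pickup despawn times.
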(event EventUpdateVolumeSize
  ;[0.5705 1 0.7202] Override parent here to handle child-specific logic
  (|CallParentFunction)
  (Transformation|SetRelativeScale3D (Variables|Default|GetPreview_Mesh) (Math|Vector|MakeVector (Variables|Default|GetZoneXScale) (Variables|Default|GetZoneYScale) (Variables|Default|GetZoneZScale))))

(event EventInitializeDevice
  ;[0.9244 0.06801 1] Initialization
  (|CallParentFunction)
  (Utilities|FlowControl|Sequence
    (:then_0
      (CallFunction|UpdateVisibilityComponentsToUserSettings))
    (:then_1
      (Visibility|UpdateVisiblity))))

(event Custom|OnPawnStartedEmote
  (params (MontageItemDef :type Fort%20Item%20Definition%20Object%20Reference) (PawnEmoting :type Fort%20Pawn%20Object%20Reference))
  (bind _returnvalue (Pawn|GetController PawnEmoting))
  ;[0.9747 1 0] Emote Events
  (Utilities|IsValid _returnvalue
    (:Is%20Valid
      (Default|TransmitonEmoteStartedIfValidActor _returnvalue PawnEmoting))))

(event Custom|OnPawnStoppedEmote
  (params (MontageItemDef :type Fort%20Item%20Definition%20Object%20Reference) (PawnEmoting :type Fort%20Pawn%20Object%20Reference))
  (bind _returnvalue (Pawn|GetController PawnEmoting))
  ;[0.9747 1 0] Emote Events
  (Utilities|FlowControl|Delay 0.0) ; Wait for next tick to ensure\nEmoteAbility is not Active anymore
  (Utilities|IsValid _returnvalue
    (:Is%20Valid
      (bind _asathena_player_controller (Utilities|Casting|CastToAthena_PlayerController _returnvalue))
      (if (not (Emote|IsPlayingEmote _asathena_player_controller))
        (Default|TransmitonEmoteEndedIfValidActor _asathena_player_controller PawnEmoting)))))

(event Custom|UpdateTeamOnReceived
  (params (Instigator :type Controller%20Object%20Reference))
  ;[0.6869 1 0.2143] Update the selected Team or Class when receiving from channel transmissions
  ;;[0.3222 1 0.9701] Check the Authority, and Game Progress
  (Networking|SwitchHasAuthority
    (:Authority
      (bind _currentminigame (Default|GetCurrentMinigame)
        (:Is%20Valid
          (if (Minigame|IsinProgress _currentminigame)
            ;;[1 0.3279 0.2682] Get the Team and Class of the player
            (bind _asfort_player_controller_athena (Utilities|Casting|CastToFortPlayerControllerAthena Instigator))
            (Utilities|IsValid _asfort_player_controller_athena
              (:Is%20Valid
                ;;[0.9653 0.3944 1] Update the team and class to affect
                (UserOptions|CustomInitializeTeams (Knobs|Team|ReshiftTeamIndexValue (Math|Conversions|ToInteger(Byte) (Team|GetActorPvPTeamIndex _asfort_player_controller_athena))))
                (Registration|RevalidateOverlappingPawns)))))))))

(event Custom|UpdateClassOnReceived
  (params (Instigator :type Controller%20Object%20Reference))
  ;[0.6869 1 0.2143] Update the selected Team or Class when receiving from channel transmissions
  ;;[0.3222 1 0.9701] Check the Authority, and Game Progress
  (Networking|SwitchHasAuthority
    (:Authority
      (bind _currentminigame (Default|GetCurrentMinigame)
        (:Is%20Valid
          (Utilities|IsValid (Minigame|GetFNEClassComponent _currentminigame)
            (:Is%20Valid
              (if (Minigame|IsinProgress _currentminigame)
                ;;[1 0.3279 0.2682] Get the Team and Class of the player
                (bind _asfort_player_controller_athena (Utilities|Casting|CastToFortPlayerControllerAthena Instigator))
                (bind _asfort_player_state (Utilities|Casting|CastToFortPlayerState (Variables|/Script/Engine.Controller|GetPlayerState _asfort_player_controller_athena)))
                ;;[0.9653 0.3944 1] Update the team and class to affect
                (UserOptions|CustomInitializeClass (Math|Conversions|ToInteger(Byte) (Minigame|GetClassSlotforPlayer (Minigame|GetFNEClassComponent _currentminigame) _asfort_player_state false)))
                (Registration|RevalidateOverlappingPawns)))))))))

(event Custom|SpawnFX
  (params (Pawn :type Fort%20Pawn%20Object%20Reference))
  ;[0.1994 0.1886 1] Replicate it so that it actually spawns VFX
  (Networking|CanExecuteCosmeticEvents
    (:True
      (Effects|SpawnFXSystemAttached (Variables|Default|GetEmitter%20Template) (Variables|/Script/Engine.Actor|GetRootComponent Pawn)))))

(event EventPawnEnteredTriggerVolume
  (params (Pawn :type Pawn%20Object%20Reference) (PlayerController :type Player%20Controller%20Object%20Reference))
  (bind _returnvalue (Pawn|GetController Pawn))
  ;[0.2776 1 0.43] Filter through what kind of pawn has entered the volume and apply the appropriate Effect
  (Utilities|IsValid Pawn
    (:Is%20Valid
      (if (Utilities|IsValid PlayerController :!class /Script/Engine.KismetSystemLibrary)
        (bind _bwouldaffectplayer (CallFunction|WouldVolumeAffectPlayer Pawn))
        (if (and _bwouldaffectplayer (not (CallFunction|DoesPawnAlreadyHaveEffect Pawn)))
          (Default|TriggerEffect Pawn)
          (Utilities|IsValid _returnvalue
            (:Is%20Valid
              (GameplayEvent|BroadcastEvent (Variables|UserOptions|Events|GetOn%20Player%20Entering%20Zone) 0 (Utilities|Struct|MakeMutatorPlayerEnteredInstigator _returnvalue)))))
        (else
          (if (or (CallFunction|ShouldAffectCreatures Pawn) (CallFunction|ShouldAffectGuards Pawn))
            (Default|TriggerEffect Pawn))))
      (if (Variables|DeviceOptions|GetbKnob_EnableVFX)
        (bind _asfort_pawn (Utilities|Casting|CastToFortPawn Pawn))
        (CallFunction|SpawnFX _asfort_pawn)))))

(event EventPawnExitedTriggerVolume
  (params (Pawn :type Pawn%20Object%20Reference) (PlayerController :type Player%20Controller%20Object%20Reference))
  (bind _returnvalue (Pawn|GetController Pawn))
  ;[1 0.1927 0.1927] End the Effect on pawns leaving the volume
  (Utilities|IsValid Pawn
    (:Is%20Valid
      (if (Utilities|IsValid PlayerController :!class /Script/Engine.KismetSystemLibrary)
        (Default|EndEffect 0 Pawn)
        (Utilities|IsValid _returnvalue
          (:Is%20Valid
            (GameplayEvent|BroadcastEvent (Variables|UserOptions|Events|GetOn%20Player%20Exiting%20Zone) 0 (Utilities|Struct|MakeMutatorPlayerExitedInstigator _returnvalue))))
        (elif (Variables|DamageOptions|GetbDamageCreatures)
          (Default|EndEffect 0 Pawn))))))

(event EventNon-PawnExitedTriggerVolume
  (params (Actor :type Actor%20Object%20Reference))
  ;[1 0.1927 0.1927] End the Effect on pawns leaving the volume
  (if (Variables|DamageOptions|GetbDamageVehicles)
    (bind _asfort_athena_vehicle (Utilities|Casting|CastToFortAthenaVehicle Actor))
    (CallFunction|UnbindVehicleEvents _asfort_athena_vehicle)
    (Default|EndEffect _asfort_athena_vehicle)))

(event EventOnPlayerChangedTeam
  (params (PlayerState :type Fort%20Player%20State%20Object%20Reference) (TeamId :type Byte) (OldTeamId :type Byte) (bTeamChangedWithoutRespawn :type Boolean))
  ;[0.2776 1 0.43] Filter through what kind of pawn has entered the volume and apply the appropriate Effect
  ;;[0.2785 1 0.9402 0.5] Re-fresh application on the damage volume for a player who's changed class or teams
  (bind (_outpawn _pawnreturned) (Class|AbilityFunctions|GetPawnfromPlayerStateZone PlayerState))
  (CallFunction|ReapplyEffectsonOverlappingPlayers _outpawn _pawnreturned))

(event EventPlayerBotExitedTriggerVolume
  (params (Pawn :type Pawn%20Object%20Reference) (AI%20Bot%20Controller :type Fort%20Athena%20AIBot%20Controller%20Object%20Reference))
  ;[1 0.1927 0.1927] End the Effect on pawns leaving the volume
  (Utilities|IsValid Pawn
    (:Is%20Valid
      (if (Utilities|IsValid AI%20Bot%20Controller :!class /Script/Engine.KismetSystemLibrary)
        (if (Variables|DamageOptions|GetbDamagePlayerBots)
          (Default|EndEffect 0 Pawn))))))

(event EventOnPlayerClassApplied
  (params (PlayerState :type Fort%20Player%20State%20Object%20Reference) (ClassSlotIndex :type Byte))
  ;[0.2776 1 0.43] Filter through what kind of pawn has entered the volume and apply the appropriate Effect
  ;;[0.2785 1 0.9402 0.5] Re-fresh application on the damage volume for a player who's changed class or teams
  (bind (_outpawn _pawnreturned) (Class|AbilityFunctions|GetPawnfromPlayerStateZone PlayerState))
  (CallFunction|ReapplyEffectsonOverlappingPlayers _outpawn _pawnreturned))

(event EventNon-PawnEnteredTriggerVolume
  (params (Actor :type Actor%20Object%20Reference))
  ;[0.2776 1 0.43] Filter through what kind of pawn has entered the volume and apply the appropriate Effect
  (Networking|SwitchHasAuthority
    (:Authority
      (if (and (Variables|DamageOptions|GetbDamageVehicles) (CallFunction|CanPassinCreateMode))
        (bind _asfort_athena_vehicle (Utilities|Casting|CastToFortAthenaVehicle Actor))
        (CallFunction|BindVehicleEvents _asfort_athena_vehicle)
        (Default|TriggerEffect 0 _asfort_athena_vehicle)
        (else
          (Default|HandlePickUpBeginOverlap Actor))))))

(event MinigameStartedDelegate(FortMinigameProgress)
  ;[0.2619 1 0.8901 0.702] Bind and Unbind events for player team and class states changing
  (Networking|SwitchHasAuthority
    (:Authority
      (bind _currentminigame (Default|GetCurrentMinigame)
        (:Is%20Valid
          (Minigame|BindEventtoOnPlayerClassApplied _currentminigame (EventDispatchers|CreateEvent :!event OnPlayerClassApplied))
          (Minigame|BindEventtoPlayerChangedTeamDelegate _currentminigame (EventDispatchers|CreateEvent :!event OnPlayerChangedTeam)))))))

(event Custom|Player%20Exited%20Vehicle
  (params (Vehicle :type Fort%20Vehicle%20Interface%20Interface%20%28by%20ref%29) (PlayerPawn :type Fort%20Pawn%20Object%20Reference) (SeatIndex :type Integer))
  ;[0.2776 1 0.43] Filter through what kind of pawn has entered the volume and apply the appropriate Effect
  ;;[0.271 0.222 1 0.321] When a player enters or exits a vehicle, re-trigger the effects on the vehicle to filter out if it's still affected
  (if (< (Utilities|Array|Length (Vehicle|GetAllPassengers Vehicle)) 1)
    (CallFunction|Re-TriggerEffectsOnVehicle Vehicle)))

(event MinigameEndedDelegate(FortMinigameProgress)
  ;[0.2619 1 0.8901 0.702] Bind and Unbind events for player team and class states changing
  (CallFunction|CleanUpatGameEnd)
  (Networking|SwitchHasAuthority
    (:Authority
      (bind _currentminigame (Default|GetCurrentMinigame)
        (:Is%20Valid
          (Minigame|UnbindEventfromOnPlayerClassApplied _currentminigame (EventDispatchers|CreateEvent :!event OnPlayerClassApplied))
          (Minigame|UnbindEventfromPlayerChangedTeamDelegate _currentminigame (EventDispatchers|CreateEvent :!event OnPlayerChangedTeam)))))))

(event Custom|Player%20Entered%20Vehicle
  (params (Vehicle :type Fort%20Vehicle%20Interface%20Interface%20%28by%20ref%29) (PlayerPawn :type Fort%20Pawn%20Object%20Reference) (SeatIndex :type Integer))
  ;[0.2776 1 0.43] Filter through what kind of pawn has entered the volume and apply the appropriate Effect
  ;;[0.271 0.222 1 0.321] When a player enters or exits a vehicle, re-trigger the effects on the vehicle to filter out if it's still affected
  (if (== (Utilities|Array|Length (Vehicle|GetAllPassengers Vehicle)) 1)
    (CallFunction|Re-TriggerEffectsOnVehicle Vehicle)))

(event EventPlayerBotEnteredTriggerVolume
  (params (Pawn :type Pawn%20Object%20Reference) (AI%20Bot%20Controller :type Fort%20Athena%20AIBot%20Controller%20Object%20Reference))
  ;[0.2776 1 0.43] Filter through what kind of pawn has entered the volume and apply the appropriate Effect
  (Utilities|IsValid Pawn
    (:Is%20Valid
      (if (Utilities|IsValid AI%20Bot%20Controller :!class /Script/Engine.KismetSystemLibrary)
        (if (Math|Boolean|ANDBoolean (StateChecks|IsValidTeam (Variables|Default|GetCreativeUseRestrictionComponent) Pawn) (Variables|DamageOptions|GetbDamagePlayerBots) (CallFunction|CanPassinCreateMode) (not (CallFunction|DoesPawnAlreadyHaveEffect Pawn)))
          (Default|TriggerEffect Pawn)
          (if (Variables|DeviceOptions|GetbKnob_EnableVFX)
            (bind _asfort_pawn (Utilities|Casting|CastToFortPawn Pawn))
            (CallFunction|SpawnFX _asfort_pawn)))))))

(event Custom|Add%20or%20Remove%20Affected%20Actor
  (params (Remove%20Actor :type Boolean) (Actor :type Actor%20Object%20Reference))
  (bind _affectedactor (Variables|Default|GetAffectedActor))
  ;[0.5208 1 1 0.4] Keep a list of affected actors to handle on game end cleanup
  (Networking|SwitchHasAuthority
    (:Authority
      (if Remove%20Actor
        (Utilities|Array|RemoveItem _affectedactor Actor)
        (else
          (Utilities|Array|AddUnique _affectedactor Actor)))
      (Default|BindorUnbindonPlayerEmote Actor (not Remove%20Actor)))))

(event Custom|Clean%20Up%20At%20Game%20End
  (bind _affectedactor (Variables|Default|GetAffectedActor))
  ;[0.5208 1 1 0.4] Keep a list of affected actors to handle on game end cleanup
  (Networking|SwitchHasAuthority
    (:Authority
      (bind (_arrayindex _arrayelement) (Utilities|Array|ReverseforEachLoop _affectedactor)
        (:LoopBody
          (Default|EndEffect _arrayelement))
        (:Completed
          (Utilities|Array|Clear _affectedactor))))))

(fn AddHandleToHoldingArray
  (params (ActiveGameplayEffectHandle :type Active%20Gameplay%20Effect%20Handle%20Structure) (Ability%20System%20Component :type Ability%20System%20Component%20Object%20Reference))
  ;[0.2257 0.8599 1] Add the player's ability system component and the active gameplay and put them in a struct so we can retain knowledge of the handle of the GE we just applied to the player
  (bind _output_get (Variables|Default|SetActiveGameplayEffectHandle ActiveGameplayEffectHandle))
  (bind _output_get_2 (Variables|SetGE%20Handle_ActiveGameplayEffectHandle_7_C6FF36454DF9DCBF6904BD96974FFF0B (Utilities|Struct|MakeAbilityHandleAndSystemHold _output_get Ability%20System%20Component)))
  (Utilities|Array|Add (Variables|Default|GetAbility%20Handle%20Array) _output_get_2))

(fn AffectsTarget
  (params (Target%20Actor :type Actor%20Object%20Reference))
  (bind _asfort_player_pawn (Utilities|Casting|CastToFortPlayerPawn Target%20Actor)
    (:then
      (if (Variables|DamageOptions|GetbDamagePlayers)
        (return (DEPRECATED|CanActorInteract (Variables|Default|GetCreativeUseRestrictionComponent) _asfort_player_pawn true false))
        (else
          (return false))))
    (:CastFailed
      (bind _asfort_pawn (Utilities|Casting|CastToFortPawn Target%20Actor))
      (if (Variables|DamageOptions|GetbDamageCreatures)
        (return (StateChecks|IsValidTeam (Variables|Default|GetCreativeUseRestrictionComponent) _asfort_pawn))
        (else
          (return false))))))

(fn ApplyEffectsAndAssignHandles
  (params (AbilitySystemComponent :type Ability%20System%20Component%20Object%20Reference) (SpecHandle :type Gameplay%20Effect%20Spec%20Handle%20Structure%20%28by%20ref%29))
  ;[0.426 1 0.2344] Apply the Effect and add it to our custom array of structs so that we remember what we applied to who
  (bind _returnvalue (GameplayEffects|ApplyGameplayEffectSpecToTarget AbilitySystemComponent SpecHandle AbilitySystemComponent))
  (CallFunction|AddHandletoHoldingArray _returnvalue AbilitySystemComponent))

(fn ApplyGE
  (params (Actor :type Actor%20Object%20Reference))
  (bind _returnvalue (Ability|GetAbilitySystemComponent Actor))
  (bind _returnvalue_2 (GameplayEffects|MakeEffectContext _returnvalue))
  ;[1 0.9023 0.1292] Using the targeted object, create a Gameplay Effect Spec and Apply them based on the GE being applied
  (CallFunction|AddOrRemoveAffectedActor false Actor)
  (bind _gameplay_effects (CallFunction|GetVolumeGE))
  (for _array_element _gameplay_effects
    (if (Utilities|Equal(Class) (Variables|GameplayEffects|GetGE_BuildingSpeed) _array_element)
      (bind _returnvalue_3 (Ability|GameplayEffect|AssignTagSetbyCallerMagnitude (GameplayEffects|MakeOutgoingSpec _returnvalue _array_element 0.0 _returnvalue_2) "(TagName=\"SetByCaller.Magnitude\")" (Variables|DeviceOptions|GetBuildingSpeed)))
      (CallFunction|ApplyEffectsandAssignHandles _returnvalue _returnvalue_3)
      (elif (Utilities|Equal(Class) _array_element (Variables|GameplayEffects|GetGE_BuildingCost))
        (bind _returnvalue_4 (Ability|GameplayEffect|AssignTagSetbyCallerMagnitude (GameplayEffects|MakeOutgoingSpec _returnvalue _array_element 0.0 _returnvalue_2) "(TagName=\"SetByCaller.Magnitude\")" (Variables|DeviceOptions|GetBuildingCost)))
        (CallFunction|ApplyEffectsandAssignHandles _returnvalue _returnvalue_4)
        (elif (Utilities|Equal(Class) _array_element (Variables|GameplayEffects|GetGE%20Gravity))
          (bind _returnvalue_5 (Ability|GameplayEffect|AssignTagSetbyCallerMagnitude (GameplayEffects|MakeOutgoingSpec _returnvalue _array_element 0.0 _returnvalue_2) "(TagName=\"SetByCaller.Magnitude\")" (Variables|Default|GetGravityOverride)))
          (CallFunction|ApplyEffectsandAssignHandles _returnvalue _returnvalue_5)
          (elif (Utilities|Equal(Class) _array_element (Variables|GameplayEffects|GetGE%20PawnSpeed))
            (CallFunction|ModifyPawnSpeed (Variables|GetActor))
            (else
              (CallFunction|ApplyEffectsandAssignHandles _returnvalue (GameplayEffects|MakeOutgoingSpec _returnvalue _array_element 0.0 _returnvalue_2)))))))))

(fn BindOrUnbindOnPlayerEmote
  (params (Pawn :type Actor%20Object%20Reference) (Bind :type Boolean))
  (bind _outputdelegate (EventDispatchers|CreateEvent :!event OnPawnStartedEmote))
  (bind _outputdelegate_2 (EventDispatchers|CreateEvent :!event OnPawnStoppedEmote))
  (Networking|SwitchHasAuthority
    (:Authority
      (bind _asfort_pawn (Utilities|Casting|CastToFortPawn Pawn))
      (if (Variables|GetBind)
        (EventDispatchers|BindEventtoOnPawnStartedEmote _asfort_pawn _outputdelegate)
        (EventDispatchers|BindEventtoOnPawnStoppedEmote _asfort_pawn _outputdelegate_2)
        (else
          (EventDispatchers|UnbindEventfromOnPawnStartedEmote _asfort_pawn _outputdelegate)
          (EventDispatchers|UnbindEventfromOnPawnStoppedEmote _asfort_pawn _outputdelegate_2))))))

(fn BindVehicleEvents
  (params (FortAthenaVehicle :type Fort%20Athena%20Vehicle%20Object%20Reference))
  ;[0.505 1 0.2499] Bind events when players enter or exit this vehicle
  (CallFunction|UnbindVehicleEvents FortAthenaVehicle)
  (Vehicle|BindEventtoOnPawnExitedVehicleSeat FortAthenaVehicle (EventDispatchers|CreateEvent :!event Player%20Exited%20Vehicle))
  (Vehicle|BindEventtoOnPawnEnteredVehicleSeat FortAthenaVehicle (EventDispatchers|CreateEvent :!event Player%20Entered%20Vehicle)))

(fn CanPassInCreateMode
  ;[1 0.9202 0.2854] Check to see if we can pass during Create mode when playing in creative
  (return (or (CallFunction|IsinPlayMode) (and (not (CallFunction|IsinPlayMode)) (Variables|DamageOptions|GetbActiveInEditor)))))

(fn CheckVehicleDamageAbility
  (params (Object :type Object%20Reference))
  ;[1 0.8039 0.2999] Should this vehicle be damaged at all?
  (return (or (and (Variables|DamageOptions|GetbDamageVehicles) (not (CallFunction|ReturnIsVehicleEmpty Object))) (and (Variables|DamageOptions|GetbDamageVehicles) (Variables|DamageOptions|GetbDamageEmptyVehicles)))))

(fn CheckForAIGameplayEffectReplacement
  (params (TargetActor :type Actor%20Object%20Reference) (InGameplayEffect :type Gameplay%20Effect%20Class%20Reference))
  (bind _selectedge (LocalVariables|GetSelectedGE))
  (bind _aigameplayeffectreplacements (Variables|GameplayEffects|GetAIGameplayEffectReplacements))
  (bind (_value _returnvalue) (Utilities|Map|Find _aigameplayeffectreplacements _selectedge))
  (LocalVariables|SetSelectedGE InGameplayEffect)
  (bind _asfort_pawn (Utilities|Casting|CastToFortPawn TargetActor)
    (:then
      (Utilities|Casting|CastToAIController (Pawn|GetController _asfort_pawn)
        (:then
          (if (Pawn|IsBotControlled _asfort_pawn)
            _
            (elif _returnvalue
              (bind _output_get (LocalVariables|SetSelectedGE _value))
              (return _output_get (Utilities|NotEqual(Class) _output_get (Variables|GetInGameplayEffect))))))
        (:CastFailed
          _)))
    (:CastFailed
      _)))

(fn CheckPlayerTeamAndClassValidity
  (params (Actor :type Actor%20Object%20Reference))
  ;[0.4111 1 0.2042] Check whether or not the controller is valid based on its team and class compared to what is being filtered
  (bind (_valid_controller _fortplayerpawn) (DEPRECATED|ActorHasValidController (Variables|Default|GetCreativeUseRestrictionComponent) Actor true true))
  (return _valid_controller _fortplayerpawn))

(fn DoesPawnAlreadyHaveEffect
  (params (Pawn :type Object%20Reference))
  ;[0.955 1 0.2919] Cycle through all registered players to see if a pawn currently entering has already been registered. this is purely to prevent double triggers on single entry.
  (bind _array_element (Utilities|Array|ForEachLoopwithBreak (Variables|Default|GetAffectedActor))
    (:LoopBody
      (if (Utilities|Equal(Object) _array_element Pawn)
        (LocalVariables|SetPlayerHasEffect true)))
    (:Completed
      _))
  (return (LocalVariables|GetPlayerHasEffect)))

(fn EndEffect
  (params (Vehicle :type Actor%20Object%20Reference) (Pawn :type Pawn%20Object%20Reference))
  (bind _returnvalue (Utilities|IsValid Pawn :!class /Script/Engine.KismetSystemLibrary))
  (bind _returnvalue_2 (select _returnvalue Pawn Vehicle))
  ;[1 0.6484 0.217] Remove all relevant GEs upon leaving the volume
  (Utilities|IsValid _returnvalue_2
    (:Is%20Valid
      (CallFunction|RemoveActiveHandlesFromPlayer (Ability|GetAbilitySystemComponent _returnvalue_2))
      (CallFunction|AddOrRemoveAffectedActor true _returnvalue_2))))

(fn GetVolumeGE
  (bind _gameplayeffects (LocalVariables|GetGameplayEffects))
  ;[1 0.5233 0.4534] Lets filter through our options, and set the correct Volume GE in our array of what to use down the line
  (Utilities|Array|Clear (LocalVariables|GetGameplayEffects))
  (Utilities|FlowControl|Sequence
    (:then_0
      (if (not (Variables|DeviceOptions|GetbAllowEditing))
        (Utilities|Array|Add _gameplayeffects (Variables|GameplayEffects|GetGE_DisableEditing))))
    (:then_1
      (if (not (Variables|DeviceOptions|GetKnob_bAllowFiring))
        (Utilities|Array|Add _gameplayeffects (Variables|GameplayEffects|GetGE_DisableWeapons))))
    (:then_2
      (switch Utilities|FlowControl|Switch|SwitchonEBuildingMode (Variables|DeviceOptions|GetAllowBuilding)
        (:None
          (Utilities|Array|Add _gameplayeffects (Variables|GameplayEffects|GetGE_DisableBuilding)))))
    (:then_3
      (if (not (Variables|DeviceOptions|GetbAllowJumping))
        (Utilities|Array|Add _gameplayeffects (Variables|GameplayEffects|GetGE_DisableJumping))))
    (:then_4
      (if (!= (Variables|DeviceOptions|GetBuildingSpeed) 1.0)
        (Utilities|Array|Add _gameplayeffects (Variables|GameplayEffects|GetGE_BuildingSpeed))))
    (:then_5
      (if (!= (Variables|DeviceOptions|GetBuildingCost) 1.0)
        (Utilities|Array|Add _gameplayeffects (Variables|GameplayEffects|GetGE_BuildingCost))))
    (:then_6
      (if (!= (Variables|Default|GetSpeedMultiplier) 1.0)
        (Utilities|Array|Add _gameplayeffects (Variables|GameplayEffects|GetGE%20PawnSpeed))))
    (:then_7
      (if (!= (Variables|Default|GetGravityOverride) 1.0)
        (Utilities|Array|Add _gameplayeffects (Variables|GameplayEffects|GetGE%20Gravity))))
    (:then_8
      (return _gameplayeffects))))

(fn GetDesiredLifeSpan
  (return (Math|Float|SelectFloat 0.1 (Variables|DeviceOptions|GetPickUpLifeSpan) (Math|Float|NearlyEqual(Float) (Variables|DeviceOptions|GetPickUpLifeSpan)))))

(fn HandlePickUpBeginOverlap
  (params (Actor :type Actor%20Object%20Reference))
  (if (Math|Boolean|ANDBoolean (Utilities|IsValid Actor :!class /Script/Engine.KismetSystemLibrary) (Utilities|ClassIsChildOf (Utilities|GetClass Actor) "/Script/FortniteGame.FortPickup") (>= (Variables|DeviceOptions|GetPickUpLifeSpan) 0.0))
    ;[1 1 1] When disabled we reset Pickup Life Span
    (if (Class|CreativeEnabledComponent|IsEnabled (Variables|Default|GetEnabledComponent))
      (if (Math|Float|NearlyEqual(Float) (CallFunction|GetDesiredLifeSpan) -1.0)
        (CallFunction|ResetPickupLifespanandStopDespawnTimer Actor)
        (else
          (CallFunction|ResetPickupLifespanandStopDespawnTimer Actor)
          (bind _asfort_pickup (Utilities|Casting|CastToFortPickup Actor))
          (Pickup|SetDespawnTime _asfort_pickup (CallFunction|GetDesiredLifeSpan))))
      (else
        (CallFunction|ResetPickupLifespanandStopDespawnTimer Actor)))))

(fn InitAIGameplayEffectReplacements
  (Variables|GameplayEffects|SetAIGameplayEffectReplacements (Utilities|Map|MakeMap (Variables|GameplayEffects|GetGE%20PawnSpeed) (Variables|GameplayEffects|GetGE%20AI%20PawnSpeed))))

(fn ModifyPawnSpeed
  (params (TargetActor :type Actor%20Object%20Reference))
  (bind (_outgameplayeffectspec _wasreplaced) (Default|CheckforAIGameplayEffectReplacement TargetActor (select (and (Variables|DeviceOptions|GetbDisableMovement) (== (Variables|Default|GetSpeedMultiplier) 0.0)) (Variables|GameplayEffects|GetGEPawnImmobilize) (Variables|GameplayEffects|GetGE%20PawnSpeed))))
  (LocalVariables|SetbUseAIEffect _wasreplaced)
  (bind _returnvalue (Ability|GameplayEffect|AssignTagSetbyCallerMagnitude (GameplayEffects|MakeOutgoingSpec :self (Ability|GetAbilitySystemComponent (Variables|GetTargetActor)) :GameplayEffectClass (select (and _wasreplaced (== (Variables|Default|GetSpeedMultiplier) 0.0)) (Variables|GameplayEffects|GetGE%20AI%20Immobilize) _outgameplayeffectspec) :Level 0.0) (select (LocalVariables|GetbUseAIEffect) "(TagName=\"SetByCaller.MoveSpeed\")" "(TagName=\"SetByCaller.Magnitude\")") (Variables|Default|GetSpeedMultiplier)))
  (CallFunction|ApplyEffectsandAssignHandles (Ability|GetAbilitySystemComponent (Variables|GetTargetActor)) _returnvalue))

(fn OnPreviewStart
  (return false))

(fn Re-TriggerEffectsonVehicle
  (params (Object :type Fort%20Vehicle%20Interface%20Interface))
  ;[0.2308 0.1369 1] Remove then re-apply the effects on the vehicle in the event that it becomes occupied or vacant
  (bind _asfort_athena_vehicle (Utilities|Casting|CastToFortAthenaVehicle Object))
  (Default|EndEffect _asfort_athena_vehicle)
  (CallFunction|Non-PawnEnteredTriggerVolume _asfort_athena_vehicle))

(fn ReapplyEffectsOnOverlappingPlayers
  (params (Player%20Pawn :type Fort%20Player%20Pawn%20Object%20Reference) (Pawn%20Returned :type Boolean))
  (bind _playercontroller (Default|GetPlayerControllerfromActor Player%20Pawn))
  ;[0.2458 1 0.79 0.5] If we have a player pawn, and are overlaping it, check to see if volume's affect on the player has changed
  (Utilities|IsValid Player%20Pawn
    (:Is%20Valid
      (Utilities|IsValid _playercontroller
        (:Is%20Valid
          (LocalVariables|SetPlayerController _playercontroller)
          (if (and (Collision|IsOverlappingActor self Player%20Pawn) Pawn%20Returned)
            (LocalVariables|SetbOldAffectedByVolume (CallFunction|DoesPawnAlreadyHaveEffect (Variables|GetPlayer%20Pawn)))
            (bind _bwouldaffectplayer (CallFunction|WouldVolumeAffectPlayer (Variables|GetPlayer%20Pawn)))
            (LocalVariables|SetbNewAffectedByVolume _bwouldaffectplayer)
            ;[0.2458 0.5908 0.3671 0.5] If the volume no longer affects the player then trigger the pawn exited event
            (if (and (not (LocalVariables|GetbNewAffectedByVolume)) (LocalVariables|GetbOldAffectedByVolume))
              (CallFunction|PawnExitedTriggerVolume (Variables|GetPlayer%20Pawn) (LocalVariables|GetPlayerController))
              (else
                ;[-0.0543 0.7613 0.9724 0.5] If the pawn is now affected by the volume then trigger the pawn entered event
                (if (and (LocalVariables|GetbNewAffectedByVolume) (not (LocalVariables|GetbOldAffectedByVolume)))
                  (CallFunction|PawnEnteredTriggerVolume (Variables|GetPlayer%20Pawn) (LocalVariables|GetPlayerController)))))))))))

(fn RemoveActiveHandlesfromPlayer
  (params (Ability%20System%20Component :type Ability%20System%20Component%20Object%20Reference))
  ;[0.3118 1 0.9701] Retrieve the Gameplay handle used to identify the gameplay effect.
  (bind (_arrayindex _arrayelement) (Utilities|Array|ReverseforEachLoop (Variables|Default|GetAbility%20Handle%20Array))
    (:LoopBody
      (if (Utilities|Equal(Object) Ability%20System%20Component (.AbilitySystemComponent_8_57E9F25741437CC55361B58F9DB743E0 (Utilities|Struct|BreakAbilityHandleandSystemHold _arrayelement)))
        (GameplayEffects|RemoveActiveGameplayEffect Ability%20System%20Component (Utilities|Struct|BreakAbilityHandleandSystemHold _arrayelement) -1)
        (Utilities|Array|RemoveIndex (Variables|Default|GetAbility%20Handle%20Array) _arrayindex)))))

(fn ResetPickupLifespanAndStopDespawnTimer
  (params (Pickup :type Actor%20Object%20Reference))
  (Actor|SetLifeSpan Pickup 0.0)
  (bind _asfort_pickup (Utilities|Casting|CastToFortPickup Pickup))
  (Pickup|SetDespawnTime _asfort_pickup 0.0))

(fn ReturnIsVehicleEmpty
  (params (Object :type Object%20Reference))
  ;[0.8591 1 0.3345] Return whether or not this vehicle is empty
  (bind _asfort_athena_vehicle (Utilities|Casting|CastToFortAthenaVehicle Object)
    (:then
      (return (<= (Utilities|Array|Length (Vehicle|GetAllPassengers _asfort_athena_vehicle)) 0)))
    (:CastFailed
      (return false))))

(fn SendTriggerMessage
  (params (Trigger%20Message%20Component :type Fort%20Gameplay%20Trigger%20Message%20Component%20Object%20Reference) (Pawn :type Pawn%20Object%20Reference))
  (bind _returnvalue (Pawn|GetController Pawn))
  (Utilities|IsValid _returnvalue
    (:Is%20Valid
      (Class|FortGameplayTriggerMessageComponent|SendTriggerMessage Trigger%20Message%20Component _returnvalue))))

(fn ShouldAffectCreatures
  (params (Pawn :type Pawn%20Object%20Reference))
  (return (and (Default|ActorIsACreature Pawn) (Variables|DamageOptions|GetbDamageCreatures))))

(fn ShouldAffectGuards
  (params (Pawn :type Pawn%20Object%20Reference))
  (return (and (Default|ActorIsAPlayerBot Pawn) (Variables|DamageOptions|GetbDamagePlayerBots))))

(fn TransmitOnEmoteEndedifValidActor
  (params (Controller :type Controller%20Object%20Reference) (Actor :type Actor%20Object%20Reference))
  (Networking|SwitchHasAuthority
    (:Authority
      (bind _shouldaffect (Default|AffectsTarget Actor))
      (if _shouldaffect
        (bind _controller (Variables|GetController)
          (:then
            (LocalVariables|SetValidController _controller))
          (:else
            (bind _asfort_player_pawn_athena (Utilities|Casting|CastToFortPlayerPawnAthena (Variables|GetActor)))
            (LocalVariables|SetValidController (Pawn|GetController _asfort_player_pawn_athena))))
        (bind _validcontroller (LocalVariables|GetValidController)
          (:then
            (GameplayEvent|BroadcastEvent (Variables|UserOptions|Events|GetOn%20Player%20Stops%20Emoting) 0 (Utilities|Struct|MakeMutatorEmoteEndedInstigator _validcontroller))
            (return))
          (:else))))))

(fn TransmitOnEmoteStartedifValidActor
  (params (Controller :type Controller%20Object%20Reference) (Actor :type Actor%20Object%20Reference))
  (Networking|SwitchHasAuthority
    (:Authority
      (bind _shouldaffect (Default|AffectsTarget Actor))
      (if _shouldaffect
        (bind _controller (Variables|GetController)
          (:then
            (LocalVariables|SetValidController _controller))
          (:else
            (bind _asfort_player_pawn_athena (Utilities|Casting|CastToFortPlayerPawnAthena (Variables|GetActor)))
            (LocalVariables|SetValidController (Pawn|GetController _asfort_player_pawn_athena))))
        (bind _validcontroller (LocalVariables|GetValidController)
          (:then
            (GameplayEvent|BroadcastEvent (Variables|UserOptions|Events|GetOn%20Player%20Starts%20Emoting) 0 (Utilities|Struct|MakeMutatorEmotestartedInstigator _validcontroller))
            (return))
          (:else))))))

(fn TriggerEffect
  (params (Pawn :type Pawn%20Object%20Reference) (Vehicle :type Actor%20Object%20Reference))
  (bind _returnvalue (Ability|GetFortAbilitySystemComponent Pawn))
  ;[0.1639 0.174 1] Filter out if we're affecting players or vehicles
  (if (CallFunction|IsinPlayMode)
    (if (Utilities|IsValid Pawn :!class /Script/Engine.KismetSystemLibrary)
      ;[0.3017 1 0.2404] Apply the GE and put the Vehicle or Pawn in our Affected Arrays
      (CallFunction|ApplyGE Pawn)
      ;[0.3779 1 0.8944] Stop attacks if we don't allow firing
      (Networking|SwitchHasAuthority
        (:Authority
          (if (not (Variables|DeviceOptions|GetKnob_bAllowFiring))
            (Utilities|IsValid _returnvalue
              (:Is%20Valid
                (Tags|CancelAbilitiesWithTags _returnvalue (GameplayTags|MakeLiteralGameplayTagContainer (Variables|Default|GetCancelAbilitiesWithTheseTags))))))))
      (else
        (if (CallFunction|CheckVehicleDamageAbility Vehicle)
          (CallFunction|ApplyGE Vehicle))))))

(fn UnbindVehicleEvents
  (params (FortAthenaVehicle :type Fort%20Athena%20Vehicle%20Object%20Reference))
  ;[1 0.6892 0.2661] Unbind all events where players enter or leave the specific vehicle
  (Vehicle|UnbindEventfromOnPawnEnteredVehicleSeat FortAthenaVehicle (EventDispatchers|CreateEvent :!event Player%20Entered%20Vehicle))
  (Vehicle|UnbindEventfromOnPawnExitedVehicleSeat FortAthenaVehicle (EventDispatchers|CreateEvent :!event Player%20Exited%20Vehicle)))

(fn UpdateVisibilityComponentstoUserSettings
  (bind (_zone_components _base_plate_components) (Visibility|UpdateVisualComponents))
  ;[0.4546 1 0.379] Set our visibility components to use later
  (bind _zonevisibilitycomponent (Variables|Default|GetZoneVisibilityComponent)
    (:then
      (Visibility|SetVisibilityComponents _zonevisibilitycomponent _zone_components true))
    (:else
      _))
  (bind _basevisibilitycomponent (Variables|Default|GetBaseVisibilityComponent)
    (:then
      (Visibility|SetVisibilityComponents _basevisibilitycomponent _base_plate_components true))
    (:else)))

(fn UpdateVisualComponents
  ;[1 0.8606 0.217 0.864] Create an array of all components we need to affect visually
  (return (Utilities|Array|MakeArray 0 (Variables|Default|GetZoneMesh)) (Utilities|Array|MakeArray (Variables|Debug|GetStaticMeshComponent))))

(fn ConstructionScript
  (|CallParentFunction)
  (Default|InitAIGameplayEffectReplacements))

(fn VFXParameters
  (bind (_returnvalue_x _returnvalue_y _returnvalue_z) (Math|Vector|BreakVector (* (Math|Vector|MakeVector 512.0 512.0 384.0) (Math|Vector|MakeVector (Variables|Default|GetZoneXScale) (Variables|Default|GetZoneYScale) (Variables|Default|GetZoneZScale)))))
  (bind (_zoneedgessize_x _zoneedgessize_y _zoneedgessize_z) (Math|Vector|BreakVector (Variables|GetZoneEdgesSize_X)))
  (return (Utilities|Array|MakeArray (Utilities|Struct|MakeParticleSysParam "BlocksVolume" "PSPT_VectorRand" 0.0 0.0 (Variables|Default|GetBlocksVolumeHigh) (Variables|Default|GetBlocksVolumeLow)) (Utilities|Struct|MakeParticleSysParam "BlocksSpawnRateScale" "PSPT_Scalar" (Utilities|ClampMaxFloat (/ (Math|Float|float*float _returnvalue_x _returnvalue_y _returnvalue_z) 100663296.0) 1e+01) 0.0 "0.000000,0.000000,0.000000" (0.0 0.0 0.0) "(B=0,G=0,R=0,A=0)" 0 0) (Utilities|Struct|MakeParticleSysParam "MeshSize" "PSPT_Vector" 0.0 0.0 (Math|Vector|MakeVector _zoneedgessize_x _zoneedgessize_y (Math|Float|SafeDivide _zoneedgessize_z (Variables|Default|GetZoneZScale))) "0.000000,0.000000,0.000000" "(B=0,G=0,R=0,A=0)" 0 0) (Utilities|Struct|MakeParticleSysParam "BlocksSize" "PSPT_Vector" 0.0 0.0 (Math|Vector|MakeVector (Utilities|Reciprocal (Variables|Default|GetZoneXScale)) (Utilities|Reciprocal (Variables|Default|GetZoneXScale))) "0.000000,0.000000,0.000000" "(B=0,G=0,R=0,A=0)" 0 0) (Utilities|Struct|MakeParticleSysParam "BlocksVelocity" "PSPT_Vector" 0.0 0.0 (Math|Vector|MakeVector 0.0 0.0 (* -2e+01 (Utilities|Reciprocal (Variables|Default|GetZoneZScale)))) "0.000000,0.000000,0.000000" "(B=0,G=0,R=0,A=0)" 0 0))))

(fn WouldVolumeAffectPlayer
  (params (Pawn :type Actor%20Object%20Reference))
  (bind _valid_controller (CallFunction|CheckPlayerTeamandClassValidity Pawn))
  (return (Math|Boolean|ANDBoolean _valid_controller (Variables|DamageOptions|GetbDamagePlayers) (CallFunction|CanPassinCreateMode))))
