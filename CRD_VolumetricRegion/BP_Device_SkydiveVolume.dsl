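; Creative skydive volume device that puts players inside its trigger box into a vortex, pushing them along its axis and optionally locking them inside.
; On the server, entry applies in-vortex and optional FX gameplay effects, sets vortex parameters from user options and broadcasts enter/occupied events.
; Exit, leaving the island or minigame end removes the effects, with exit also firing exit/emptied events.
; Lock receivers and a use-restriction check decide who can leave, re-applying parameters to affected pawns.
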(event EventBeginPlay
  ;[0.139 0.8021 0.04948] Initialization - Skydive volumes want to register to a central location (for FX Updating)
  (|CallParentFunction)
  (CreativeObjectRegistry|GetOrCreateObjectRegister self "(TagName=\"Creative.Register.Vortex\")" true 0 (Variables|Default|GetVortex%20Register))
  (bind (_asynctaskproxy _payload) (GameplayEventRouter|StartListeningToEventBackwardCompatible "(TagName=\"EventMessage.Component.Playspace.PlayerExitedSpatialActor.LegacyMinigame\")" (GameplayEventRouter|GetGlobalEventRouter) "/Script/FortniteGame.PlayerExitSpatialActorContextLegacyMinigame" 0 "Down")
    (:then
      (Variables|Default|SetExitedEventHandle _asynctaskproxy)
      (Variables|Default|SetCachedMinigameForActor (Minigame|GetMinigame self :!class /Script/FortniteGame.MinigameBlueprintLibrary)))
    (:OnEventReceived
      (CallFunction|OnPlayerExitAnyIsland (.ExitingPlayerState (Utilities|Struct|BreakPlayerExitSpatialActorContextLegacyMinigame _payload)) (Utilities|Struct|BreakPlayerExitSpatialActorContextLegacyMinigame _payload)))))

(event EventInitializeDevice
  (bind _fortminigameprogress (Variables|Default|GetFortMinigameProgress))
  (bind _currentminigame (Variables|/Script/FortniteGame.FortMinigameLogicComponent|GetCurrentMinigame _fortminigameprogress))
  ;[0.139 0.8021 0.04948] Initialize Device
  (|CallParentFunction)
  (CallFunction|InitializeCreativeIcone)
  (Utilities|IsValid _currentminigame
    (:Is%20Valid
      (Class|CreativeEnabledComponent|SetEnabledOnPhase (Variables|Default|GetEnabledComponent) (Variables|/Script/FortniteGame.FortMinigame|GetCurrentState _currentminigame) (Variables|LegacyOptions|GetEnabledIndex))
      (Registration|RevalidateOverlappingPawns)
      (Networking|SwitchHasAuthority
        (:Authority
          (Variables|Default|SetbLockPlayersInVolume (Variables|UserKnobs|GetbLockPlayersKnob))
          (CallFunction|UpdateVortexParametersforPlayerinVolume))))))

(event EventPawnEnteredTriggerVolume
  (params (Pawn :type Pawn%20Object%20Reference) (PlayerController :type Player%20Controller%20Object%20Reference))
  (bind _self self)
  (bind _returnvalue (Ability|GetAbilitySystemComponent _self))
  ;[0.3543 1 0.1729] Pawn Entered Volume
  (bind _asfort_player_pawn (Utilities|Casting|CastToFortPlayerPawn Pawn))
  (Utilities|IsPawnValid _asfort_player_pawn
    (:Valid
      (Registration|RegisterPlayerinVolume _asfort_player_pawn)
      (bind _returnvalue_2 (Utilities|Array|AddUnique (Variables|Default|GetAffectedPawns) _asfort_player_pawn))
      (if (and (>= _returnvalue_2 0) (Networking|HasAuthority _asfort_player_pawn))
        (Utilities|FlowControl|Sequence
          (:then_0
            (Utilities|IsValid (Ability|GetAbilitySystemComponent _asfort_player_pawn)
              (:Is%20Valid
                (Utilities|IsValid _returnvalue
                  (:Is%20Valid
                    (GameplayEffects|ApplyGameplayEffectToTarget :self _returnvalue :GameplayEffectClass (Variables|Default|GetInVortexGameplayEffect) :Target (Ability|GetAbilitySystemComponent _asfort_player_pawn) :Level 0.0)
                    (if (Variables|UserKnobs|GetbShowPlayerFX)
                      (bind _returnvalue_3 (GameplayEffects|ApplyGameplayEffectToTarget :self _returnvalue :GameplayEffectClass (Variables|Default|GetPlayerFXGameplayEffect) :Target (Ability|GetAbilitySystemComponent _asfort_player_pawn) :Level 0.0))
                      (Utilities|Map|Add (Variables|Default|GetPlayerFXGameplayEffectHandle) _asfort_player_pawn _returnvalue_3)))))))
          (:then_1
            (CallFunction|SetPlayerinVortexVolume _asfort_player_pawn)
            (Utilities|IsValid (Pawn|GetController _asfort_player_pawn)
              (:Is%20Valid
                (GameplayEvent|BroadcastEvent (Variables|UserOptions|Events|GetOn%20Player%20Enter) 0 (Utilities|Struct|MakeGameplayEventEnterSkyDiveVol (Pawn|GetController _asfort_player_pawn)))
                (if (== (Utilities|Array|Length (Variables|Default|GetAffectedPawns)) 1)
                  (GameplayEvent|BroadcastEvent (Variables|UserOptions|Events|GetOn%20Zone%20Occupied) 0 (Utilities|Struct|MakeGameplayEventOccupiedSkyDive (Pawn|GetController _asfort_player_pawn))))))))))))

(event EventPawnExitedTriggerVolume
  (params (Pawn :type Pawn%20Object%20Reference) (PlayerController :type Player%20Controller%20Object%20Reference))
  (bind _invortexgameplayeffect (Variables|Default|GetInVortexGameplayEffect))
  (bind _playerfxgameplayeffecthandle (Variables|Default|GetPlayerFXGameplayEffectHandle))
  ;[1 0.5521 0.0603] Pawn Exited Volume
  (bind _asfort_player_pawn (Utilities|Casting|CastToFortPlayerPawn Pawn))
  (Registration|UnregisterPlayerinVolume _asfort_player_pawn)
  (Utilities|Array|RemoveItem (Variables|Default|GetAffectedPawns) _asfort_player_pawn)
  (Networking|SwitchHasAuthority
    (:Authority
      (Utilities|FlowControl|Sequence
        (:then_0
          ;;[1 0.2446 0.06435] Remove skydive effects from player
          (Utilities|IsValid (Ability|GetAbilitySystemComponent _asfort_player_pawn)
            (:Is%20Valid
              (GameplayEffects|RemoveActiveGameplayEffect (Ability|GetAbilitySystemComponent _asfort_player_pawn) (Utilities|Map|Find _playerfxgameplayeffecthandle _asfort_player_pawn) 1)
              (Utilities|Map|Remove _playerfxgameplayeffecthandle _asfort_player_pawn)
              (GameplayEffects|RemoveActiveGameplayEffectbySourceEffect (Ability|GetAbilitySystemComponent _asfort_player_pawn) _invortexgameplayeffect 0 1)
              (if (<= (GameplayEffects|GetGameplayEffectCount (Ability|GetAbilitySystemComponent _asfort_player_pawn) _invortexgameplayeffect 0 true) 0)
                (Vortex|SetinVortex _asfort_player_pawn false 0.0 0.0 0.0 0.0 0.0 true)))))
        (:then_1
          ;;[0.3281 0.6669 1] Trigger Channels
          (Utilities|IsValid (Pawn|GetController _asfort_player_pawn)
            (:Is%20Valid
              (GameplayEvent|BroadcastEvent (Variables|UserOptions|Events|GetOn%20Player%20Exits) 0 (Utilities|Struct|MakeGameplayEventExitSkyDiveVol (Pawn|GetController _asfort_player_pawn)))
              (if (== (Utilities|Array|Length (Variables|Default|GetAffectedPawns)) 0)
                (GameplayEvent|BroadcastEvent (Variables|UserOptions|Events|GetOn%20Zone%20Emptied) 0 (Utilities|Struct|MakeGameplayEventEmptySkyDiveVol (Pawn|GetController _asfort_player_pawn)))))))))))

(event MinigameEndedDelegate(FortMinigameProgress)
  (bind _affectedpawns (Variables|Default|GetAffectedPawns))
  ;[1 0.4479 0.4479 0.75] Cleanup after game
  (Networking|SwitchHasAuthority
    (:Authority
      (if (> (Utilities|Array|Length _affectedpawns) 0)
        (for _array_element _affectedpawns
          (CallFunction|PawnExitedTriggerVolume _array_element))
        (else
          _))
      (Utilities|Array|Clear _affectedpawns))))

(event Custom|Update%20Vortex%20Parameters%20For%20Player%20In%20Volume
  ;[0.1763 0.8701 1] Update players vortex parameters
  (for _array_element (Variables|Default|GetAffectedPawns)
    (CallFunction|SetPlayerinVortexVolume _array_element)))

(event Custom|LockPlayersInVolumeFunction
  (params (Instigator :type Controller%20Object%20Reference))
  ;[0.492 0.2848 1] Receivers
  (Variables|Default|SetbLockPlayersInVolume true)
  (CallFunction|UpdateVortexParametersforPlayerinVolume))

(event Custom|AllowPlayersToLeaveVolumeFunction
  (params (Instigator :type Controller%20Object%20Reference))
  ;[0.492 0.2848 1] Receivers
  (Variables|Default|SetbLockPlayersInVolume false)
  (CallFunction|UpdateVortexParametersforPlayerinVolume))

(event Custom|On%20Player%20Exit%20Any%20Island
  (params (Client :type Player%20State%20Object%20Reference) (Volume :type Actor%20Object%20Reference))
  ;[1 0.4875 0.1602] Remove volume effects from player if they leave the bounds of our volume and are currently being affected
  (if (== (Minigame|GetMinigame Volume :!class /Script/FortniteGame.MinigameBlueprintLibrary) (Variables|Default|GetCachedMinigameForActor))
    (Utilities|IsValid Client
      (:Is%20Valid
        (bind _asfort_player_pawn (Utilities|Casting|CastToFortPlayerPawn (PlayerState|GetPawn Client)))
        (if (Utilities|Array|ContainsItem (Variables|Default|GetAffectedPawns) _asfort_player_pawn)
          (CallFunction|PawnExitedTriggerVolume _asfort_player_pawn) ; Remove the effects of this volume from the player
        )))))

(event EventEndPlay
  (params (EndPlayReason :type EEndPlayReason%20Enum :default "Destroyed"))
  (bind _exitedeventhandle (Variables|Default|GetExitedEventHandle))
  ;[1 0.2283 0.06831] End play
  (|CallParentFunction EndPlayReason)
  (Utilities|FlowControl|Sequence
    (:then_0
      ;;[0.9008 1 0.04388] Unbind events.
      (Utilities|IsValid _exitedeventhandle
        (:Is%20Valid
          (GameplayEventRouter|StopListeningtoEvent _exitedeventhandle))))
    (:then_1
      ;;[1 0.5485 0.04598] Unregister any remaining pawns in the volume.
      (for _array_element (Variables|Default|GetAffectedPawns)
        (Utilities|IsValid _array_element
          (:Is%20Valid
            (Registration|UnregisterPlayerinVolume _array_element)))))))

(fn GetCollisionStaticMesh
  (return "0, 0, 0" (Variables|/Script/Engine.StaticMeshComponent|GetStaticMesh (Variables|Default|GetCreativeIcon))))

(fn GetMeshComponents
  (return (Utilities|Array|MakeArray (Variables|Default|GetCreativeIcon) (Variables|Default|GetEdit%20Only%20Zone) (Variables|Default|GetCreativeEditOnlyMesh))))

(fn GetVisualComponents
  (return (Utilities|Array|MakeArray (Variables|Default|GetEdit%20Only%20Zone) (Variables|Default|GetCreativeIcon) (Variables|Default|GetCreativeEditOnlyMesh)) (Utilities|Array|MakeArray (Variables|Default|GetIn%20Game%20Zone))))

(fn GetVolumeSizeParameters
  (bind _trigger (Variables|Default|GetTrigger))
  ;[1 0.5402 0.07099] Constructs the transform the vortex volume should use based on the devices trigger volume, and calculates the thickness of the padding at the edge of the volume that pushes the player toward the center
  (Utilities|IsValid _trigger
    (:Is%20Valid
      (return (Math|Transform|MakeTransform (Transformation|GetWorldLocation _trigger) (Transformation|GetWorldRotation _trigger) (Components|Box|GetScaledBoxExtent _trigger)) (Math|Float|Max(Float) (* (Math|Vector|GetMinElement (Components|Box|GetScaledBoxExtent _trigger)) (Variables|VortexParameters|GetInnerBoundsSizeRatio)) (Variables|VortexParameters|GetMinBoundsSize))))))

(fn GetVortexDirection
  (params (For%20Actor :type Actor%20Object%20Reference))
  (bind _asfort_player_pawn (Utilities|Casting|CastToFortPlayerPawn For%20Actor)
    (:then
      (return (Utilities|Array|ContainsItem (Variables|Default|GetAffectedPawns) _asfort_player_pawn) (* (Transformation|GetActorUpVector) (Math|Float|Sign(Float) (Variables|UserKnobs|GetPushForce)))))
    (:CastFailed
      (return false "0, 0, 0"))))

(fn HandleDeviceScaling
  (bind (_relativescale3d_x _relativescale3d_y _relativescale3d_z) (Math|Vector|BreakVector (Variables|/Script/Engine.SceneComponent|GetRelativeScale3D_X (Variables|Debug|GetStaticMeshComponent))))
  (Transformation|SetWorldScale3D (Variables|Default|GetCreativeIcon) (Math|Vector|MakeVector (* (* (Utilities|Operators|Add _relativescale3d_x _relativescale3d_y _relativescale3d_z) 0.33) (Variables|Default|GetIconScalar)) (* (* (Utilities|Operators|Add _relativescale3d_x _relativescale3d_y _relativescale3d_z) 0.33) (Variables|Default|GetIconScalar)) (* (* (Utilities|Operators|Add _relativescale3d_x _relativescale3d_y _relativescale3d_z) 0.33) (Variables|Default|GetIconScalar)))))

(fn IgnoreWorldGridScaleOverride
  (return false))

(fn InitializeCreativeIcone
  (Resizing|HandleDeviceScaling)
  (if (not (Editor|InEditor))
    (Class|CreativeEditOnlyMeshComponent|ReloadMeshandMaterials (Variables|Default|GetCreativeIcon))))

(fn RemovePawnFromVolume
  (params (Player%20Pawn :type Fort%20Player%20Pawn%20Object%20Reference)))

(fn ResizeFrame
  (Variables|Default|SetMaterialScalar (Variables|Default|GetFrameSizeOverride))
  (|CallParentFunction))

(fn SetPlayerInVortexVolume
  (params (Player :type Fort%20Player%20Pawn%20Object%20Reference))
  (bind (_volume_transform _volume_thickness) (CallFunction|GetVolumeSizeParameters))
  ;[0.2104 1 0.5208] Sets the given player to be in a vortex volume with parameters from user options. This event can be used to either initially set the player as in a vortex or update the parameters of a player already in a vortex
  (Utilities|IsValid Player
    (:Is%20Valid
      (Vortex|SetinVortexVolume Player true (Variables|UserKnobs|GetLaunchVelocity) _volume_transform _volume_thickness (* (Variables|UserKnobs|GetPushForce) -1.0) (select (CallFunction|ShouldLockPlayerinVolume Player) 0.0 (Variables|VortexParameters|GetLeavingVolumeTerminalVelocity)) (Variables|UserKnobs|GetbAllowGliderDeploy)))))

(fn ShouldLockPlayerInVolume
  (params (Player%20Actor :type Actor%20Object%20Reference))
  (bind _creativeuserestrictioncomponent (Variables|Default|GetCreativeUseRestrictionComponent))
  ;[1 0.3582 0.06447] Returns whether player should not be able to leave the volume based on channel reciver, locked team option, and locked class option
  (Utilities|IsValid _creativeuserestrictioncomponent
    (:Is%20Valid
      (bind _valid_controller (DEPRECATED|ActorHasValidController _creativeuserestrictioncomponent Player%20Actor true true))
      (return (and _valid_controller (Variables|Default|GetbLockPlayersInVolume))))
    (:Is%20Not%20Valid
      (return false))))

(fn UpdateTriggerCollision
  (Collision|SetCollisionEnabled (Variables|Default|GetTrigger) (select (Class|CreativeEnabledComponent|IsEnabled (Variables|Default|GetEnabledComponent)) "QueryAndPhysics" "NoCollision")))

(fn ConstructionScript
  (|CallParentFunction))
