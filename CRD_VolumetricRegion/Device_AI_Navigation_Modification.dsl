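; Creative AI Navigation Modification device: a resizable, shaped volume that changes the navmesh area class inside it.
; Zone shape and size options drive the async-loaded mesh, scale and frame; enabled state and modification type drive the nav modifier, reverting to FortNavArea when disabled.
; During play, the server teleports tagged AI pawns overlapping the volume out, using a reverse line trace to find its edge.
; Cosmetic-only logic hides or fades bounds by play mode and local-player overlap.
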(event EventInitializeDevice
  ;[0 1 0.1154] Initialization
  (|CallParentFunction)
  (CallFunction|SetBarrierVisuals) ; async loads the appropriate Barrier mesh \nthen sets the scale
  (Visibility|UpdateBaseVisibility)
  (Navigation|UpdateNavVolumeState))

(event EventBeginPlay
  ;[0 1 0.1154] Initialization
  (|CallParentFunction)
  (Networking|SwitchHasAuthority
    (:Authority
      (Collision|SetCollisionEnabled (Variables|Default|GetFxBarrierOverlapVolume) "NoCollision"))))

(event Custom|Set%20Barrier%20Visuals
  (bind _zoneshape (Variables|UserOptions|GetZoneShape))
  (bind _volume_shapes (Variables|Default|GetVolume%20Shapes))
  (bind (_value _returnvalue) (Utilities|Map|Find _volume_shapes _zoneshape))
  ;[0 0.2775 2] Set Mesh
  (if (and (Utilities|IsValidSoftObjectReference _value) _returnvalue)
    (bind _loaded_object (Creative|Loading|AsyncLoadEditorSafe _value)
      (:Loaded
        (bind _asstatic_mesh (Utilities|Casting|CastToStaticMesh _loaded_object))
        (Components|StaticMesh|SetStaticMesh (Variables|Default|GetBarrierBounds) _asstatic_mesh)
        (Components|StaticMesh|SetStaticMesh (Variables|Default|GetBarrierBounds1) _asstatic_mesh)
        (Default|SetBarrierScale)
        (CallFunction|SetBarrierMaterial))) ; AsyncLoadEditorSafe
  ))

(event Custom|Fade%20Barrier%20Bounds%20Visibility
  (params (On :type Boolean) (Immediate :type Boolean))
  ;[1 0.7982 0] Fade Visual Bound Visibility
  (if (not (CallFunction|IsinPlayMode))
    (Networking|CanExecuteCosmeticEvents
      (:True
        (if (not Immediate)
          (Development|SetHiddeninGame (Variables|Default|GetBarrierBounds) (not (Default|BarrierBoundsNotDisabled)) false)
          (if On
            (bind (_direction _lerp) (Timeline|Barrier%20Fade Reverse :NewTime (Math|Float|SelectFloat 0.0 (Components|Timeline|GetTimelineLength (Variables|Components|GetFadeBoundsTL)) On))
              (:Update
                (Rendering|Material|SetScalarParameterValue (Variables|Default|GetBoundsMID) "Opacity" _lerp :!class /Script/Engine.MaterialInstanceDynamic))
              (:Finished
                (switch Utilities|FlowControl|Switch|SwitchonETimelineDirection _direction
                  (:Forward
                    (Development|SetHiddeninGame (Variables|Default|GetBarrierBounds) false false))))) ; 0s = Full Opacity\n0.3s (End) = No Opacity
            (else
              (Timeline|Barrier%20Fade Play)))
          (else
            (Development|SetHiddeninGame (Variables|Default|GetBarrierBounds) (not (and On (Default|BarrierBoundsNotDisabled))) false)
            (Rendering|Material|SetScalarParameterValue (Variables|Default|GetBoundsMID) "Opacity" (Math|Float|SelectFloat 0.0 1.0 On) :!class /Script/Engine.MaterialInstanceDynamic)
            (Timeline|Barrier%20Fade SetNewTime)))))
    (else
      (Development|SetHiddeninGame (Variables|Default|GetBarrierBounds) true false))))

(event Custom|Set%20Barrier%20Material
  ;[1 0.6163 0] Asynchronously Load Material for the Barrier Bounds, then update visuals
  (Networking|CanExecuteCosmeticEvents
    (:True
      (Rendering|Material|SetMaterial (Variables|Default|GetBarrierBounds) 0 "/CRD_VolumetricRegion/Creative/Devices/AI_Navigation_Modification/Materials/MI_CP_AI_Navigation_Modification_01_Volume_Placed.MI_CP_AI_Navigation_Modification_01_Volume_Placed") ; Reattach MID (In case phone tool has \nstolen it)
      (Default|UpdateBoundsVisibility))))

(event OnComponentBeginOverlap(FxBarrierOverlapVolume)
  (params (OverlappedComponent :type Primitive%20Component%20Object%20Reference) (OtherActor :type Actor%20Object%20Reference) (OtherComp :type Primitive%20Component%20Object%20Reference) (OtherBodyIndex :type Integer) (bFromSweep :type Boolean) (SweepResult :type Hit%20Result%20Structure%20%28by%20ref%29))
  ;[1 1 1] Overlap Effects (local client only)
  (Default|FadeonOverlap false OtherActor))

(event Custom|EnabledOnReceived
  (params (Instigator :type Controller%20Object%20Reference))
  ;[0.6648 0 1] Transmitters and receivers
  (Class|CreativeEnabledComponent|SetEnabled (Variables|Default|GetEnabledComponent) true))

(event ChildFunctions|EventPlayModeChanged
  (params (Minigame :type Fort%20Minigame%20Object%20Reference) (bIsInPlayMode :type Boolean))
  ;[0 1 0.8098 0.25] State Changes
  (Default|UpdateDeviceVisuals)
  (Navigation|UpdateNavVolumeState)
  (Visibility|UpdateBaseVisibility))

(event Custom|DisabledOnReceived
  (params (Instigator :type Controller%20Object%20Reference))
  ;[0.6648 0 1] Transmitters and receivers
  (Class|CreativeEnabledComponent|SetEnabled (Variables|Default|GetEnabledComponent) false))

(event EventOnMinigameStarted
  ;[0 1 0.8098 0.25] State Changes
  (Class|CreativeEnabledComponent|SetEnabled (Variables|Default|GetEnabledComponent) (Variables|UserOptions|GetEnabledOnGameStart)))

(event OnComponentEndOverlap(FxBarrierOverlapVolume)
  (params (OverlappedComponent :type Primitive%20Component%20Object%20Reference) (OtherActor :type Actor%20Object%20Reference) (OtherComp :type Primitive%20Component%20Object%20Reference) (OtherBodyIndex :type Integer))
  ;[1 1 1] Overlap Effects (local client only)
  (Default|FadeonOverlap true OtherActor))

(event OnEnabledStateChanged(EnabledComponent)
  (params (Enabled :type Boolean))
  ;[0 1 0.8098 0.25] State Changes
  (Navigation|UpdateNavVolumeState))

(event OnComponentBeginOverlap(BarrierBounds1)
  (params (OverlappedComponent :type Primitive%20Component%20Object%20Reference) (OtherActor :type Actor%20Object%20Reference) (OtherComp :type Primitive%20Component%20Object%20Reference) (OtherBodyIndex :type Integer) (bFromSweep :type Boolean) (SweepResult :type Hit%20Result%20Structure%20%28by%20ref%29))
  (bind (_bblockinghit _binitialoverlap _time _distance _location _impactpoint) (Collision|BreakHitResult SweepResult))
  ;[1 0.4422 0.03086] Teleport any AI Pawns stuck inside barriers when spawning into them
  (Networking|SwitchHasAuthority
    (:Authority
      (if (ForcedPawnMovement|ShouldTeleportOverlappingPawns OtherActor)
        (bind _returnvalue (GameplayTags|HasAnyMatchingGameplayTags(Message) OtherActor (GameplayTags|MakeGameplayTagContainerfromArray (Variables|Default|GetEnemyPawnTags))))
        (if _returnvalue
          (bind _asfort_aipawn (Utilities|Casting|CastToFortAIPawn OtherActor))
          (ForcedPawnMovement|TeleportAIPawnToNearestPoint _asfort_aipawn (select (Math|Vector|VectorIsNearlyZero _impactpoint) (Transformation|GetActorLocation OtherActor) _impactpoint)))))))

(fn FadeOnOverlap
  (params (FadeUp :type Boolean) (Actor :type Actor%20Object%20Reference))
  (Variables|Default|Setlocal%20Fade%20Up FadeUp)
  (Networking|CanExecuteCosmeticEvents
    (:True
      (bind _aspawn (Utilities|Casting|CastToPawn Actor))
      (if (Pawn|IsLocallyControlled _aspawn)
        (Variables|Default|SetLocalPlayerIsWithinFxVolume (not (Variables|Default|Getlocal%20Fade%20Up)))
        (if (and (Variables|UserOptions%28Developer%29|GetEnableDeveloperMaterials) (Variables|Default|GetAllowBoundsTransition))
          (CallFunction|FadeBarrierBoundsVisibility (Variables|Default|Getlocal%20Fade%20Up)))))))

(fn GetCreativeActorBounds
  (return false "0, 0, 0"))

(fn GetCreativeActorOrigin
  (return true (Variables|/Script/Engine.SceneComponent|GetRelativeLocation (Variables|Default|GetFxBarrierOverlapVolume))))

(fn GetCreativeActorRelativeTransform
  (return false 0))

(fn GetGridAxisGuideAttachmentPoint
  (return false false "0, 0, 0"))

(fn GetVolumeScale
  (return (Math|Vector|MakeVector (Math|Float|Max(Float) (Variables|UserOptions|Basic|GetBarrierWidth) 0.05) (Math|Float|Max(Float) (Variables|UserOptions|Basic|GetBarrierDepth) 0.05) (Variables|UserOptions|Basic|GetBarrierHeight))))

(fn IsTargetableByLinearMover
  (return true))

(fn ResizeFrame
  (Class|CreativeVolumeFrame|InitializeComponent (Variables|Default|GetCreativeVolumeFrame) (Variables|UserOptions|GetZoneShape) 1.0 (Utilities|Operators|Multiply (CallFunction|GetVolumeScale) (5.12 5.12 3.84))))

(fn SetBarrierScale
  (bind _zoneshape (Variables|UserOptions|GetZoneShape))
  (bind _volume_transform (Variables|ChildFunctions|GetVolume%20Transform))
  (bind (_value _returnvalue) (Utilities|Map|Find _volume_transform _zoneshape))
  (bind _value_location (Math|Transform|BreakTransform _value))
  (bind _scale (CallFunction|GetVolumeScale))
  (bind (_value_2 _returnvalue_2) (Utilities|Map|Find (Variables|ChildFunctions|GetVolume%20Transform) (Variables|UserOptions|GetZoneShape)))
  (bind (_value_location_2 _value_rotation _value_scale) (Math|Transform|BreakTransform _value_2))
  (bind _fxbarrieroverlapvolume (Variables|Default|GetFxBarrierOverlapVolume))
  (Utilities|FlowControl|Sequence
    (:then_0
      (LocalVariables|Setlocal%20scale (* (CallFunction|GetVolumeScale) (Math|Vector|SelectVector _value_scale "1.000000,1.000000,1.000000" _returnvalue_2)))
      (Transformation|SetRelativeScale3D (Variables|Default|GetBarrierBounds) (LocalVariables|Getlocal%20scale))
      (CallFunction|ResizeFrame)
      (Transformation|SetRelativeScale3D (Variables|Default|GetBarrierBounds) (LocalVariables|Getlocal%20scale))
      (Transformation|SetRelativeScale3D _fxbarrieroverlapvolume _scale)
      (Transformation|SetRelativeLocation _fxbarrieroverlapvolume (Math|Vector|MakeVector 0.0 0.0 (* (.z _scale) 192.0)) false false))
    (:then_1
      (if _returnvalue
        (Transformation|SetRelativeLocation (Variables|Default|GetBarrierOffset) (* (LocalVariables|Getlocal%20scale) _value_location) false false)))))

(fn ShouldInvisiblePropsBeAllowed
  (return true))

(fn ShouldShowInGame
  (return (Math|Boolean|ANDBoolean (Class|CreativeEnabledComponent|IsEnabled (Variables|Default|GetEnabledComponent)) (Variables|Default|GetBaseVisibleDuringGame) (not (Replay|GetIsPlayingReplay)))))

(fn ShouldTeleportOverlappingPawns
  (params (Other%20Actor :type Actor%20Object%20Reference))
  ;[0.8977 1 0.2243] Return if the volume should be able to teleport overlapping AI Pawns
  (Utilities|Casting|CastToFortPlayerPawn Other%20Actor
    (:CastFailed
      (bind (_outactors _returnvalue) (Collision|ComponentOverlapActors (Variables|Default|GetBarrierBounds1) (Transformation|GetWorldTransform (Variables|Default|GetBarrierBounds1)) (Utilities|Array|MakeArray "ObjectTypeQuery7")))
      (return (and _returnvalue (CallFunction|IsinPlayMode))))))

(fn TeleportAIPawntoNearestPoint
  (params (AI%20Pawn :type Fort%20AIPawn%20Object%20Reference) (ImpactPoint :type Vector))
  (bind _returnvalue (ForcedPawnMovement|TryTeleportPawntoDirection (Variables|GetAI%20Pawn) (Math|Vector|Normalize (select (Math|Vector|VectorIsNearlyZero (Transformation|GetVelocity (Variables|GetAI%20Pawn))) (- ImpactPoint (Collision|GetComponentBounds (Variables|Default|GetBarrierBounds1))) (* (Transformation|GetVelocity (Variables|GetAI%20Pawn)) -1.0))) (Variables|GetImpactPoint)))
  (return _returnvalue))

(fn TryTeleportPawnToDirection
  (params (Pawn :type Fort%20AIPawn%20Object%20Reference) (Direction :type Vector) (ImpactPoint :type Vector))
  (bind (_hitlocation _hitnormal _bonename _outhit _returnvalue) (Collision|LineTraceComponent (Variables|Default|GetBarrierBounds1) (+ (* Direction 1536.0) ImpactPoint) ImpactPoint true false true))
  (if _returnvalue
    (bind _returnvalue_2 (AI|TeleportAIPawn (Variables|GetPawn) (+ (* Direction 3e+02) _hitlocation) (Transformation|GetActorRotation (Variables|GetPawn))))
    (return _returnvalue_2)
    (else
      (return false))))

(fn UpdateDeviceVisuals
  (Default|UpdateBoundsVisibility)
  (Default|UpdateFade (or (Class|CreativeEnabledComponent|IsEnabled (Variables|Default|GetEnabledComponent)) (not (CallFunction|IsinPlayMode)))))

(fn UpdateNavVolumeState
  ;[1 0.457 0.4371] Set the Nav Settings of the Volume
  (if (Class|CreativeEnabledComponent|IsEnabled (Variables|Default|GetEnabledComponent))
    (Class|FortNavModifierComponent|SetComponentAffectingNavMesh (Variables|Default|GetFortNavModifier) (Variables|Default|GetBarrierBounds1))
    (AI|Navigation|SetAreaClass (Variables|Default|GetFortNavModifier) (Utilities|Map|Find (Variables|Default|GetNav%20Modificaiton%20to%20Area%20Class%20Map) (Variables|UserOptions|Basic|GetNavigation%20Modification%20Type)) :!class /Script/NavigationSystem.NavModifierComponent)
    (Default|UpdateDeviceVisuals)
    (Navigation|UpdateNavigationProperties)
    (else
      (Class|FortNavModifierComponent|SetComponentAffectingNavMesh (Variables|Default|GetFortNavModifier) (Variables|Default|GetBarrierBounds1))
      (AI|Navigation|SetAreaClass (Variables|Default|GetFortNavModifier) "/Script/FortniteAI.FortNavArea" :!class /Script/NavigationSystem.NavModifierComponent)
      (Default|UpdateDeviceVisuals)
      (Navigation|UpdateNavigationProperties))))

(fn UpdateBaseVisibility
  (Visibility|SetVisiblityDuringPhase (Variables|Default|GetVisibleInGameComponent) (select (Visibility|ShouldShowinGame) "NewEnumerator3" "NewEnumerator1")))

(fn UpdateBoundsVisibility
  (Visibility|UpdateBaseVisibility)
  (UEFNConversionHelper|SetComponentHiddenEditor (Variables|Default|GetBarrierBounds) (CallFunction|IsinPlayMode))
  (Rendering|SetCastShadow (Variables|Default|GetBarrierBounds) false)
  (return (Default|BarrierBoundsNotDisabled)))

(fn UpdateFade
  (params (On :type Boolean) (Immediate :type Boolean))
  (Networking|CanExecuteCosmeticEvents
    (:True
      (Variables|Default|SetAllowBoundsTransition On)
      (CallFunction|FadeBarrierBoundsVisibility (and On (or (not (Variables|Default|GetLocalPlayerIsWithinFxVolume)) (CallFunction|IsinPlayMode))) Immediate) ; Only Fade if player is outwith bounds \n(If minigame isn't running)
    )))

(fn UpdateNavigationProperties
  (bind _enabledcomponent (Variables|Default|GetEnabledComponent))
  (bind _enabled (Class|CreativeEnabledComponent|IsEnabled _enabledcomponent))
  (CallFunction|SetIsNavigationRelevant _enabled)
  (Navigation|SetAlwaysExportNavRelevantComponent _enabled)
  (Navigation|SetNavigationObstacleOverride "ForceEnabled"))

(fn ConstructionScript
  (|CallParentFunction)
  (Default|SetBarrierScale))
