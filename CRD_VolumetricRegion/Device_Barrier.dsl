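; Creative Barrier device: a scalable, shaped volume that blocks pawns, and optionally weapons, with configurable visuals.
; Shape and zone knobs drive async-loaded mesh, material and scale; enabled and play-mode state drive collision, visibility and navigation.
; A cosmetic overlap box fades the bounds for the local player through a timeline.
; On authority, overlapping players outside vehicles get line-trace teleported out or killed on failure; tagged non-player pawns are killed.
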(event EventInitializeDevice
  ;[0 1 0.1154] Initialization
  (|CallParentFunction)
  (Default|UpdateBlockingChannels)
  (CallFunction|SetBarrierVisuals) ; async loads the appropriate Barrier mesh \nthen sets the scale
  (Visibility|UpdateBaseVisibility)
  (Default|UpdateBarrierState)
  ;; SetupFrameVisibilityComponent
  (Initialization|InitializeWithSceneComponents (Variables|Default|GetFrameVisibility) (Utilities|Array|MakeArray (Variables|Default|GetFrameMesh)) (Utilities|Map|MakeMap) "NewEnumerator4"))

(event EventBeginPlay
  ;[0 1 0.1154] Initialization
  (|CallParentFunction)
  (Networking|SwitchHasAuthority
    (:Authority
      (Collision|SetCollisionEnabled (Variables|Default|GetFxBarrierOverlapVolume) "NoCollision"))))

(event Custom|Set%20Barrier%20Visuals
  (bind _knob_volumeshape (Variables|UserKnobs|GetKnob_VolumeShape))
  (bind _volume_shapes (Variables|Default|GetVolume%20Shapes))
  (bind (_value _returnvalue) (Utilities|Map|Find _volume_shapes _knob_volumeshape))
  ;[0 0.2775 2] Set Mesh
  (if (and (Utilities|IsValidSoftObjectReference _value) _returnvalue)
    (bind _loaded_object (Creative|Loading|AsyncLoadEditorSafe _value)
      (:Loaded
        (bind _asstatic_mesh (Utilities|Casting|CastToStaticMesh _loaded_object))
        (Components|StaticMesh|SetStaticMesh (Variables|Default|GetBarrierBounds) _asstatic_mesh)
        (Default|SetBarrierScale)
        (CallFunction|SetBarrierMaterial))) ; AsyncLoadEditorSafe
  ))

(event OnMeshandMaterialsLoaded(FrameMesh)
  ;[0 0.2775 2] Set Mesh
  (CallFunction|ResizeFrame))

(event Custom|Fade%20Barrier%20Bounds%20Visibility
  (params (On :type Boolean) (Immediate :type Boolean))
  ;[1 0.7982 0] Fade Visual Bound Visibility
  (Networking|CanExecuteCosmeticEvents
    (:True
      (if (not Immediate)
        (Development|SetHiddeninGame (Variables|Default|GetBarrierBounds) (not (Default|BarrierBoundsNotDisabled)) false)
        (if On
          (bind (_direction _lerp) (Timeline|Barrier%20Fade Reverse :NewTime (Math|Float|SelectFloat 0.0 (Components|Timeline|GetTimelineLength (Variables|Components|GetFadeBoundsTL)) On))
            (:Update
              (Rendering|Material|SetScalarParameterValue (Variables|Default|GetBoundsMID) "Opacity" _lerp :!class /Script/Engine.MaterialInstanceDynamic))
            (:Finished
              (switch Utilities|FlowControl|Switch|SwitchonETimelineDirection _direction
                (:Forward
                  (Development|SetHiddeninGame (Variables|Default|GetBarrierBounds) true false))))) ; 0s = Full Opacity\n0.3s (End) = No Opacity
          (else
            (Timeline|Barrier%20Fade Play)))
        (else
          (Development|SetHiddeninGame (Variables|Default|GetBarrierBounds) (not (and On (Default|BarrierBoundsNotDisabled))) false)
          (Rendering|Material|SetScalarParameterValue (Variables|Default|GetBoundsMID) "Opacity" (Math|Float|SelectFloat 0.0 1.0 On) :!class /Script/Engine.MaterialInstanceDynamic)
          (Timeline|Barrier%20Fade SetNewTime))))))

(event Custom|Set%20Barrier%20Material
  ;[1 0.6163 0] Asynchronously Load Material for the Barrier Bounds, then update visuals
  (Networking|CanExecuteCosmeticEvents
    (:True
      (if (Building|IsPreviewBuildingActor)
        (CallFunction|OverrideBoundsOnPreview)
        (else
          (bind (_bhasshadow _material) (Utilities|ASyncGetBoundsMaterial)
            (:Visible%20Material
              (Variables|Default|SetBoundsMID (Material|GetMIDfromMaterial (Variables|Default|GetBarrierBounds) 0 _material))
              (Variables|Default|SetisOpaqueBoundsMaterial _bhasshadow))
            (:Invisible%20Material
              (Variables|Default|SetBoundsMID 0)
              (Variables|Default|SetisOpaqueBoundsMaterial false)))
          (Default|UpdateBoundsVisibility))))))

(event OnComponentBeginOverlap(FxBarrierOverlapVolume)
  (params (OverlappedComponent :type Primitive%20Component%20Object%20Reference) (OtherActor :type Actor%20Object%20Reference) (OtherComp :type Primitive%20Component%20Object%20Reference) (OtherBodyIndex :type Integer) (bFromSweep :type Boolean) (SweepResult :type Hit%20Result%20Structure%20%28by%20ref%29))
  ;[1 1 1] Overlap Effects (local client only)
  (Default|FadeonOverlap false OtherActor))

(event Custom|EnabledOnReceived
  (params (Instigator :type Controller%20Object%20Reference))
  ;[0.6648 0 1] Transmitters and receivers
  (Class|CreativeEnabledComponent|SetEnabled (Variables|Default|GetEnabledComponent) true))

(event ChildFunctions|EventPlayModeChanged
  (params (Minigame :type Fort%20Minigame%20Object%20Reference) (bIsInPlayMode :type Boolean))
  ;[0 1 0.8098 0.25] State Changes
  (Collision|SetCollisionResponsetoChannel (Variables|Default|GetFxBarrierOverlapVolume) "ECC_Pawn" (select bIsInPlayMode "ECR_Ignore" "ECR_Overlap"))
  (Default|UpdateDeviceVisuals)
  (Default|UpdateBarrierState))

(event Custom|DisabledOnReceived
  (params (Instigator :type Controller%20Object%20Reference))
  ;[0.6648 0 1] Transmitters and receivers
  (Class|CreativeEnabledComponent|SetEnabled (Variables|Default|GetEnabledComponent) false))

(event OnComponentEndOverlap(FxBarrierOverlapVolume)
  (params (OverlappedComponent :type Primitive%20Component%20Object%20Reference) (OtherActor :type Actor%20Object%20Reference) (OtherComp :type Primitive%20Component%20Object%20Reference) (OtherBodyIndex :type Integer))
  ;[1 1 1] Overlap Effects (local client only)
  (Default|FadeonOverlap true OtherActor))

(event OnEnabledStateChanged(EnabledComponent)
  (params (Enabled :type Boolean))
  ;[0 1 0.8098 0.25] State Changes
  (Default|UpdateBarrierState))

(event OnComponentBeginOverlap(BarrierBounds1)
  (params (OverlappedComponent :type Primitive%20Component%20Object%20Reference) (OtherActor :type Actor%20Object%20Reference) (OtherComp :type Primitive%20Component%20Object%20Reference) (OtherBodyIndex :type Integer) (bFromSweep :type Boolean) (SweepResult :type Hit%20Result%20Structure%20%28by%20ref%29))
  (bind (_bblockinghit _binitialoverlap _time _distance _location _impactpoint) (Collision|BreakHitResult SweepResult))
  ;[1 0.4422 0.03086] Teleport any players stuck inside barriers when spawning into them - Does not apply to hollow boxes
  (Networking|SwitchHasAuthority
    (:Authority
      (if (PlayerStuck|ShouldTeleportOverlappingPawns)
        (bind _asfort_player_pawn (Utilities|Casting|CastToFortPlayerPawn OtherActor)
          (:then
            (Utilities|IsValid (Vehicle|GetVehicleActor _asfort_player_pawn :!class /Script/FortniteGame.FortPlayerPawn)
              (:Is%20Not%20Valid
                (bind _bsucceed (PlayerStuck|TeleportPlayerToNearestPoint _asfort_player_pawn (select (Math|Vector|VectorIsNearlyZero _impactpoint) (Transformation|GetActorLocation _asfort_player_pawn) _impactpoint)))
                (if (not _bsucceed)
                  (if (Pawn|IsBotControlled _asfort_player_pawn)
                    (Utilities|FlowControl|DelayUntilNextTick)
                    (FortPawn|ForceKill _asfort_player_pawn "(TagName=\"Gameplay.Damage.OutsideSafeZone\")" 0 _asfort_player_pawn)
                    (else
                      (FortPawn|ForceKill _asfort_player_pawn (Variables|Default|GetDeath%20Reason) 0 0)))))) ; If the player is NOT in a vehicle, then attempt to teleport them.\nWe do this because the pawn's overlap area may extend beyond the vehicle\ncausing the player to overlap the barrier even though their vehicle does NOT.\n\nIdeally we would detect vehicle overlaps and eject players in those vehicles\nbut,  currently some vehicles have overlaps bigger than their blocking collision and thus can end up overlapping a barrier they are not actually in.
          )
          (:CastFailed
            ;[1 0.7182 0.5835 0.77] Check for a relevant tag, and if possessed, kill the pawn (as long as it is a pawn)
            (bind _returnvalue (GameplayTags|HasAnyMatchingGameplayTags(Message) OtherActor (GameplayTags|MakeGameplayTagContainerfromArray (Variables|Default|GetEnemyPawnTags))))
            (if _returnvalue
              (bind _asfort_pawn (Utilities|Casting|CastToFortPawn OtherActor))
              (FortPawn|ForceKill _asfort_pawn (Variables|Default|GetDeath%20Reason) 0 0))))))))

(fn FadeOnOverlap
  (params (FadeUp :type Boolean) (Actor :type Actor%20Object%20Reference))
  (Variables|Default|Setlocal%20Fade%20Up (or FadeUp (Utilities|Enum|Equal(Enum) (Variables|UserKnobs|GetKnob_VolumeShape) "NewEnumerator2")))
  (Networking|CanExecuteCosmeticEvents
    (:True
      (bind _aspawn (Utilities|Casting|CastToPawn Actor))
      (if (Pawn|IsLocallyControlled _aspawn)
        (Variables|Default|SetLocalPlayerIsWithinFxVolume (not (Variables|Default|Getlocal%20Fade%20Up)))
        (if (and (or (>= (Variables|UserKnobs|GetZoneVisibiltyIndex) 2) (Variables|Default|GetUseDeveloperMaterials)) (Variables|Default|GetAllowBoundsTransition))
          (CallFunction|FadeBarrierBoundsVisibility (Variables|Default|Getlocal%20Fade%20Up)))))))

(fn GetClosestDirection
  (params (Pawn :type Actor%20Object%20Reference) (PossibleDirections :type Array%20of%20Vectors))
  (bind _barrierbounds1 (Variables|Default|GetBarrierBounds1))
  (bind _origin (Collision|GetComponentBounds _barrierbounds1))
  (bind _pawn (Variables|GetPawn))
  (bind _returnvalue (Transformation|GetActorLocation _pawn))
  (bind _returnvalue_2 (- _returnvalue _origin))
  (for (_array_element _array_index) PossibleDirections
    (if (or (== (LocalVariables|GetMinIndex) -1) (< (Math|Vector|Distance(Vector) _array_element _returnvalue_2) (LocalVariables|GetMinDistance)))
      (LocalVariables|SetMinIndex _array_index)
      (LocalVariables|SetMinDistance (Math|Vector|Distance(Vector) _array_element _returnvalue_2))))
  (return (Utilities|Array|Get(acopy) (Variables|GetPossibleDirections) (LocalVariables|GetMinIndex)) (LocalVariables|GetMinIndex)))

(fn GetOverrideMeshMaterial
  (params (MeshComp :type Static%20Mesh%20Component%20Object%20Reference) (MatIdx :type Integer))
  (if (Utilities|Equal(Object) MeshComp (Variables|Default|GetBarrierBounds))
    (return (Variables|Default|GetBoundsMID))
    (elif (Utilities|Equal(Object) MeshComp (Variables|Default|GetFrameMesh))
      (return (Variables|Default|GetFrameMID))
      (else
        (return (|CallParentFunction MeshComp MatIdx))))))

(fn IsVisibleInGame
  (return (Visibility|ShouldShowinGame)))

(fn IsHollowVolumeShape
  (params (VolumeShape :type EGameplayVolumeShape%20Enum :default "NewEnumerator0"))
  (return (Utilities|Array|ContainsItem (Variables|Default|GetHollowVolumeShapes) VolumeShape)))

(fn IsTargetableByLinearMover
  (return true))

(fn OverrideBoundsonPreview
  ;[0.8677 1 0.6756 0.75] If we are in preview, set a valid material and visibility if we are invisible, otherwise, let it go, because we draw the bounds anyway.
  (Networking|CanExecuteCosmeticEvents
    (:True
      (if (and (or (< (Variables|UserKnobs|GetZoneVisibiltyIndex) 2) (not (Utilities|Array|IsValidIndex (Variables|Default|GetBarrier%20Materials) (Variables|UserKnobs|GetZoneVisibiltyIndex)))) (not (Variables|Default|GetUseDeveloperMaterials)))
        (Rendering|Material|SetMaterial (Variables|Default|GetBarrierBounds) 0 (Variables|Default|GetPreview%20Material))
        (Development|SetHiddeninGame (Variables|Default|GetBarrierBounds) false false)))))

(fn ResizeFrame
  (bind _framemid (Variables|Default|GetFrameMID))
  (if (> (Rendering|Material|GetNumMaterials (Variables|Default|GetFrameMesh)) 0)
    (Material|ReattachOrCreateMID (Variables|Default|GetFrameMesh) _framemid)
    (Rendering|Material|SetVectorParameterValue _framemid "SizeBias" (Math|Conversions|ToLinearColor(Vector) (- (* (Variables|Default|GetMaterialScalar) (Math|Vector|MakeVector (Variables|UserKnobs|GetZoneXScale) (Variables|UserKnobs|GetZoneYScale) (Variables|UserKnobs|GetZoneZScale))) (Variables|Default|GetMaterialScalar))) :!class /Script/Engine.MaterialInstanceDynamic)
    (Rendering|SetBoundsScale (Variables|Default|GetFrameMesh) (* (Math|Float|Max(Float) (Variables|UserKnobs|GetZoneXScale) (Variables|UserKnobs|GetZoneYScale) (Variables|UserKnobs|GetZoneZScale)) 1.5))))

(fn SetBarrierScale
  (bind _knob_volumeshape (Variables|UserKnobs|GetKnob_VolumeShape))
  (bind _volume_offsets (Variables|ChildFunctions|GetVolume%20Offsets))
  (bind (_value _returnvalue) (Utilities|Map|Find _volume_offsets _knob_volumeshape))
  (bind (_barrier_scale_x _barrier_scale_y _barrier_scale_z) (Math|Vector|BreakVector (Variables|GetBarrier%20Scale_X)))
  (bind _fxbarrieroverlapvolume (Variables|Default|GetFxBarrierOverlapVolume))
  (bind (_relativelocation_x _relativelocation_y) (Math|Vector|BreakVector (Variables|/Script/Engine.SceneComponent|GetRelativeLocation_X (Variables|Default|GetFxBarrierOverlapVolume))))
  (Utilities|FlowControl|Sequence
    (:then_0
      (LocalVariables|Setlocal%20scale (Math|Vector|MakeVector (Variables|UserKnobs|GetZoneXScale) (Variables|UserKnobs|GetZoneYScale) (Variables|UserKnobs|GetZoneZScale)))
      (Variables|ChildFunctions|SetBarrier%20Scale (Math|Vector|ClampVectorSize (- (* (LocalVariables|Getlocal%20scale) (Variables|ChildFunctions|GetFXBarrierGridScale)) (Variables|ChildFunctions|GetBarrier%20Shrink)) 1.0 1e+06))
      (Transformation|SetRelativeScale3D (Variables|Default|GetBarrierBounds) (Variables|ChildFunctions|GetBarrier%20Scale))
      (CallFunction|ResizeFrame)
      (Components|Box|SetBoxExtent _fxbarrieroverlapvolume (* (LocalVariables|Getlocal%20scale) (LocalVariables|GetFXBarrierBoxExtent)) true)
      (Transformation|SetRelativeLocation _fxbarrieroverlapvolume (Math|Vector|MakeVector _relativelocation_x _relativelocation_y (* (Variables|UserKnobs|GetZoneZScale) 192.0)) false false)
      (Transformation|SetRelativeScale3D (Variables|Default|GetPreviewZone) (LocalVariables|Getlocal%20scale)))
    (:then_1
      (if _returnvalue
        (Transformation|SetRelativeLocation (Variables|Default|GetBarrierOffset) (Math|Vector|MakeVector 0.0 0.0 (* _barrier_scale_z _value)) false false)))))

(fn ShouldInvisiblePropsBeAllowed
  (return true))

(fn ShouldShowInGame
  (return (Math|Boolean|ANDBoolean (Class|CreativeEnabledComponent|IsEnabled (Variables|Default|GetEnabledComponent)) (Variables|Default|GetBaseMeshVisibleInGame) (not (Replay|GetIsPlayingReplay)))))

(fn ShouldTeleportOverlappingPawns
  ;[0.8977 1 0.2243] Return if the volume should be able to teleport overlapping players
  (return (and (CallFunction|IsinPlayMode) (not (Utilities|Enum|Equal(Enum) (Variables|UserKnobs|GetKnob_VolumeShape) "NewEnumerator2")))))

(fn TeleportPlayertoNearestPoint
  (params (PlayerPawn :type Fort%20Player%20Pawn%20Object%20Reference) (ImpactPoint :type Vector))
  (bind _returnvalue (PlayerStuck|TryTeleportPlayertoDirection (Variables|GetPlayerPawn) (Math|Vector|Normalize (select (Math|Vector|VectorIsNearlyZero (Transformation|GetVelocity (Variables|GetPlayerPawn))) (- ImpactPoint (Collision|GetComponentBounds (Variables|Default|GetBarrierBounds1))) (* (Transformation|GetVelocity (Variables|GetPlayerPawn)) -1.0))) (Variables|GetImpactPoint)))
  (return _returnvalue))

(fn TryTeleportPlayerToAnySide
  (params (Pawn :type Fort%20Player%20Pawn%20Object%20Reference) (ImpactPoint :type Vector))
  (bind _teleportpossibledirection (Variables|PlayerStuck|GetTeleportPossibleDirection))
  (bind _pawn (Variables|GetPawn))
  (bind _possibledirectionsremaining (LocalVariables|GetPossibleDirectionsRemaining))
  (LocalVariables|SetPossibleDirectionsRemaining _teleportpossibledirection)
  (for _ _teleportpossibledirection
    (bind (_nearestdirection _directionindex) (PlayerStuck|GetClosestDirection _pawn _possibledirectionsremaining))
    (Utilities|Array|RemoveIndex _possibledirectionsremaining _directionindex)
    (bind _returnvalue (PlayerStuck|TryTeleportPlayertoDirection _pawn _nearestdirection (Variables|GetImpactPoint)))
    (if _returnvalue
      (return true)))
  (return false))

(fn TryTeleportPlayerToDirection
  (params (Pawn :type Fort%20Player%20Pawn%20Object%20Reference) (Direction :type Vector) (ImpactPoint :type Vector))
  (bind (_hitlocation _hitnormal _bonename _outhit _returnvalue) (Collision|LineTraceComponent (Variables|Default|GetBarrierBounds1) (+ (* Direction 1536.0) ImpactPoint) ImpactPoint true false true))
  (if _returnvalue
    (bind _returnvalue_2 (Transform|Actor|TeleportPlayerPawn (Variables|GetPawn) _hitlocation (Transformation|GetActorRotation (Variables|GetPawn))))
    (return _returnvalue_2)
    (else
      (return false))))

(fn UpdateBarrierState
  ;[1 0.457 0.4371] Set the Collision State of the Barrier and kill volume and then update the visuals of the barrier
  (Collision|SetCollisionEnabled (Variables|Default|GetBarrierBounds) (select (or (Class|CreativeEnabledComponent|IsEnabled (Variables|Default|GetEnabledComponent)) (and (not (CallFunction|IsinPlayMode)) (== (Variables|LegacyOptions|GetEnabledIndex) 0))) "QueryAndPhysics" "NoCollision"))
  (Default|UpdateDeviceVisuals)
  (CallFunction|SetasBuildingVolume (Utilities|Enum|Equal(Enum) (Variables|UserKnobs|GetKnob_VolumeShape) "NewEnumerator2"))
  (Navigation|UpdateNavigationProperties))

(fn UpdateDeviceVisuals
  (Default|UpdateBoundsVisibility)
  (Default|UpdateFade (or (Class|CreativeEnabledComponent|IsEnabled (Variables|Default|GetEnabledComponent)) (not (CallFunction|IsinPlayMode)))))

(fn UpdateBaseVisibility
  (bind _returnvalue (Minigame|GetMinigameforActor self))
  (Utilities|IsValid _returnvalue
    (:Is%20Valid
      (LocalVariables|SetIsPlayMode (Minigame|IsinPlayMode _returnvalue :!class /Script/FortniteGame.FortMinigame)))
    (:Is%20Not%20Valid
      _))
  (Visibility|SetVisiblityDuringPhase (Variables|Default|GetVisibleInGameComponent) (select (select (LocalVariables|GetIsPlayMode) (Visibility|ShouldShowinGame) (Variables|Default|GetBaseMeshVisibleInGame)) "NewEnumerator3" "NewEnumerator1")))

(fn UpdateBlockingChannels
  (bind _blockweaponsingame (Variables|Default|GetBlockWeaponsInGame))
  (bind _returnvalue (select _blockweaponsingame "ECR_Block" "ECR_Ignore"))
  (bind _barrierbounds (Variables|Default|GetBarrierBounds))
  (Collision|SetCollisionResponsetoChannel _barrierbounds "ECC_GameTraceChannel7" _returnvalue)
  (Collision|SetCollisionResponsetoChannel _barrierbounds "ECC_GameTraceChannel5" _returnvalue)
  (Collision|SetCollisionResponsetoChannel _barrierbounds "ECC_GameTraceChannel3" _returnvalue)
  (Collision|SetCollisionResponsetoChannel _barrierbounds "ECC_GameTraceChannel8" _returnvalue)
  (Collision|SetCollisionResponsetoChannel _barrierbounds "ECC_GameTraceChannel12" _returnvalue)
  ;[0.05415 0.4247 1] Setting device to ignore pawn collision while it is a preview actor.
  (Collision|SetCollisionResponsetoChannel _barrierbounds "ECC_Pawn" (select (Class|CreativeBlueprintLibrary|IsAPreviewActor self) "ECR_Ignore" "ECR_Block")))

(fn UpdateBoundsVisibility
  (bind _returnvalue (Default|BarrierBoundsNotDisabled))
  (Visibility|UpdateBaseVisibility)
  (Development|SetHiddeninGame (Variables|Default|GetBarrierBounds) (not (and _returnvalue (Math|Boolean|ORBoolean (not (Variables|Default|GetLocalPlayerIsWithinFxVolume)) (CallFunction|IsinPlayMode) (Utilities|Enum|Equal(Enum) (Variables|UserKnobs|GetKnob_VolumeShape) "NewEnumerator2")))) false)
  (Rendering|SetCastShadow (Variables|Default|GetBarrierBounds) false)
  (return _returnvalue))

(fn UpdateFade
  (params (On :type Boolean) (Immediate :type Boolean))
  (Networking|CanExecuteCosmeticEvents
    (:True
      (Variables|Default|SetAllowBoundsTransition On)
      (CallFunction|FadeBarrierBoundsVisibility (and On (or (not (Variables|Default|GetLocalPlayerIsWithinFxVolume)) (CallFunction|IsinPlayMode))) Immediate) ; Only Fade if player is outwith bounds \n(If minigame isn't running)
    )))

(fn UpdateNavigationProperties
  (bind _enabledcomponent (Variables|Default|GetEnabledComponent))
  (bind _enabled (Class|CreativeEnabledComponent|IsEnabled _enabledcomponent))
  (CallFunction|SetIsNavigationRelevant _enabled)
  (Navigation|SetAlwaysExportNavRelevantComponent _enabled)
  (Navigation|SetNavigationObstacleOverride (select (CallFunction|IsHollowVolumeShape (Variables|UserKnobs|GetKnob_VolumeShape)) "ExportAsPrimitive" "ForceEnabled")))

(fn ConstructionScript
  (|CallParentFunction)
  (Default|SetBarrierScale))
