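; Creative Barrier device that builds a configurable blocking volume players can be allowed through by team, class or ignore list.
; Options drive an async-loaded shape mesh, scale and material, with an older scaling path kept for pre-39.40 saves.
; Minigame, team and class delegates update per-player collision on server and local clients.
; Server overlaps teleport stuck players out or kill them, while local overlaps toggle bounds visibility.
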(event EventInitializeDevice
  ;[0 1 0.1154] Initialization
  (|CallParentFunction)
  (Default|UpdateBlockingChannels)
  (CallFunction|SetBarrierVisuals) ; async loads the appropriate Barrier mesh \nthen sets the scale
  (Visibility|UpdateBaseVisibility)
  (Default|UpdateBarrierState)
  (Delegates|BindDelegates true)
  (Class|CreativeEnabledComponent|Initialize (Variables|Default|GetEnabledComponent) (Variables|Default|GetFortMinigameProgress) (Variables|UserOptions|GetEnabledOnPhase))
  (CallFunction|UpdatePropMoverCollisionSize)
  (CallFunction|UpdateRelevancy)
  ;; SetupFrameVisibilityComponent
  (Initialization|InitializeWithSceneComponents (Variables|Default|GetFrameVisibility) (Utilities|Array|MakeArray (Variables|Default|GetCreativeVolumeFrame)) (Utilities|Map|MakeMap) "NewEnumerator4"))

(event OnAnyPropertyChanged(ToyOptionsComponent)
  ;[0 1 0.1154] Initialization
  (Default|UpdateBlockingChannels)
  (CallFunction|SetBarrierVisuals) ; async loads the appropriate Barrier mesh \nthen sets the scale
  (Visibility|UpdateBaseVisibility)
  (Default|UpdateBarrierState)
  (Delegates|BindDelegates true)
  (Class|CreativeEnabledComponent|Initialize (Variables|Default|GetEnabledComponent) (Variables|Default|GetFortMinigameProgress) (Variables|UserOptions|GetEnabledOnPhase))
  (CallFunction|UpdatePropMoverCollisionSize)
  (CallFunction|UpdateRelevancy)
  ;; SetupFrameVisibilityComponent
  (Initialization|InitializeWithSceneComponents (Variables|Default|GetFrameVisibility) (Utilities|Array|MakeArray (Variables|Default|GetCreativeVolumeFrame)) (Utilities|Map|MakeMap) "NewEnumerator4"))

(event EventBeginPlay
  ;[0 1 0.1154] Initialization
  (|CallParentFunction)
  (Networking|SwitchHasAuthority
    (:Authority
      (Collision|SetCollisionEnabled (Variables|Default|GetFxBarrierOverlapVolume) "NoCollision"))))

(event HandleVersionUpgrades(CreativeVersion)
  (params (LastSavedVersion :type Creative%20Version%20Id%20Structure))
  ;[0 1 0.1154] Initialization
  (CreativeVersion|CompareVersiontoTag LastSavedVersion "(TagName=\"Creative.Version.Release.39.40\")"
    (:OlderVersion
      (Variables|Default|SetUseOldScaling true))))

(event Custom|Set%20Barrier%20Visuals
  (bind _zoneshape (Variables|UserOptions|GetZoneShape))
  (bind _volume_shapes (Variables|Default|GetVolume%20Shapes))
  (bind (_value _returnvalue) (Utilities|Map|Find _volume_shapes _zoneshape))
  (bind _zoneshape_2 (Variables|UserOptions|GetZoneShape))
  (bind _scaled_volume_shapes (Variables|Default|GetScaled%20Volume%20Shapes))
  (bind (_value_2 _returnvalue_2) (Utilities|Map|Find _scaled_volume_shapes _zoneshape_2))
  ;[0 0.2775 2] Set Mesh
  ;; If we are in 39.40 or newer, we want to use the new Scaled Volume Shapes and the modified scaling function (Set Barrier Scale Updated). These fix a plethora of bugs related to collision when the shapes are scaled to large sizes.
  (if (Variables|Default|GetUseOldScaling)
    (if (and (Utilities|IsValidSoftObjectReference _value) _returnvalue)
      (Default|SetBarrierScale))
    (elif (and (Utilities|IsValidSoftObjectReference _value_2) _returnvalue_2)
      (Default|SetBarrierScaleUpdated)))
  (bind _loaded_object (Creative|Loading|AsyncLoadEditorSafe (select (not (Variables|Default|GetUseOldScaling)) _value_2 _value))
    (:Loaded
      (bind _asstatic_mesh (Utilities|Casting|CastToStaticMesh _loaded_object))
      (Components|StaticMesh|SetStaticMesh (Variables|Default|GetBarrierBounds) _asstatic_mesh)
      (Components|StaticMesh|SetStaticMesh (Variables|Default|GetBarrierBounds1) _asstatic_mesh)
      (CallFunction|SetBarrierMaterial))) ; AsyncLoadEditorSafe
)

(event ChildFunctions|EventPlayModeChanged
  (params (Minigame :type Fort%20Minigame%20Object%20Reference) (bIsInPlayMode :type Boolean))
  ;[0 1 0.8098 0.25] State Changes
  (Collision|SetCollisionResponsetoChannel (Variables|Default|GetFxBarrierOverlapVolume) "ECC_Pawn" (select bIsInPlayMode "ECR_Ignore" "ECR_Overlap"))
  (Default|UpdateDeviceVisuals)
  (Default|UpdateBarrierState))

(event Custom|Set%20Barrier%20Material
  (bind _barrierbounds (Variables|Default|GetBarrierBounds))
  ;[1 0.6163 0] Asynchronously Load Material for the Barrier Bounds, then update visuals
  (Networking|CanExecuteCosmeticEvents
    (:True
      (bind (_bhasshadow _material) (Utilities|ASyncGetBoundsMaterial)
        (:Visible%20Material
          (bind _output_get (Variables|Default|SetBoundsMID (Material|GetMIDfromMaterial _barrierbounds 0 _material)))
          (Rendering|Material|SetMaterial _barrierbounds 0 _output_get) ; Reattach MID (In case phone tool has \nstolen it)
          (Variables|Default|SetisOpaqueBoundsMaterial _bhasshadow))
        (:Invisible%20Material
          (Variables|Default|SetBoundsMID 0)
          (Variables|Default|SetisOpaqueBoundsMaterial false)))
      (Default|UpdateBoundsVisibility))))

(event OnEnabledStateChanged(EnabledComponent)
  (params (Enabled :type Boolean))
  ;[0 1 0.8098 0.25] State Changes
  (Default|UpdateBarrierState))

(event OnComponentBeginOverlap(FxBarrierOverlapVolume)
  (params (OverlappedComponent :type Primitive%20Component%20Object%20Reference) (OtherActor :type Actor%20Object%20Reference) (OtherComp :type Primitive%20Component%20Object%20Reference) (OtherBodyIndex :type Integer) (bFromSweep :type Boolean) (SweepResult :type Hit%20Result%20Structure%20%28by%20ref%29))
  ;[1 1 1] Overlap Effects (local client only)
  (Default|HideonOverlap false OtherActor))

(event Custom|EnabledOnReceived
  (params (Instigator :type Controller%20Object%20Reference))
  ;[0.6648 0 1] Transmitters and receivers
  (Class|CreativeEnabledComponent|SetEnabled (Variables|Default|GetEnabledComponent) true))

(event Custom|Event%20On%20Receive%20Add%20Player%20To%20Ignore%20List
  (params (Instigator :type Controller%20Object%20Reference))
  ;[1 0 0.4776] Collision Filtering - Players that meet the Ignored Team & Class conditions or in the Ignored List can pass through the barrier
  (IgnoredPlayerSystem|UpdateIgnoredList true Instigator))

(event Custom|Event%20On%20Receive%20Remove%20All%20Players%20From%20Ignore%20List
  (params (Instigator :type Controller%20Object%20Reference))
  ;[1 0 0.4776] Collision Filtering - Players that meet the Ignored Team & Class conditions or in the Ignored List can pass through the barrier
  (IgnoredPlayerSystem|ClearIgnoreList))

(event Custom|DisabledOnReceived
  (params (Instigator :type Controller%20Object%20Reference))
  ;[0.6648 0 1] Transmitters and receivers
  (Class|CreativeEnabledComponent|SetEnabled (Variables|Default|GetEnabledComponent) false)
  ;;[1 1 1] Detach Sticky Mines/C4
  (for _array_element (Actor|GetAttachedActors true)
    (StickyProjectile|DetachStickyProjectile(Message) _array_element)))

(event Custom|Hide%20Barrier%20Bounds%20Visibility
  (params (Barrier%20Visible :type Boolean) (Immediate :type Boolean))
  ;[1 0.7982 0] Hide Visual Bound Visibility
  ;;[1 1 1] FORT-925606: Made the fade get the current MID instead of the stored one, and removed the timelined fade to avoid multiple bugs regarding the Phone Tool and the barrier fighting for the materials.
  ;;; FORT-1003369: Adding missing boolean check from the FORT-925606 change, and also swapping "Set Visibility" to "Set Hidden in Game". Seems like visibility was messing with the MID and causing them to never be visible agian.
  (Networking|CanExecuteCosmeticEvents
    (:True
      (Development|SetHiddeninGame (Variables|Default|GetBarrierBounds) (not (and (Default|BarrierBoundsNotDisabled) Barrier%20Visible)) false))))

(event Custom|Event%20On%20Receive%20Remove%20Player%20From%20Ignore%20List
  (params (Instigator :type Controller%20Object%20Reference))
  ;[1 0 0.4776] Collision Filtering - Players that meet the Ignored Team & Class conditions or in the Ignored List can pass through the barrier
  (IgnoredPlayerSystem|UpdateIgnoredList false Instigator))

(event OnComponentEndOverlap(FxBarrierOverlapVolume)
  (params (OverlappedComponent :type Primitive%20Component%20Object%20Reference) (OtherActor :type Actor%20Object%20Reference) (OtherComp :type Primitive%20Component%20Object%20Reference) (OtherBodyIndex :type Integer))
  ;[1 1 1] Overlap Effects (local client only)
  (Default|HideonOverlap true OtherActor))

(event MinigameStartedDelegate(FortMinigameProgress)
  ;[1 0 0.4776] Collision Filtering - Players that meet the Ignored Team & Class conditions or in the Ignored List can pass through the barrier
  (IgnoredPlayerSystem|ClearIgnoredTeamandClass)
  (IgnoredPlayerSystem|SetupIgnoredTeamandClass)
  ;; SetupOverlapsOrAICollision
  (Networking|SwitchHasAuthority
    (:Authority
      (bind (_bareallpawnsignored _bareai_ignored) (CallFunction|CheckIgnoreState))
      (if _bareallpawnsignored
        (Variables|/Script/Engine.PrimitiveComponent|SetbGenerateOverlapEvents false (Variables|Default|GetBarrierBounds1))
        (elif _bareai_ignored
          (bind _returnvalue (AI|GetAIPopulationTracker))
          (AI|BindEventtoOnAIPawnSpawned _returnvalue (delegate Custom|AI%20Spawned)))))
    (:Remote
      (Variables|/Script/Engine.PrimitiveComponent|SetbGenerateOverlapEvents false (Variables|Default|GetBarrierBounds1))))
  (bind _returnvalue_2 (GameplayEventRouter|GetEventRouter "/Script/FortniteGame.FortPlayspace" self)) ; ContextActor is how we bind to a specific playspace
  (bind (_asynctaskproxy _payload) (GameplayEventRouter|StartListeningToEvent _returnvalue_2 "/Script/FortniteGame.FortTeamChangedContext")
    (:then
      (Variables|Default|SetTeamChangedObserver _asynctaskproxy))
    (:OnEventReceived
      (bind (_payload_player _payload_teamid) (Utilities|Struct|BreakFortTeamChangedContext _payload))
      (Delegates|OnTeamChanged _payload_player _payload_teamid 0 true) ; As of writing, OldTeamId isn't used
    )) ; Context here is irrelevant
)

(event MinigameEndedDelegate(FortMinigameProgress)
  ;[1 0 0.4776] Collision Filtering - Players that meet the Ignored Team & Class conditions or in the Ignored List can pass through the barrier
  (IgnoredPlayerSystem|ClearIgnoreList)
  (Utilities|FlowControl|Sequence
    (:then_0
      (IgnoredPlayerSystem|ClearIgnoredTeamandClass))
    (:then_1
      (GameplayEventRouter|StopListeningtoEvent (Variables|Default|GetTeamChangedObserver)))))

(event OnComponentBeginOverlap(BarrierBounds1)
  (params (OverlappedComponent :type Primitive%20Component%20Object%20Reference) (OtherActor :type Actor%20Object%20Reference) (OtherComp :type Primitive%20Component%20Object%20Reference) (OtherBodyIndex :type Integer) (bFromSweep :type Boolean) (SweepResult :type Hit%20Result%20Structure%20%28by%20ref%29))
  (bind (_origin _boxextent) (Collision|GetComponentBounds (Variables|Default|GetBarrierBounds1)))
  (bind (_bblockinghit _binitialoverlap _time _distance _location _impactpoint) (Collision|BreakHitResult SweepResult))
  ;[1 0.4422 0.03086] Teleport any players stuck inside barriers when spawning into them - Does not apply to hollow boxes
  (Networking|SwitchHasAuthority
    (:Authority
      (if (PlayerStuck|ShouldTeleportOverlappingPawns)
        (bind _asfort_player_pawn (Utilities|Casting|CastToFortPlayerPawn OtherActor)
          (:then
            (Utilities|IsValid (Vehicle|GetVehicleActor _asfort_player_pawn :!class /Script/FortniteGame.FortPlayerPawn)
              (:Is%20Not%20Valid
                (if (Math|Geometry|IsPointinBox (Transformation|GetActorLocation _asfort_player_pawn) _origin _boxextent)
                  (bind _bsucceed (PlayerStuck|TeleportPlayerToNearestPoint _asfort_player_pawn (select (Math|Vector|VectorIsNearlyZero _impactpoint) (Transformation|GetActorLocation _asfort_player_pawn) _impactpoint)))
                  (if _bsucceed
                    _
                    (elif (Pawn|IsBotControlled _asfort_player_pawn)
                      (Utilities|FlowControl|DelayUntilNextTick)
                      (FortPawn|ForceKill _asfort_player_pawn "(TagName=\"Gameplay.Damage.OutsideSafeZone\")" 0 _asfort_player_pawn)
                      (else
                        (FortPawn|ForceKill _asfort_player_pawn (Variables|Default|GetDeath%20Reason) 0 0))))))) ; If the player is NOT in a vehicle, then attempt to teleport them.\nWe do this because the pawn's overlap area may extend beyond the vehicle\ncausing the player to overlap the barrier even though their vehicle does NOT.\n\nIdeally we would detect vehicle overlaps and eject players in those vehicles\nbut,  currently some vehicles have overlaps bigger than their blocking collision and thus can end up overlapping a barrier they are not actually in.
          )
          (:CastFailed
            ;[1 0.7182 0.5835 0.77] Check for a relevant tag, and if possessed, kill the pawn (as long as it is a pawn)
            (bind _returnvalue (GameplayTags|HasAnyMatchingGameplayTags(Message) OtherActor (GameplayTags|MakeGameplayTagContainerfromArray (Variables|Default|GetEnemyPawnTags))))
            (if _returnvalue
              (bind _asfort_pawn (Utilities|Casting|CastToFortPawn OtherActor))
              (FortPawn|ForceKill _asfort_pawn (Variables|Default|GetDeath%20Reason) 0 0))))))) ; Overlaps should not be happening on the remote (It's disabled in SetupOverlaps)\nBut for safety we are keeping this!
)

(event OnRemovedfromMinigame(FortMinigameProgress)
  (params (Minigame :type Fort%20Minigame%20Object%20Reference))
  ;[1 0 0.4776] Collision Filtering - Players that meet the Ignored Team & Class conditions or in the Ignored List can pass through the barrier
  (Delegates|BindDelegates false)
  (IgnoredPlayerSystem|ClearIgnoredTeamandClass))

(event OnAddedtoMinigame(FortMinigameProgress)
  (params (Minigame :type Fort%20Minigame%20Object%20Reference))
  ;[1 0 0.4776] Collision Filtering - Players that meet the Ignored Team & Class conditions or in the Ignored List can pass through the barrier
  (Delegates|BindDelegates true))

(event Custom|Player%20Joined%20In%20Progress
  (params (Pawn :type Pawn%20Object%20Reference))
  ;[1 0 0.4776] Collision Filtering - Players that meet the Ignored Team & Class conditions or in the Ignored List can pass through the barrier
  (CallFunction|JoininProgressLocal Pawn))

(event Custom|Join%20In%20Progress%20Local
  (params (Pawn :type Pawn%20Object%20Reference))
  ;[1 0 0.4776] Collision Filtering - Players that meet the Ignored Team & Class conditions or in the Ignored List can pass through the barrier
  (if (Pawn|IsLocallyControlled Pawn)
    (Delegates|BindLocalDelegates true)
    (CallFunction|UpdateLocalIgnore (Pawn|GetController Pawn) (Variables|/Script/FortniteGame.FortMinigameLogicComponent|GetCurrentMinigame (Variables|Default|GetFortMinigameProgress)))))

(fn BindDelegates
  (params (Bind :type Boolean))
  (bind _outputdelegate (EventDispatchers|CreateEvent :!event OnMinigamePlayerSpawned))
  (bind _outputdelegate_2 (EventDispatchers|CreateEvent :!event OnClassChanged))
  (bind _outputdelegate_3 (EventDispatchers|CreateEvent :!event Player%20Joined%20In%20Progress))
  (bind _outputdelegate_4 (EventDispatchers|CreateEvent :!event OnTeamChanged))
  (bind _currentminigame (LocalVariables|GetCurrentMinigame))
  (bind _currentminigame_2 (LocalVariables|GetCurrentMinigame))
  (bind _currentminigame_3 (LocalVariables|GetCurrentMinigame))
  (Creative|Networking|Authority/Cosmetic
    (:Authority
      (bind _currentminigame_4 (Variables|/Script/FortniteGame.FortMinigameLogicComponent|GetCurrentMinigame (Variables|Default|GetFortMinigameProgress))
        (:then
          (LocalVariables|SetCurrentMinigame _currentminigame_4)
          (Utilities|FlowControl|Sequence
            (:then_0
              (if (and (Variables|UserOptions|GetIgnoreTeam_Override) (Variables|GetBind))
                (Minigame|BindEventtoPlayerChangedTeamDelegate _currentminigame _outputdelegate_4)
                (else
                  (Minigame|UnbindEventfromPlayerChangedTeamDelegate _currentminigame _outputdelegate_4))))
            (:then_1
              (if (and (Variables|UserOptions|GetIgnoreClass_Override) (Variables|GetBind))
                (Minigame|BindEventtoOnPlayerClassApplied _currentminigame_2 _outputdelegate_2)
                (else
                  (Minigame|UnbindEventfromOnPlayerClassApplied _currentminigame_2 _outputdelegate_2))))
            (:then_2
              (if (Variables|GetBind)
                (Minigame|BindEventtoOnMinigamePlayerSpawned _currentminigame_3 _outputdelegate)
                (Minigame|BindEventtoOnPlayerJoininProgress _currentminigame_3 _outputdelegate_3)
                (else
                  (Minigame|UnbindEventfromOnMinigamePlayerSpawned _currentminigame_3 _outputdelegate)
                  (Minigame|UnbindEventfromOnPlayerJoininProgress _currentminigame_3 _outputdelegate_3))))))
        (:else)))
    (:Cosmetic
      (Delegates|BindLocalDelegates Bind))))

(fn BindLocalDelegates
  (params (Bind :type Boolean))
  (bind _outputdelegate (EventDispatchers|CreateEvent :!event OnLocalPlayerClassChanged))
  (bind _outputdelegate_2 (EventDispatchers|CreateEvent :!event OnLocalPlayerTeamChanged))
  (bind _outputdelegate_3 (EventDispatchers|CreateEvent :!event OnLocalFortPawnChanged))
  (bind _localplayercontroller (LocalVariables|GetLocalPlayerController))
  (Networking|CanExecuteCosmeticEvents
    (:True
      (bind _array_element (Default|ForEachFortLocalPlayerController)
        (:Loop%20Body
          (Utilities|IsValid _array_element
            (:Is%20Valid
              (LocalVariables|SetLocalPlayerController _array_element)
              (Utilities|FlowControl|Sequence
                (:then_0
                  (bind _asfort_player_controller_athena (Utilities|Casting|CastToFortPlayerControllerAthena (LocalVariables|GetLocalPlayerController)))
                  (if (and (Variables|GetBind) (Variables|UserOptions|GetIgnoreTeam_Override))
                    (FortPlayerControllerAthena|BindEventtoOnLocalPlayerChangedTeams _asfort_player_controller_athena _outputdelegate_2)
                    (else
                      (FortPlayerControllerAthena|UnbindEventfromOnLocalPlayerChangedTeams _asfort_player_controller_athena _outputdelegate_2))))
                (:then_1
                  (bind _currentminigame (Variables|/Script/FortniteGame.FortMinigameLogicComponent|GetCurrentMinigame (Variables|Default|GetFortMinigameProgress))
                    (:then
                      (if (and (Variables|GetBind) (Variables|UserOptions|GetIgnoreClass_Override))
                        (Minigame|BindEventtoOnAnyLocalPlayerChangedClass _currentminigame _outputdelegate)
                        (else
                          (Minigame|UnbindEventfromOnAnyLocalPlayerChangedClass _currentminigame _outputdelegate))))
                    (:else)))
                (:then_2
                  (if (Variables|GetBind)
                    (EventDispatchers|BindEventtoOnFortPawnChangedEvent _localplayercontroller _outputdelegate_3)
                    (else
                      (EventDispatchers|UnbindEventfromOnFortPawnChangedEvent _localplayercontroller _outputdelegate_3))))))))))))

(fn CheckIgnoreState
  (bind _ignoreclass_classtype (Utilities|Struct|BreakCreativeClassOption (Variables|GetIgnoreClass_ClassType)))
  (bind _ignoreclass_classtype_2 (Utilities|Struct|BreakCreativeClassOption (Variables|GetIgnoreClass_ClassType)))
  (Utilities|FlowControl|Sequence
    (:then_0
      (if (Variables|UserOptions|GetIgnoreTeam_Override)
        (if (Variables|UserOptions|GetIgnoreClass_Override)
          (if (Variables|UserOptions|GetMatchTeamAndClass)
            (if (Creative|TeamOption|IsTeamType (Variables|UserOptions|GetIgnoreTeam) "Any")
              (switch Utilities|FlowControl|Switch|SwitchonECreativeClassType _ignoreclass_classtype :!label _label
                (:NoClass
                  (return false true))
                (:All
                  _)
                (:Any
                  _)))
            (elif (Creative|TeamOption|IsTeamType (Variables|UserOptions|GetIgnoreTeam) "Any")
              _
              (else
                (goto _label))))
          (return true true)
          (elif (Creative|TeamOption|IsTeamType (Variables|UserOptions|GetIgnoreTeam) "Any")
            (return true true)))
        (elif (Variables|UserOptions|GetIgnoreClass_Override)
          (switch Utilities|FlowControl|Switch|SwitchonECreativeClassType _ignoreclass_classtype_2
            (:NoClass
              (return false true))
            (:All
              _)
            (:Any
              _))
          (return true true))))
    (:then_1
      (return false false))))

(fn CheckNPCAddedToIgnoreList
  (params (Add :type Boolean) (Controller :type Controller%20Object%20Reference))
  (if (Variables|UserOptions|GetCan%20NPC%20Added%20To%20Ignore%20List)
    ;[1 1 1] As long as an AI character is tried to be added to the list, force unblock all AIs since the same navigation data and update Barrier collision
    (bind _asaicontroller (Utilities|Casting|CastToAIController (Variables|GetController))
      (:then
        (CallFunction|SetIsNavigationRelevant false)
        (Navigation|SetAlwaysExportNavRelevantComponent false)
        (IgnoredPlayerSystem|UpdateCollisionforNPC _asaicontroller (Variables|GetAdd))
        (if (Variables|GetAdd)
          (Utilities|Array|AddUnique (Variables|Transient|GetIgnoredNPCList) _asaicontroller)
          (else
            (Utilities|Array|RemoveItem (Variables|Transient|GetIgnoredNPCList) _asaicontroller))))
      (:CastFailed
        _)))
  (return))

(fn ClearIgnoredTeamAndClass
  (bind _array (Utilities|Array|MakeArray))
  (Variables|Transient|SetIgnoredPlayersByTeam _array)
  (Variables|Transient|SetIgnoredPlayersByClass _array)
  (IgnoredPlayerSystem|UpdateAllPlayersCollision))

(fn ClearIgnoreList
  (Variables|Transient|SetIgnoredPlayersList (Utilities|Array|MakeArray))
  (IgnoredPlayerSystem|UpdateAllPlayersCollision)
  (for _array_element (Variables|Transient|GetIgnoredNPCList)
    (IgnoredPlayerSystem|UpdateCollisionforNPC _array_element))
  (Variables|Transient|SetIgnoredNPCList (Utilities|Array|MakeArray)))

(fn GetEnabledStateasIndex
  (return (DeviceConversion|EnabledPhaseEnumtoLegacyInt (Variables|UserOptions|GetEnabledOnPhase))))

(fn GetClosestDirection
  (params (Pawn :type Actor%20Object%20Reference) (PossibleDirections :type Array%20of%20Vectors))
  (bind _barrierbounds1 (Variables|Default|GetBarrierBounds1))
  (bind _origin (Collision|GetComponentBounds _barrierbounds1))
  (bind _pawn (Variables|GetPawn))
  (bind _returnvalue (Transformation|GetActorLocation _pawn))
  (bind _returnvalue_2 (- _returnvalue _origin))
  (for (_array_element _array_index) PossibleDirections
    (if (or (== (LocalVariables|GetMinIndex) -1) (< (Math|Vector|Distance(Vector) _array_element _returnvalue_2) (LocalVariables|GetMinDistance)))
      (LocalVariables|SetMinIndex _array_index)
      (LocalVariables|SetMinDistance (Math|Vector|Distance(Vector) _array_element _returnvalue_2))))
  (return (Utilities|Array|Get(acopy) (Variables|GetPossibleDirections) (LocalVariables|GetMinIndex)) (LocalVariables|GetMinIndex)))

(fn GetCreativeActorBounds
  (return false "0, 0, 0"))

(fn GetCreativeActorOrigin
  (return true (Variables|/Script/Engine.SceneComponent|GetRelativeLocation (Variables|Default|GetFxBarrierOverlapVolume))))

(fn GetCreativeActorRelativeTransform
  (return false 0))

(fn GetGridAxisGuideAttachmentPoint
  (return false false "0, 0, 0"))

(fn GetOverrideMeshMaterial
  (params (MeshComp :type Static%20Mesh%20Component%20Object%20Reference) (MatIdx :type Integer))
  (if (Utilities|Equal(Object) MeshComp (Variables|Default|GetBarrierBounds))
    (return (Variables|Default|GetBoundsMID))
    (else
      (return (|CallParentFunction MeshComp MatIdx)))))

(fn GetVolumeScale
  (return (Math|Vector|MakeVector (Variables|UserOptions|Basic|GetBarrierWidth) (Variables|UserOptions|Basic|GetBarrierDepth) (Variables|UserOptions|Basic|GetBarrierHeight))))

(fn HideOnOverlap
  (params (Visible :type Boolean) (Actor :type Actor%20Object%20Reference))
  (Variables|Default|SetIs%20Bound%20Visible (or Visible (Utilities|Enum|Equal(Enum) (Variables|UserOptions|GetZoneShape) "NewEnumerator2")))
  (Networking|CanExecuteCosmeticEvents
    (:True
      (bind _aspawn (Utilities|Casting|CastToPawn Actor))
      (if (Pawn|IsLocallyControlled _aspawn)
        (Variables|Default|SetLocalPlayerIsWithinFxVolume (not (Variables|Default|GetIs%20Bound%20Visible)))
        (if (and (or (Utilities|IsValidSoftObjectReference (Variables|UserOptions|Basic|GetBarrierMaterial)) (Variables|UserOptions%28Developer%29|GetEnableDeveloperMaterials)) (Variables|Default|GetAllowBoundsTransition))
          (CallFunction|HideBarrierBoundsVisibility (Variables|Default|GetIs%20Bound%20Visible)))))))

(fn IsVisibleInGame
  (return (Visibility|ShouldShowinGame)))

(fn IsHollowVolumeShape
  (params (VolumeShape :type EGameplayVolumeShape%20Enum :default "NewEnumerator0"))
  (return (Utilities|Array|ContainsItem (Variables|Default|GetHollowVolumeShapes) VolumeShape)))

(fn IsIgnoredClass
  (params (ClassSlotIndex :type Byte))
  (return (or (Creative|ClassOption|IsClassType (Variables|UserOptions|GetIgnoreClass) "Any") (Math|Byte|Equal(Byte) (Creative|ClassOption|ConverttoClassSlotIndex (Variables|UserOptions|GetIgnoreClass)) ClassSlotIndex))))

(fn IsIgnoredPlayer
  (params (PlayerState :type Player%20State%20Object%20Reference))
  (return (or (IgnoredPlayerSystem|IsPlayerIgnoredTeamandClass PlayerState) (Utilities|Array|ContainsItem (Variables|Transient|GetIgnoredPlayersList) PlayerState))))

(fn IsIgnoredTeam
  (params (TeamId :type Byte))
  (return (or (Creative|TeamOption|IsTeamType (Variables|UserOptions|GetIgnoreTeam) "Any") (Math|Byte|Equal(Byte) (Creative|TeamOption|ConverttoTeam (Variables|UserOptions|GetIgnoreTeam)) TeamId))))

(fn IsPlayerIgnoredTeamAndClass
  (params (PlayerState :type Player%20State%20Object%20Reference))
  (bind _fortplayerstate (LocalVariables|GetFortPlayerState))
  (bind _returnvalue (Team|GetActorTeam _fortplayerstate))
  (bind _isignored (IgnoredPlayerSystem|IsIgnoredTeam _returnvalue))
  (bind _fortplayerstate_2 (LocalVariables|GetFortPlayerState))
  (bind _ignoreteam_override (Variables|UserOptions|GetIgnoreTeam_Override))
  (bind _ignoreclass_override (Variables|UserOptions|GetIgnoreClass_Override))
  (Utilities|FlowControl|Sequence
    (:then_0
      (bind _asfort_player_state (Utilities|Casting|CastToFortPlayerState PlayerState))
      (LocalVariables|SetFortPlayerState _asfort_player_state)
      (if (and (== (Variables|UserOptions|GetIgnoreTeam_Override) false) (== (Variables|UserOptions|GetIgnoreClass_Override) false))
        (return false)
        (else
          (bind _fortminigameprogress (Variables|Default|GetFortMinigameProgress)
            (:then
              (bind _currentminigame (Variables|/Script/FortniteGame.FortMinigameLogicComponent|GetCurrentMinigame _fortminigameprogress)
                (:then
                  (Utilities|IsValid (Minigame|GetFNEClassComponent _currentminigame)
                    (:Is%20Valid
                      (if (Variables|UserOptions|GetMatchTeamAndClass)
                        (return (and (or _isignored (not _ignoreteam_override)) (or (not _ignoreclass_override) (IgnoredPlayerSystem|IsIgnoredClass (Minigame|GetClassSlotforPlayer (Minigame|GetFNEClassComponent _currentminigame) _fortplayerstate_2 false)))))
                        (else
                          (return (or (and _isignored _ignoreteam_override) (and _ignoreclass_override (IgnoredPlayerSystem|IsIgnoredClass (Minigame|GetClassSlotforPlayer (Minigame|GetFNEClassComponent _currentminigame) _fortplayerstate_2 false))))))))))
                (:else)))
            (:else)))))
    (:then_1
      (return false))))

(fn IsTargetableByLinearMover
  (return true))

(fn MigrateLegacyUserOptions
  (bind _zonevisibiltyindex (Class|GetClassDefaults "/CRD_VolumetricRegion/Device_Barrier_V2.Device_Barrier_V2_C"))
  (bind (_zonezscale _zonexscale _zoneyscale) (Class|GetClassDefaults "/CRD_VolumetricRegion/Device_Barrier_V2.Device_Barrier_V2_C"))
  (bind _zonevisibiltyindex_2 (Variables|Legacy|GetZoneVisibiltyIndex))
  (|CallParentFunction)
  (Utilities|FlowControl|Sequence
    (:then_0
      (Utilities|FlowControl|Sequence
        (:then_0
          (bind _migratedproperties (DeviceConversion|MigrateLegacyEnabledIndex (Variables|UserOptions|GetEnabledOnPhase) (Class|GetClassDefaults "/CRD_VolumetricRegion/Device_Barrier_V2.Device_Barrier_V2_C")))
          (LocalVariables|SetOptionsMigrated _migratedproperties))
        (:then_1
          (if (!= _zonevisibiltyindex_2 _zonevisibiltyindex)
            (Variables|UserOptions|Basic|SetBarrierMaterial (Utilities|Array|Get(acopy) (Variables|Legacy|GetBarrier%20Materials) _zonevisibiltyindex_2))
            (Variables|Legacy|SetZoneVisibiltyIndex _zonevisibiltyindex)
            (LocalVariables|SetOptionsMigrated true)))
        (:then_2
          (Utilities|FlowControl|Sequence
            (:then_0
              (if (!= _zonezscale (Variables|Legacy|GetZoneZScale))
                (Variables|UserOptions|Basic|SetBarrierHeight (Variables|Legacy|GetZoneZScale))
                (Variables|Legacy|SetZoneZScale (Class|GetClassDefaults "/CRD_VolumetricRegion/Device_Barrier_V2.Device_Barrier_V2_C"))
                (LocalVariables|SetOptionsMigrated true)))
            (:then_1
              (if (!= _zonexscale (Variables|Legacy|GetZoneXScale))
                (Variables|UserOptions|Basic|SetBarrierWidth (Variables|Legacy|GetZoneXScale))
                (Variables|Legacy|SetZoneXScale (Class|GetClassDefaults "/CRD_VolumetricRegion/Device_Barrier_V2.Device_Barrier_V2_C"))
                (LocalVariables|SetOptionsMigrated true)))
            (:then_2
              (if (!= _zoneyscale (Variables|Legacy|GetZoneYScale))
                (Variables|UserOptions|Basic|SetBarrierDepth (Variables|Legacy|GetZoneYScale))
                (Variables|Legacy|SetZoneYScale (Class|GetClassDefaults "/CRD_VolumetricRegion/Device_Barrier_V2.Device_Barrier_V2_C"))
                (LocalVariables|SetOptionsMigrated true)))))))
    (:then_1
      (if (LocalVariables|GetOptionsMigrated)
        (OptionData|RefreshPropertyOverrides (Variables|Default|GetToyOptionsComponent))))))

(fn OnClassChanged
  (params (PlayerState :type Fort%20Player%20State%20Object%20Reference) (ClassSlotIndex :type Byte))
  (IgnoredPlayerSystem|UpdatePlayerStateArray (Variables|Transient|GetIgnoredPlayersByClass) (Variables|GetPlayerState) (IgnoredPlayerSystem|IsIgnoredClass ClassSlotIndex))
  (IgnoredPlayerSystem|UpdateCollisionforPlayer (Variables|GetPlayerState) (IgnoredPlayerSystem|IsIgnoredPlayer (Variables|GetPlayerState))))

(fn OnLocalFortPawnChanged
  (params (NewPawn :type Fort%20Pawn%20Object%20Reference))
  (Utilities|IsValid NewPawn
    (:Is%20Valid
      (Delegates|OnMinigamePlayerSpawned (Pawn|GetController NewPawn)))))

(fn OnLocalPlayerClassChanged
  (bind _fortminigameprogress (Variables|Default|GetFortMinigameProgress)
    (:then
      (bind _currentminigame (Variables|/Script/FortniteGame.FortMinigameLogicComponent|GetCurrentMinigame _fortminigameprogress)
        (:then
          (Utilities|IsValid (Minigame|GetFNEClassComponent _currentminigame)
            (:Is%20Valid
              (LocalVariables|SetLocalMinigameFNEClassComponent (Minigame|GetFNEClassComponent _currentminigame))
              (bind _array_element (Default|ForEachFortLocalPlayerController)
                (:Loop%20Body
                  (Utilities|IsValid _array_element
                    (:Is%20Valid
                      (bind _asfort_player_state (Utilities|Casting|CastToFortPlayerState (Variables|/Script/Engine.Controller|GetPlayerState _array_element)))
                      (Delegates|OnClassChanged _asfort_player_state (Minigame|GetClassSlotforPlayer (LocalVariables|GetLocalMinigameFNEClassComponent) _asfort_player_state false)))))))))
        (:else)))
    (:else)))

(fn OnLocalPlayerTeamChanged
  (bind _array_element (Default|ForEachFortLocalPlayerController)
    (:Loop%20Body
      (Utilities|IsValid _array_element
        (:Is%20Valid
          (bind _asfort_player_state (Utilities|Casting|CastToFortPlayerState (Variables|/Script/Engine.Controller|GetPlayerState _array_element)))
          (Delegates|OnTeamChanged _asfort_player_state (Team|GetActorTeam (Variables|/Script/Engine.Controller|GetPlayerState _array_element))))))))

(fn OnMinigamePlayerSpawned
  (params (PlayerController :type Controller%20Object%20Reference))
  (Utilities|IsValid PlayerController
    (:Is%20Valid
      (IgnoredPlayerSystem|UpdateCollisionforPlayer (Variables|/Script/Engine.Controller|GetPlayerState PlayerController) (IgnoredPlayerSystem|IsIgnoredPlayer (Variables|/Script/Engine.Controller|GetPlayerState PlayerController))))))

(fn OnPreviewStart
  (return false))

(fn OnRep_IgnoredPlayersList
  (Networking|CanExecuteCosmeticEvents
    (:True
      (IgnoredPlayerSystem|UpdateAllPlayersCollision))) ; Usually the collision was just handled on the server
)

(fn OnTeamChanged
  (params (PlayerState :type Fort%20Player%20State%20Object%20Reference) (TeamId :type Byte) (OldTeamId :type Byte) (bTeamChangedWithoutRespawn :type Boolean))
  (IgnoredPlayerSystem|UpdatePlayerStateArray (Variables|Transient|GetIgnoredPlayersByTeam) (Variables|GetPlayerState) (IgnoredPlayerSystem|IsIgnoredTeam TeamId))
  (IgnoredPlayerSystem|UpdateCollisionforPlayer (Variables|GetPlayerState) (IgnoredPlayerSystem|IsIgnoredPlayer (Variables|GetPlayerState))))

(fn ResizeFrame
  (Class|CreativeVolumeFrame|InitializeComponent (Variables|Default|GetCreativeVolumeFrame) (Variables|UserOptions|GetZoneShape) (Math|Float|SelectFloat 0.99 1.0 (Utilities|Enum|Equal(Enum) (Variables|UserOptions|GetZoneShape) "NewEnumerator2")) (Utilities|Operators|Multiply (CallFunction|GetVolumeScale) (5.12 5.12 3.84))))

(fn SetBarrierScale
  (bind _zoneshape (Variables|UserOptions|GetZoneShape))
  (bind _volume_transform (Variables|ChildFunctions|GetVolume%20Transform))
  (bind (_value _returnvalue) (Utilities|Map|Find _volume_transform _zoneshape))
  (bind _value_location (Math|Transform|BreakTransform _value))
  (bind _scale (CallFunction|GetVolumeScale))
  (bind (_value_2 _returnvalue_2) (Utilities|Map|Find (Variables|ChildFunctions|GetVolume%20Transform) (Variables|UserOptions|GetZoneShape)))
  (bind (_value_location_2 _value_rotation _value_scale) (Math|Transform|BreakTransform _value_2))
  (bind _fxbarrieroverlapvolume (Variables|Default|GetFxBarrierOverlapVolume))
  (Utilities|FlowControl|Sequence
    (:then_0
      (LocalVariables|Setlocal%20scale (* (CallFunction|GetVolumeScale) (Math|Vector|SelectVector _value_scale "1.000000,1.000000,1.000000" _returnvalue_2)))
      ;[1 1 1] Shrinking allows building on the sides and on top of the Barrier
      (bind _output_get (LocalVariables|SetShrunkenBarrierScale (Math|Vector|SelectVector (- (LocalVariables|Getlocal%20scale) (Utilities|Map|Find (Variables|ChildFunctions|GetBarrier%20Shrink) (Variables|UserOptions|GetZoneShape))) (LocalVariables|Getlocal%20scale) (Variables|UserOptions|GetShrink%20to%20allow%20building))))
      (Transformation|SetRelativeScale3D (Variables|Default|GetBarrierBounds) _output_get)
      (CallFunction|ResizeFrame)
      (Transformation|SetRelativeScale3D _fxbarrieroverlapvolume _scale)
      (Transformation|SetRelativeLocation _fxbarrieroverlapvolume (Math|Vector|MakeVector 0.0 0.0 (* (.z _scale) 192.0)) false false))
    (:then_1
      (if _returnvalue
        (Transformation|SetRelativeLocation (Variables|Default|GetBarrierOffset) (select (Utilities|Enum|Equal(Enum) (Variables|UserOptions|GetZoneShape) "NewEnumerator2") (Math|Vector|MakeVector (.x (* (LocalVariables|GetShrunkenBarrierScale) _value_location)) (.y (* (LocalVariables|GetShrunkenBarrierScale) _value_location)) (+ (.z (* (LocalVariables|GetShrunkenBarrierScale) _value_location)) (* (LocalVariables|GetZOffsetPerHeightUnit) (Variables|UserOptions|Basic|GetBarrierHeight)))) (* (LocalVariables|GetShrunkenBarrierScale) _value_location)) false false)))
    (:then_2
      (Culling|UpdateCullDistance false))))

(fn SetBarrierScaleUpdated
  (bind _zoneshape (Variables|UserOptions|GetZoneShape))
  (bind _volume_transform_scaled (Variables|ChildFunctions|GetVolume%20Transform%20Scaled))
  (bind (_value _returnvalue) (Utilities|Map|Find _volume_transform_scaled _zoneshape))
  (bind _value_location (Math|Transform|BreakTransform _value))
  (bind _scale (CallFunction|GetVolumeScale))
  (bind _value_2 (Utilities|Map|Find (Variables|ChildFunctions|GetVolume%20Transform%20Scaled) (Variables|UserOptions|GetZoneShape)))
  (bind (_value_location_2 _value_rotation _value_scale) (Math|Transform|BreakTransform _value_2))
  (bind _fxbarrieroverlapvolume (Variables|Default|GetFxBarrierOverlapVolume))
  ; With the new Scaled shapes, we no longer need to scale these up to grid-size EXCEPT for the Cylinder which uses the primitive Cylinder.
  (Utilities|FlowControl|Sequence
    (:then_0
      (LocalVariables|Setlocal%20scale (* (CallFunction|GetVolumeScale) (Math|Vector|SelectVector _value_scale "1.000000,1.000000,1.000000" (Utilities|Enum|Equal(Enum) (Variables|UserOptions|GetZoneShape) "NewEnumerator4"))))
      (bind _output_get (LocalVariables|SetShrunkenBarrierScale (Math|Vector|SelectVector (* (LocalVariables|Getlocal%20scale) (LocalVariables|GetShrinkPercent)) (LocalVariables|Getlocal%20scale) (Variables|UserOptions|GetShrink%20to%20allow%20building))))
      ; Apply the new Scaling and resize all of the other parts of the barrier to fit the new scale.
      (Transformation|SetRelativeScale3D (Variables|Default|GetBarrierBounds) _output_get)
      (CallFunction|ResizeFrame)
      (Transformation|SetRelativeScale3D _fxbarrieroverlapvolume _scale)
      (Transformation|SetRelativeLocation _fxbarrieroverlapvolume (Math|Vector|MakeVector 0.0 0.0 (* (.z _scale) 192.0)) false false))
    (:then_1
      ; After we have scaled the actual barrier bounds, we want to adjust its precise position to make sure it lines up with the outer bounds. If it's anything other the hollow barrier, we grab the Location for the Volume Transforms and multiple it by the shrunken scale to get the right offset. If it's a Hollow Box, we have to move the  barrier ddownwards to ensure it matches properly. So we have a negative Z-Offset value that we use with the Z Scale to determine how many units it has to be moved down
      (if _returnvalue
        (Transformation|SetRelativeLocation (Variables|Default|GetBarrierOffset) (Math|Vector|SelectVector (* _value_location (LocalVariables|GetShrunkenBarrierScale)) (Math|Vector|MakeVector 0.0 0.0 (* (.z (LocalVariables|GetShrunkenBarrierScale)) (LocalVariables|GetZOffsetPerHeightUnit))) (Utilities|Enum|NotEqual(Enum) (Variables|UserOptions|GetZoneShape) "NewEnumerator2")) false false)))
    (:then_2
      (Culling|UpdateCullDistance false))))

(fn SetupIgnoredTeamAndClass
  (bind _ignoreteam (Variables|UserOptions|GetIgnoreTeam))
  (bind _ignoreclass (Variables|UserOptions|GetIgnoreClass))
  (bind _fortminigameprogress (Variables|Default|GetFortMinigameProgress)
    (:then
      (bind _currentminigame (Variables|/Script/FortniteGame.FortMinigameLogicComponent|GetCurrentMinigame _fortminigameprogress)
        (:then
          (LocalVariables|SetCurrentMinigame _currentminigame)
          (Utilities|IsValid (Minigame|GetMinigameTeamComponent _currentminigame)
            (:Is%20Valid
              (LocalVariables|SetLocalMinigameTeamComponent (Minigame|GetMinigameTeamComponent _currentminigame))
              (Utilities|IsValid (Minigame|GetFNEClassComponent _currentminigame)
                (:Is%20Valid
                  (LocalVariables|SetLocalMinigameFNEClassComponent (Minigame|GetFNEClassComponent _currentminigame))
                  (Networking|SwitchHasAuthority
                    (:Authority
                      (bind _bareallpawnsignored (CallFunction|CheckIgnoreState))
                      (if _bareallpawnsignored
                        (Collision|SetCollisionResponsetoChannel (Variables|Default|GetBarrierBounds1) "ECC_Pawn" "ECR_Ignore")
                        (Networking|CanExecuteCosmeticEvents
                          (:True
                            (Rendering|SetActorHiddenInGame (Variables|UserOptions|GetInvisibleToIgnoredPlayers)) ; Updating the barrier visibility based on whether\nwe want ignored players to see it.\nThis uses ActorHiddenInGame to avoid\ninterfering with the barrier visibility set by the \nenabled component, which manually sets \nBarrier Bounds vis.
                            (bind _array_element (Default|ForEachFortLocalPlayerController)
                              (:Loop%20Body
                                (CallFunction|UpdateCollisionforLocalPlayer _array_element true)))))
                        (else
                          (Utilities|FlowControl|Sequence
                            (:then_0
                              ;[0 0.8066 1] Setup the Ignored Team & Class Players. Only needs to be done by Server. If the type is set to Any, we just ignore pawn collision as we need to include AI as well. (AI is not supported for specific teams since we can't tell the nav mesh to ignore this device for only certain AI)
                              (if (Variables|UserOptions|GetIgnoreTeam_Override)
                                (if (Creative|TeamOption|IsTeamType _ignoreteam "Any")
                                  (Variables|Transient|SetIgnoredPlayersByTeam (Minigame|GetParticipatingPlayers (LocalVariables|GetCurrentMinigame)))
                                  (else
                                    (Variables|Transient|SetIgnoredPlayersByTeam (Class|FortMinigameTeamComponent|GetTeamMembers (LocalVariables|GetLocalMinigameTeamComponent) (Creative|TeamOption|ConverttoTeam _ignoreteam)))))
                                (for _array_element_2 (Variables|Transient|GetIgnoredPlayersByTeam)
                                  (IgnoredPlayerSystem|UpdateCollisionforPlayer _array_element_2 (IgnoredPlayerSystem|IsPlayerIgnoredTeamandClass _array_element_2)))))
                            (:then_1
                              (if (Variables|UserOptions|GetIgnoreClass_Override)
                                (if (Creative|ClassOption|IsClassType _ignoreclass "Any")
                                  (Variables|Transient|SetIgnoredPlayersByClass (Minigame|GetParticipatingPlayers (LocalVariables|GetCurrentMinigame)))
                                  (else
                                    (Variables|Transient|SetIgnoredPlayersByClass (Class|FNEClassComponent|GetPlayerswithClassSlot (LocalVariables|GetLocalMinigameFNEClassComponent) (Creative|ClassOption|ConverttoClassSlotIndex _ignoreclass)))))
                                (for _array_element_3 (Variables|Transient|GetIgnoredPlayersByClass)
                                  (IgnoredPlayerSystem|UpdateCollisionforPlayer _array_element_3 (IgnoredPlayerSystem|IsPlayerIgnoredTeamandClass _array_element_3)))))))))
                    (:Remote
                      ;[0 0.2171 1] If client, handle all local players
                      (bind _array_element_4 (Default|ForEachFortLocalPlayerController)
                        (:Loop%20Body
                          (CallFunction|UpdateLocalIgnore _array_element_4 (LocalVariables|GetCurrentMinigame)))))))))))
        (:else)))
    (:else)))

(fn ShouldInvisiblePropsBeAllowed
  (return true))

(fn ShouldShowInGame
  (return (Math|Boolean|ANDBoolean (Class|CreativeEnabledComponent|IsEnabled (Variables|Default|GetEnabledComponent)) (Variables|UserOptions|GetBaseVisibleDuringGame) (not (Replay|GetIsPlayingReplay)))))

(fn ShouldTeleportOverlappingPawns
  ;[0.8977 1 0.2243] Return if the volume should be able to teleport overlapping players
  (return (and (CallFunction|IsinPlayMode) (not (Utilities|Enum|Equal(Enum) (Variables|UserOptions|GetZoneShape) "NewEnumerator2")))))

(fn TeleportPlayertoNearestPoint
  (params (PlayerPawn :type Fort%20Player%20Pawn%20Object%20Reference) (ImpactPoint :type Vector))
  (bind _returnvalue (PlayerStuck|TryTeleportPlayertoDirection (Variables|GetPlayerPawn) (Math|Vector|Normalize (select (Math|Vector|VectorIsNearlyZero (Transformation|GetVelocity (Variables|GetPlayerPawn))) (- ImpactPoint (Collision|GetComponentBounds (Variables|Default|GetBarrierBounds1))) (* (Transformation|GetVelocity (Variables|GetPlayerPawn)) -1.0))) (Variables|GetImpactPoint)))
  (return _returnvalue))

(fn TryTeleportPlayerToAnySide
  (params (Pawn :type Fort%20Player%20Pawn%20Object%20Reference) (ImpactPoint :type Vector))
  (bind _teleportpossibledirection (Variables|PlayerStuck|GetTeleportPossibleDirection))
  (bind _pawn (Variables|GetPawn))
  (bind _possibledirectionsremaining (LocalVariables|GetPossibleDirectionsRemaining))
  (LocalVariables|SetPossibleDirectionsRemaining _teleportpossibledirection)
  (for _ _teleportpossibledirection
    (bind (_nearestdirection _directionindex) (PlayerStuck|GetClosestDirection _pawn _possibledirectionsremaining))
    (Utilities|Array|RemoveIndex _possibledirectionsremaining _directionindex)
    (bind _returnvalue (PlayerStuck|TryTeleportPlayertoDirection _pawn _nearestdirection (Variables|GetImpactPoint)))
    (if _returnvalue
      (return true)))
  (return false))

(fn TryTeleportPlayerToDirection
  (params (Pawn :type Fort%20Player%20Pawn%20Object%20Reference) (Direction :type Vector) (ImpactPoint :type Vector))
  (bind (_hitlocation _hitnormal _bonename _outhit _returnvalue) (Collision|LineTraceComponent (Variables|Default|GetBarrierBounds1) (+ (* Direction 1536.0) ImpactPoint) ImpactPoint true false true))
  (if _returnvalue
    (bind _returnvalue_2 (Transform|Actor|TeleportPlayerPawn (Variables|GetPawn) _hitlocation (Transformation|GetActorRotation (Variables|GetPawn))))
    (return _returnvalue_2)
    (else
      (return false))))

(fn UpdateBarrierState
  ;[1 0.457 0.4371] Set the Collision State of the Barrier and kill volume and then update the visuals of the barrier
  (if (not (Creative|UEFN|IsUEFNEditModeTransitionPending 0))
    (Collision|SetCollisionEnabled (Variables|Default|GetBarrierBounds) (select (Class|CreativeEnabledComponent|IsEnabled (Variables|Default|GetEnabledComponent)) "QueryAndPhysics" "NoCollision"))
    (Collision|SetCollisionEnabled (Variables|Default|GetBarrierBounds1) (select (and (PlayerStuck|ShouldTeleportOverlappingPawns) (Class|CreativeEnabledComponent|IsEnabled (Variables|Default|GetEnabledComponent))) "QueryAndPhysics" "NoCollision"))
    (Default|UpdateDeviceVisuals)
    (CallFunction|SetasBuildingVolume (Utilities|Enum|Equal(Enum) (Variables|UserOptions|GetZoneShape) "NewEnumerator2"))
    (Navigation|UpdateNavigationProperties)))

(fn UpdateCollisionforAI
  (params (AIPawn :type Fort%20Pawn%20Object%20Reference) (bIsBot :type Boolean) (BotPawn :type Fort%20Player%20Pawn%20Object%20Reference))
  (Utilities|IsValid AIPawn
    (:Is%20Valid
      (LocalVariables|SetPawn (Variables|GetAIPawn)))
    (:Is%20Not%20Valid
      (if bIsBot
        (Utilities|IsValid (Variables|GetBotPawn)
          (:Is%20Valid
            (bind _output_get (LocalVariables|SetPawn (Variables|GetBotPawn))))))))
  (Collision|IgnoreActorwhenMoving (Variables|/Script/Engine.Character|GetCapsuleComponent _output_get) self true)
  (Collision|IgnoreActorwhenMoving (Variables|Default|GetBarrierBounds) (LocalVariables|GetPawn) true)
  (Collision|IgnoreActorwhenMoving (Variables|Default|GetBarrierBounds1) (LocalVariables|GetPawn) true))

(fn UpdateCollisionForLocalPlayer
  (params (PlayerController :type Object%20Reference) (bShouldIgnore :type Boolean))
  (bind _returnvalue (Utilities|Name|MakeLiteralName "NoClambering"))
  (bind _collide_with_camera (Variables|UserOptions|GetCollide%20with%20Camera))
  (bind _returnvalue_2 (select _collide_with_camera "ECR_Block" "ECR_Ignore"))
  (bind _bshouldignore (Variables|GetbShouldIgnore))
  (bind _returnvalue_3 (select _bshouldignore "ECR_Ignore" "ECR_Block"))
  (bind _barrierbounds (Variables|Default|GetBarrierBounds))
  (bind _tags (Variables|Actor|GetTags))
  (bind _barrierbounds1 (Variables|Default|GetBarrierBounds1))
  ;[1 0.4635 0 0.2] Update Actor Tags and Blocking Channels on Client to enable/disable mantling on the barrier and adjust camera interaction accordingly for the Local Player
  (bind _array_element (Default|ForEachFortLocalPlayerController)
    (:Loop%20Body
      (Utilities|IsValid _array_element
        (:Is%20Valid
          (if (== _array_element (Variables|GetPlayerController))
            (if _bshouldignore
              (if (not (Utilities|Array|ContainsItem _tags _returnvalue))
                (Utilities|Array|Add _tags _returnvalue))
              (Collision|SetCollisionResponsetoChannel _barrierbounds "ECC_Camera" _returnvalue_3)
              (Collision|SetCollisionResponsetoChannel _barrierbounds "ECC_GameTraceChannel1" "ECR_Ignore")
              (Collision|SetCollisionResponsetoChannel _barrierbounds1 "ECC_Camera" _returnvalue_3)
              (Collision|SetCollisionResponsetoChannel _barrierbounds1 "ECC_GameTraceChannel1" _returnvalue_3)
              (else
                (Utilities|Array|RemoveItem _tags _returnvalue)
                (Collision|SetCollisionResponsetoChannel _barrierbounds "ECC_Camera" _returnvalue_2)
                (Collision|SetCollisionResponsetoChannel _barrierbounds "ECC_GameTraceChannel1" _returnvalue_2)
                (Collision|SetCollisionResponsetoChannel _barrierbounds1 "ECC_Camera" _returnvalue_2)
                (Collision|SetCollisionResponsetoChannel _barrierbounds1 "ECC_GameTraceChannel1" _returnvalue_2)))))))))

(fn UpdateDeviceVisuals
  (Default|UpdateBoundsVisibility)
  (Default|UpdateFade (or (Class|CreativeEnabledComponent|IsEnabled (Variables|Default|GetEnabledComponent)) (not (CallFunction|IsinPlayMode)))))

(fn UpdateLocalIgnore
  (params (Player%20Controller :type Controller%20Object%20Reference) (Current%20Minigame :type Fort%20Minigame%20Object%20Reference))
  (bind _current_minigame (Variables|GetCurrent%20Minigame))
  (bind _returnvalue (Minigame|GetFNEClassComponent _current_minigame))
  (Utilities|IsValid Player%20Controller
    (:Is%20Valid
      (bind _asfort_player_state (Utilities|Casting|CastToFortPlayerState (Variables|/Script/Engine.Controller|GetPlayerState Player%20Controller)))
      (LocalVariables|SetPlayer%20State _asfort_player_state)
      (Utilities|FlowControl|Sequence
        (:then_0
          (if (Variables|UserOptions|GetIgnoreTeam_Override)
            (IgnoredPlayerSystem|UpdatePlayerStateArray (Variables|Transient|GetIgnoredPlayersByTeam) (LocalVariables|GetPlayer%20State) (IgnoredPlayerSystem|IsIgnoredTeam (Team|GetActorTeam (LocalVariables|GetPlayer%20State))))
            (IgnoredPlayerSystem|UpdateCollisionforPlayer (LocalVariables|GetPlayer%20State) (IgnoredPlayerSystem|IsPlayerIgnoredTeamandClass (LocalVariables|GetPlayer%20State)))))
        (:then_1
          (if (Variables|UserOptions|GetIgnoreClass_Override)
            (Utilities|IsValid _current_minigame
              (:Is%20Valid
                (Utilities|IsValid _returnvalue
                  (:Is%20Valid
                    (IgnoredPlayerSystem|UpdatePlayerStateArray (Variables|Transient|GetIgnoredPlayersByClass) (LocalVariables|GetPlayer%20State) (IgnoredPlayerSystem|IsIgnoredClass (Minigame|GetClassSlotforPlayer _returnvalue (LocalVariables|GetPlayer%20State) false)))
                    (IgnoredPlayerSystem|UpdateCollisionforPlayer (LocalVariables|GetPlayer%20State) (IgnoredPlayerSystem|IsPlayerIgnoredTeamandClass (LocalVariables|GetPlayer%20State)))))))))))))

(fn UpdateRelevancy
  (if (Variables|UserOptions|GetAffectedByFadeDistance)
    (Variables|Replication|SetbAlwaysRelevant false)
    (else
      (Variables|Replication|SetbAlwaysRelevant true))))

(fn UpdateAllPlayersCollision
  (for _array_element (Minigame|GetParticipatingPlayers (Variables|/Script/FortniteGame.FortMinigameLogicComponent|GetCurrentMinigame (Variables|Default|GetFortMinigameProgress)))
    (IgnoredPlayerSystem|UpdateCollisionforPlayer _array_element (IgnoredPlayerSystem|IsIgnoredPlayer _array_element))))

(fn UpdateBaseVisibility
  (bind _returnvalue (Minigame|GetMinigameforActor self))
  (Utilities|IsValid _returnvalue
    (:Is%20Valid
      (LocalVariables|SetIsPlayMode (Minigame|IsinPlayMode _returnvalue :!class /Script/FortniteGame.FortMinigame)))
    (:Is%20Not%20Valid
      _))
  (Visibility|SetVisiblityDuringPhase (Variables|Default|GetVisibleInGameComponent) (select (select (LocalVariables|GetIsPlayMode) (Visibility|ShouldShowinGame) (Variables|UserOptions|GetBaseVisibleDuringGame)) "NewEnumerator3" "NewEnumerator1")))

(fn UpdateBlockingChannels
  (bind _blockweaponfire (Variables|UserOptions|GetBlockWeaponFire))
  (bind _returnvalue (select _blockweaponfire "ECR_Block" "ECR_Ignore"))
  (bind _barrierbounds (Variables|Default|GetBarrierBounds))
  (Collision|SetCollisionResponsetoChannel _barrierbounds "ECC_GameTraceChannel7" _returnvalue)
  (Collision|SetCollisionResponsetoChannel _barrierbounds "ECC_GameTraceChannel5" _returnvalue)
  (Collision|SetCollisionResponsetoChannel _barrierbounds "ECC_GameTraceChannel3" _returnvalue)
  (Collision|SetCollisionResponsetoChannel _barrierbounds "ECC_GameTraceChannel8" _returnvalue)
  (Collision|SetCollisionResponsetoChannel _barrierbounds "ECC_GameTraceChannel12" _returnvalue)
  (Collision|SetCollisionResponsetoChannel _barrierbounds "ECC_Camera" (select (Variables|UserOptions|GetCollide%20with%20Camera) "ECR_Block" "ECR_Ignore"))
  ;[0.05415 0.4247 1] Setting device to ignore pawn collision while it is a preview actor.
  (Collision|SetCollisionResponsetoChannel _barrierbounds "ECC_Pawn" (select (Class|CreativeBlueprintLibrary|IsAPreviewActor self) "ECR_Ignore" "ECR_Block"))
  (Collision|SetCollisionResponsetoChannel _barrierbounds "ECC_WorldStatic" (select (Variables|UserOptions|GetBlockPhysicsProps) "ECR_Block" "ECR_Overlap")))

(fn UpdateBoundsVisibility
  (bind _returnvalue (Default|BarrierBoundsNotDisabled))
  (Visibility|UpdateBaseVisibility)
  (UEFNConversionHelper|SetComponentHiddenEditor (Variables|Default|GetBarrierBounds) (not (and _returnvalue (Math|Boolean|ORBoolean (not (Variables|Default|GetLocalPlayerIsWithinFxVolume)) (CallFunction|IsinPlayMode) (Utilities|Enum|Equal(Enum) (Variables|UserOptions|GetZoneShape) "NewEnumerator2")))))
  (Rendering|SetCastShadow (Variables|Default|GetBarrierBounds) false)
  (return _returnvalue))

(fn UpdateCollisionForNPC
  (params (AI%20Controller :type AIController%20Object%20Reference) (ShouldIgnore :type Boolean))
  (Utilities|IsValid AI%20Controller
    (:Is%20Valid
      (bind _ascharacter (Utilities|Casting|CastToCharacter (Pawn|GetControlledPawn AI%20Controller)))
      (Collision|IgnoreActorwhenMoving (Variables|/Script/Engine.Character|GetCapsuleComponent _ascharacter) self (Variables|GetShouldIgnore)))))

(fn UpdateCollisionForPlayer
  (params (PlayerState :type Player%20State%20Object%20Reference) (ShouldIgnore :type Boolean))
  (bind _playerstate (Variables|GetPlayerState))
  (bind _controller (CreativeAI|Controller|GetParticipantControllerfromActor _playerstate false))
  (Utilities|IsValid PlayerState
    (:Is%20Valid
      (Utilities|FlowControl|Sequence
        (:then_0
          ;[1 0.2809 0.08026] Same code is for both Physics and Non-Physics characters - This is called in every Client + Server
          ;;[0.05935 0.15 0.03841 0.5 12] Sometimes the Pawn may not exist (ie: respawning)
          ;; We skip and apply cosmetics
          (bind _asfort_pawn (Utilities|Casting|CastToFortPawn (PlayerState|GetPawn (Variables|GetPlayerState))))
          (Collision|IgnoreActorwhenMoving (Variables|/Script/Engine.Character|GetCapsuleComponent _asfort_pawn) self (Variables|GetShouldIgnore)) ; IgnoreActorWhenMoving is bidirectional
        )
        (:then_1
          ;[0.1604 0.4809 0.6094 0.5] Local Visibility and Camera Collision
          (Networking|CanExecuteCosmeticEvents
            (:True
              (if (Pawn|IsLocalPlayerController _controller)
                (Class|CreativeSplitscreenVisibilityComponent|HideActorfromPlayer (Variables|Default|GetCreative_SplitscreenVisibility_Device_Component) _controller self (and (Variables|UserOptions|GetInvisibleToIgnoredPlayers) (Variables|GetShouldIgnore)))
                (CallFunction|UpdateCollisionforLocalPlayer _controller (Variables|GetShouldIgnore))))))))))

(fn UpdateFade
  (params (On :type Boolean) (Immediate :type Boolean))
  (Networking|CanExecuteCosmeticEvents
    (:True
      (Variables|Default|SetAllowBoundsTransition On)
      (CallFunction|HideBarrierBoundsVisibility (and On (or (not (Variables|Default|GetLocalPlayerIsWithinFxVolume)) (CallFunction|IsinPlayMode))) Immediate) ; Only Fade if player is outwith bounds \n(If minigame isn't running)
    )))

(fn UpdateIgnoredList
  (params (Add :type Boolean) (Instigator :type Controller%20Object%20Reference))
  (Utilities|IsValid Instigator
    (:Is%20Valid
      (CallFunction|CheckNPCAddedtoIgnoreList Add Instigator)
      (bind _output_get (LocalVariables|SetInstigator%20PlayerState (Variables|/Script/Engine.Controller|GetPlayerState Instigator)))
      (Utilities|IsValid _output_get
        (:Is%20Valid
          (Utilities|Set|AddItems (LocalVariables|GetIgnored%20Players%20List%20Cache) (Variables|Transient|GetIgnoredPlayersList))
          (if (Variables|GetAdd)
            (if (not (Utilities|Set|ContainsItem (LocalVariables|GetIgnored%20Players%20List%20Cache) (LocalVariables|GetInstigator%20PlayerState)))
              (Utilities|Set|Add (LocalVariables|GetIgnored%20Players%20List%20Cache) (LocalVariables|GetInstigator%20PlayerState)))
            (else
              (bind _returnvalue (Utilities|Set|Remove (LocalVariables|GetIgnored%20Players%20List%20Cache) (LocalVariables|GetInstigator%20PlayerState)))
              (if _returnvalue
                _)))
          (bind _result (Utilities|Set|ToArray (LocalVariables|GetIgnored%20Players%20List%20Cache)))
          (Variables|Transient|SetIgnoredPlayersList _result)
          (IgnoredPlayerSystem|UpdateCollisionforPlayer (LocalVariables|GetInstigator%20PlayerState) (IgnoredPlayerSystem|IsIgnoredPlayer (LocalVariables|GetInstigator%20PlayerState))))))))

(fn UpdateNavigationProperties
  (bind _enabledcomponent (Variables|Default|GetEnabledComponent))
  (bind _enabled (Class|CreativeEnabledComponent|IsEnabled _enabledcomponent))
  (bind (_bareallpawnsignored _bareai_ignored) (CallFunction|CheckIgnoreState))
  (if _bareai_ignored
    (CallFunction|SetIsNavigationRelevant false)
    (Navigation|SetAlwaysExportNavRelevantComponent false)
    (else
      (CallFunction|SetIsNavigationRelevant _enabled)
      (Navigation|SetAlwaysExportNavRelevantComponent _enabled)
      (Navigation|SetNavigationObstacleOverride (select (CallFunction|IsHollowVolumeShape (Variables|UserOptions|GetZoneShape)) "ExportAsPrimitive" "ForceEnabled")))))

(fn UpdatePlayerStateArray
  (params (PlayerStateArray :type Array%20of%20Player%20State%20Object%20References) (PlayerState :type Player%20State%20Object%20Reference) (Add :type Boolean))
  (if Add
    (Utilities|Array|Add (Variables|GetPlayerStateArray) (Variables|GetPlayerState))
    (else
      (Utilities|Array|RemoveItem (Variables|GetPlayerStateArray) (Variables|GetPlayerState)))))

(fn UpdatePropMoverCollisionSize
  (bind _barrierheight (Variables|UserOptions|Basic|GetBarrierHeight))
  (bind _returnvalue (Math|Float|MapRangeClamped _barrierheight 0.0 1.0 0.0 1e+02))
  (Components|Box|SetBoxExtent (Variables|Default|GetPropMoverCollision) (Math|Vector|MakeVector (Math|Float|MapRangeClamped (Variables|UserOptions|Basic|GetBarrierWidth) 0.0 0.512 0.0 1e+02) (Math|Float|MapRangeClamped (Variables|UserOptions|Basic|GetBarrierDepth) 0.0 0.512 0.0 1e+02) _returnvalue) false)
  (Transformation|SetRelativeLocation (Variables|Default|GetPropMoverCollision) (Math|Vector|MakeVector 0.0 0.0 _returnvalue) false false))

(fn ConstructionScript
  (|CallParentFunction)
  (if (Editor|InEditor)
    (CallFunction|SetBarrierVisuals)))
