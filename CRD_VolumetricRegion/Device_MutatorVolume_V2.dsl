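; Creative Mutator Zone device that changes rules inside a volume, such as building, weapons, jumping, emote menu, speed and gravity, and can override pickup lifespan.
; Enter and exit events filter players, bots, creatures and vehicles, broadcasting zone and emote events.
; Options select Gameplay Effects with SetByCaller magnitudes, storing handles for removal on exit or game end.
; Team, class and vehicle occupancy changes re-evaluate effects; legacy options are migrated.
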(event EventInitializeDevice
  ;[0.9244 0.06801 1] Initialization
  (|CallParentFunction)
  (Utilities|FlowControl|Sequence
    (:then_0
      (CallFunction|UpdateVisibilityComponentsToUserSettings))
    (:then_1
      (Visibility|UpdateVisiblity))
    (:then_2
      (CallFunction|UpdateCoreExperienceSettings))))

(event Custom|OnPawnStartedEmote
  (params (MontageItemDef :type Fort%20Item%20Definition%20Object%20Reference) (PawnEmoting :type Fort%20Pawn%20Object%20Reference))
  (bind _returnvalue (Pawn|GetController PawnEmoting))
  ;[0.9747 1 0] Emote Events
  (Utilities|IsValid _returnvalue
    (:Is%20Valid
      (Default|TransmitonEmoteStartedIfValidActor _returnvalue PawnEmoting))))

(event Custom|UpdateTeamOnReceived
  (params (Instigator :type Controller%20Object%20Reference))
  (bind _controller (CreativeAI|Controller|GetParticipantControllerfromActor Instigator false))
  ;[0.6869 1 0.2143] Update the selected Team or Class when receiving from channel transmissions
  ;;[0.3222 1 0.9701] Check the Authority, and Game Progress
  (Networking|SwitchHasAuthority
    (:Authority
      (bind _currentminigame (Default|GetCurrentMinigame)
        (:Is%20Valid
          (if (Minigame|IsinProgress _currentminigame)
            ;;[1 0.3279 0.2682] Get the Team and Class of the player
            (Utilities|IsValid _controller
              (:Is%20Valid
                ;;[0.9653 0.3944 1] Update the team and class to affect
                (UserOptions|CustomInitializeTeams (Knobs|Team|ReshiftTeamIndexValue (Math|Conversions|ToInteger(Byte) (Team|GetActorPvPTeamIndex _controller))))
                (Registration|RevalidateOverlappingPawns)))))))))

(event Custom|OnPawnStoppedEmote
  (params (MontageItemDef :type Fort%20Item%20Definition%20Object%20Reference) (PawnEmoting :type Fort%20Pawn%20Object%20Reference))
  (bind _returnvalue (Pawn|GetController PawnEmoting))
  ;[0.9747 1 0] Emote Events
  (Utilities|FlowControl|Delay 0.0) ; Wait for next tick to ensure\nEmoteAbility is not Active anymore
  (Utilities|IsValid _returnvalue
    (:Is%20Valid
      (bind _asathena_player_controller (Utilities|Casting|CastToAthena_PlayerController _returnvalue))
      (if (not (Emote|IsPlayingEmote _asathena_player_controller))
        (Default|TransmitonEmoteEndedIfValidActor _asathena_player_controller PawnEmoting)))))

(event Custom|UpdateClassOnReceived
  (params (Instigator :type Controller%20Object%20Reference))
  (bind _controller (CreativeAI|Controller|GetParticipantControllerfromActor Instigator false))
  ;[0.6869 1 0.2143] Update the selected Team or Class when receiving from channel transmissions
  ;;[0.3222 1 0.9701] Check the Authority, and Game Progress
  (Networking|SwitchHasAuthority
    (:Authority
      (bind _currentminigame (Default|GetCurrentMinigame)
        (:Is%20Valid
          (Utilities|IsValid (Minigame|GetFNEClassComponent _currentminigame)
            (:Is%20Valid
              (if (Minigame|IsinProgress _currentminigame)
                ;;[1 0.3279 0.2682] Get the Team and Class of the player
                (Utilities|IsValid _controller
                  (:Is%20Valid
                    (bind _asfort_player_state (Utilities|Casting|CastToFortPlayerState (Variables|/Script/Engine.Controller|GetPlayerState _controller)))
                    ;;[0.9653 0.3944 1] Update the team and class to affect
                    (UserOptions|CustomInitializeClass (Math|Conversions|ToInteger(Byte) (Minigame|GetClassSlotforPlayer (Minigame|GetFNEClassComponent _currentminigame) _asfort_player_state false)))
                    (Registration|RevalidateOverlappingPawns)))))))))))

(event Custom|SpawnFX
  (params (Pawn :type Fort%20Pawn%20Object%20Reference))
  ;[0.1994 0.1886 1] Replicate it so that it actually spawns VFX
  (Networking|CanExecuteCosmeticEvents
    (:True
      (Effects|SpawnFXSystemAttached (Variables|Default|GetEmitter%20Template) (Variables|/Script/Engine.Actor|GetRootComponent Pawn)))))

(event EventPawnExitedTriggerVolume
  (params (Pawn :type Pawn%20Object%20Reference) (PlayerController :type Controller%20Object%20Reference))
  (bind _returnvalue (Pawn|GetController Pawn))
  ;[1 0.1927 0.1927] End the Effect on pawns leaving the volume
  (Utilities|IsValid Pawn
    (:Is%20Valid
      (if (Utilities|IsValid PlayerController :!class /Script/Engine.KismetSystemLibrary)
        (Default|EndEffect 0 Pawn)
        (bind _bwouldaffectplayer (CallFunction|WouldVolumeAffectPlayer Pawn))
        (if _bwouldaffectplayer
          (Utilities|IsValid _returnvalue
            (:Is%20Valid
              (GameplayEvent|BroadcastEvent (Variables|UserOptions|Events|GetOn%20Player%20Exiting%20Zone) 0 (Utilities|Struct|MakeMutatorPlayerExitedInstigator _returnvalue)))))
        (elif (Variables|UserOptions|GetbAffectsCreatures)
          (Default|EndEffect 0 Pawn)
          (Utilities|IsValid _returnvalue
            (:Is%20Valid
              _))))
      (GameplayEvent|BroadcastEvent (Variables|UserOptions|Events|GetOn%20Agent%20Exiting%20Zone) 0 (Utilities|Struct|MakeMutatorAgentExitedInstigator _returnvalue)))))

(event EventPawnEnteredTriggerVolume
  (params (Pawn :type Pawn%20Object%20Reference) (PlayerController :type Controller%20Object%20Reference))
  (bind _returnvalue (Pawn|GetController Pawn))
  ;[0.2776 1 0.43] Filter through what kind of pawn has entered the volume and apply the appropriate Effect
  (Utilities|IsValid Pawn
    (:Is%20Valid
      (if (Utilities|IsValid PlayerController :!class /Script/Engine.KismetSystemLibrary)
        (bind _bwouldaffectplayer (CallFunction|WouldVolumeAffectPlayer Pawn))
        (if (and _bwouldaffectplayer (not (CallFunction|DoesPawnAlreadyHaveEffect Pawn)))
          (Default|TriggerEffect Pawn)
          (Utilities|IsValid _returnvalue
            (:Is%20Valid
              (GameplayEvent|BroadcastEvent (Variables|UserOptions|Events|GetOn%20Player%20Entering%20Zone) 0 (Utilities|Struct|MakeMutatorPlayerEnteredInstigator _returnvalue)))))
        (else
          (if (CallFunction|ShouldAffectCreatures Pawn)
            (Default|TriggerEffect Pawn))))
      (GameplayEvent|BroadcastEvent (Variables|UserOptions|Events|GetOn%20Agent%20Entering%20Zone) 0 (Utilities|Struct|MakeMutatorAgentEnteredInstigator _returnvalue))
      (if (Variables|UserOptions|GetEnableVFX)
        (bind _asfort_pawn (Utilities|Casting|CastToFortPawn Pawn))
        (CallFunction|SpawnFX _asfort_pawn)))))

(event EventNon-PawnExitedTriggerVolume
  (params (Actor :type Actor%20Object%20Reference))
  ;[1 0.1927 0.1927] End the Effect on pawns leaving the volume
  (if (Variables|DamageOptions|GetbDamageVehicles)
    (bind _asfort_athena_vehicle (Utilities|Casting|CastToFortAthenaVehicle Actor))
    (CallFunction|UnbindVehicleEvents _asfort_athena_vehicle)
    (Default|EndEffect _asfort_athena_vehicle)))

(event EventPlayerBotExitedTriggerVolume
  (params (Pawn :type Pawn%20Object%20Reference) (AI%20Bot%20Controller :type Fort%20Athena%20AIBot%20Controller%20Object%20Reference))
  ;[1 0.1927 0.1927] End the Effect on pawns leaving the volume
  (Utilities|IsValid Pawn
    (:Is%20Valid
      (if (Utilities|IsValid AI%20Bot%20Controller :!class /Script/Engine.KismetSystemLibrary)
        (if (CallFunction|ShouldBotbeAffected Pawn)
          (Default|EndEffect 0 Pawn)
          (GameplayEvent|BroadcastEvent (Variables|UserOptions|Events|GetOn%20Agent%20Exiting%20Zone) 0 (Utilities|Struct|MakeMutatorAgentExitedInstigator AI%20Bot%20Controller)))))))

(event EventOnPlayerChangedTeam
  (params (PlayerState :type Fort%20Player%20State%20Object%20Reference) (TeamId :type Byte) (OldTeamId :type Byte) (bTeamChangedWithoutRespawn :type Boolean))
  ;[0.2776 1 0.43] Filter through what kind of pawn has entered the volume and apply the appropriate Effect
  ;;[0.2785 1 0.9402 0.5] Re-fresh application on the damage volume for a player who's changed class or teams
  (bind (_outpawn _pawnreturned) (Class|AbilityFunctions|GetPawnfromPlayerStateZone PlayerState))
  (CallFunction|ReapplyEffectsonOverlappingPlayers _outpawn _pawnreturned))

(event EventOnPlayerClassApplied
  (params (PlayerState :type Fort%20Player%20State%20Object%20Reference) (ClassSlotIndex :type Byte))
  ;[0.2776 1 0.43] Filter through what kind of pawn has entered the volume and apply the appropriate Effect
  ;;[0.2785 1 0.9402 0.5] Re-fresh application on the damage volume for a player who's changed class or teams
  (bind (_outpawn _pawnreturned) (Class|AbilityFunctions|GetPawnfromPlayerStateZone PlayerState))
  (CallFunction|ReapplyEffectsonOverlappingPlayers _outpawn _pawnreturned))

(event EventNon-PawnEnteredTriggerVolume
  (params (Actor :type Actor%20Object%20Reference))
  ;[0.2776 1 0.43] Filter through what kind of pawn has entered the volume and apply the appropriate Effect
  (Networking|SwitchHasAuthority
    (:Authority
      (if (and (Variables|DamageOptions|GetbDamageVehicles) (CallFunction|CanPassinCreateMode))
        (bind _asfort_athena_vehicle (Utilities|Casting|CastToFortAthenaVehicle Actor))
        (CallFunction|BindVehicleEvents _asfort_athena_vehicle)
        (Default|TriggerEffect 0 _asfort_athena_vehicle)
        (else
          (Default|HandlePickUpBeginOverlap Actor))))))

(event MinigameStartedDelegate(FortMinigameProgress)
  ;[0.2619 1 0.8901 0.702] Bind and Unbind events for player team and class states changing
  (Networking|SwitchHasAuthority
    (:Authority
      (bind _currentminigame (Default|GetCurrentMinigame)
        (:Is%20Valid
          (Minigame|BindEventtoOnPlayerClassApplied _currentminigame (EventDispatchers|CreateEvent :!event OnPlayerClassApplied))
          (Minigame|BindEventtoPlayerChangedTeamDelegate _currentminigame (EventDispatchers|CreateEvent :!event OnPlayerChangedTeam)))))))

(event Custom|Player%20Exited%20Vehicle
  (params (Vehicle :type Fort%20Vehicle%20Interface%20Interface%20%28by%20ref%29) (PlayerPawn :type Fort%20Pawn%20Object%20Reference) (SeatIndex :type Integer))
  ;[0.2776 1 0.43] Filter through what kind of pawn has entered the volume and apply the appropriate Effect
  ;;[0.271 0.222 1 0.321] When a player enters or exits a vehicle, re-trigger the effects on the vehicle to filter out if it's still affected
  (if (< (Utilities|Array|Length (Vehicle|GetAllPassengers Vehicle)) 1)
    (CallFunction|Re-TriggerEffectsOnVehicle Vehicle)))

(event MinigameEndedDelegate(FortMinigameProgress)
  ;[0.2619 1 0.8901 0.702] Bind and Unbind events for player team and class states changing
  (CallFunction|CleanUpatGameEnd)
  (Networking|SwitchHasAuthority
    (:Authority
      (bind _currentminigame (Default|GetCurrentMinigame)
        (:Is%20Valid
          (Minigame|UnbindEventfromOnPlayerClassApplied _currentminigame (EventDispatchers|CreateEvent :!event OnPlayerClassApplied))
          (Minigame|UnbindEventfromPlayerChangedTeamDelegate _currentminigame (EventDispatchers|CreateEvent :!event OnPlayerChangedTeam)))))))

(event Custom|Player%20Entered%20Vehicle
  (params (Vehicle :type Fort%20Vehicle%20Interface%20Interface%20%28by%20ref%29) (PlayerPawn :type Fort%20Pawn%20Object%20Reference) (SeatIndex :type Integer))
  ;[0.2776 1 0.43] Filter through what kind of pawn has entered the volume and apply the appropriate Effect
  ;;[0.271 0.222 1 0.321] When a player enters or exits a vehicle, re-trigger the effects on the vehicle to filter out if it's still affected
  (if (== (Utilities|Array|Length (Vehicle|GetAllPassengers Vehicle)) 1)
    (CallFunction|Re-TriggerEffectsOnVehicle Vehicle)))

(event Custom|Add%20or%20Remove%20Affected%20Actor
  (params (Remove%20Actor :type Boolean) (Actor :type Actor%20Object%20Reference))
  (bind _affectedactor (Variables|Default|GetAffectedActor))
  ;[0.5208 1 1 0.4] Keep a list of affected actors to handle on game end cleanup
  (Networking|SwitchHasAuthority
    (:Authority
      (if Remove%20Actor
        (Utilities|Array|RemoveItem _affectedactor Actor)
        (else
          (Utilities|Array|AddUnique _affectedactor Actor)))
      (Default|BindorUnbindonPlayerEmote Actor (not Remove%20Actor)))))

(event EventPlayerBotEnteredTriggerVolume
  (params (Pawn :type Pawn%20Object%20Reference) (AI%20Bot%20Controller :type Fort%20Athena%20AIBot%20Controller%20Object%20Reference))
  ;[0.2776 1 0.43] Filter through what kind of pawn has entered the volume and apply the appropriate Effect
  (Utilities|IsValid Pawn
    (:Is%20Valid
      (if (Utilities|IsValid AI%20Bot%20Controller :!class /Script/Engine.KismetSystemLibrary)
        (if (Math|Boolean|ANDBoolean (CallFunction|ShouldBotbeAffected Pawn) (CallFunction|CanPassinCreateMode) (StateChecks|IsValidTeam (Variables|Default|GetCreativeUseRestrictionComponent) Pawn) (not (CallFunction|DoesPawnAlreadyHaveEffect Pawn)))
          (Default|TriggerEffect Pawn)
          (bind _asfort_pawn (Utilities|Casting|CastToFortPawn Pawn))
          (CallFunction|SpawnFX _asfort_pawn)
          (GameplayEvent|BroadcastEvent (Variables|UserOptions|Events|GetOn%20Agent%20Entering%20Zone) 0 (Utilities|Struct|MakeMutatorAgentEnteredInstigator AI%20Bot%20Controller)))))))

(event Custom|Clean%20Up%20At%20Game%20End
  (bind _affectedactor (Variables|Default|GetAffectedActor))
  ;[0.5208 1 1 0.4] Keep a list of affected actors to handle on game end cleanup
  (Networking|SwitchHasAuthority
    (:Authority
      (bind (_arrayindex _arrayelement) (Utilities|Array|ReverseforEachLoop _affectedactor)
        (:LoopBody
          (Default|EndEffect _arrayelement))
        (:Completed
          (Utilities|Array|Clear _affectedactor))))))

(event HandleVersionUpgrades(CreativeVersion)
  (params (LastSavedVersion :type Creative%20Version%20Id%20Structure))
  (CreativeVersion|CompareVersiontoTag LastSavedVersion "(TagName=\"Creative.Version.Release.40.20\")"
    (:OlderVersion
      (Variables|UserOptions|SetEnableVFX (Variables|Legacy|GetbEnableVFX))
      (OptionData|RefreshPropertyOverrides (Variables|Default|GetToyOptionsComponent)))))

(fn AddHandleToHoldingArray
  (params (ActiveGameplayEffectHandle :type Active%20Gameplay%20Effect%20Handle%20Structure) (Ability%20System%20Component :type Ability%20System%20Component%20Object%20Reference))
  ;[0.2257 0.8599 1] Add the player's ability system component and the active gameplay and put them in a struct so we can retain knowledge of the handle of the GE we just applied to the player
  (bind _output_get (Variables|Default|SetActiveGameplayEffectHandle ActiveGameplayEffectHandle))
  (bind _output_get_2 (Variables|SetGE%20Handle_ActiveGameplayEffectHandle_7_C6FF36454DF9DCBF6904BD96974FFF0B (Utilities|Struct|MakeAbilityHandleAndSystemHold _output_get Ability%20System%20Component)))
  (Utilities|Array|Add (Variables|Default|GetAbility%20Handle%20Array) _output_get_2))

(fn AffectsTarget
  (params (Target%20Actor :type Actor%20Object%20Reference))
  (bind _asfort_player_pawn (Utilities|Casting|CastToFortPlayerPawn Target%20Actor)
    (:then
      (if (Variables|UserOptions|GetbAffectsPlayers)
        (return (CreativeUserRestrictions|CanParticipantActorInteract (Variables|Default|GetCreativeUseRestrictionComponent) _asfort_player_pawn false true false))
        (else
          (return false))))
    (:CastFailed
      (bind _asfort_pawn (Utilities|Casting|CastToFortPawn Target%20Actor))
      (if (Variables|UserOptions|GetbAffectsCreatures)
        (return (StateChecks|IsValidTeam (Variables|Default|GetCreativeUseRestrictionComponent) _asfort_pawn))
        (else
          (return false))))))

(fn ApplyEffectsAndAssignHandles
  (params (AbilitySystemComponent :type Ability%20System%20Component%20Object%20Reference) (SpecHandle :type Gameplay%20Effect%20Spec%20Handle%20Structure%20%28by%20ref%29))
  ;[0.426 1 0.2344] Apply the Effect and add it to our custom array of structs so that we remember what we applied to who
  (bind _returnvalue (GameplayEffects|ApplyGameplayEffectSpecToTarget AbilitySystemComponent SpecHandle AbilitySystemComponent))
  (CallFunction|AddHandletoHoldingArray _returnvalue AbilitySystemComponent))

(fn ApplyGE
  (params (Actor :type Actor%20Object%20Reference))
  (bind _returnvalue (Ability|GetAbilitySystemComponent Actor))
  (bind _returnvalue_2 (GameplayEffects|MakeEffectContext _returnvalue))
  ;[1 0.9023 0.1292] Using the targeted object, create a Gameplay Effect Spec and Apply them based on the GE being applied
  (CallFunction|AddOrRemoveAffectedActor false Actor)
  (bind _gameplay_effects (CallFunction|GetVolumeGE))
  (for _array_element _gameplay_effects
    (if (Utilities|Equal(Class) (Variables|GameplayEffects|GetGE_BuildingSpeed) _array_element)
      (bind _returnvalue_3 (Ability|GameplayEffect|AssignTagSetbyCallerMagnitude (GameplayEffects|MakeOutgoingSpec _returnvalue _array_element 0.0 _returnvalue_2) "(TagName=\"SetByCaller.Magnitude\")" (/ (Variables|UserOptions|GetBuildingSpeed) 1e+02)))
      (CallFunction|ApplyEffectsandAssignHandles _returnvalue _returnvalue_3)
      (elif (Utilities|Equal(Class) _array_element (Variables|GameplayEffects|GetGE_BuildingCost))
        (bind _returnvalue_4 (Ability|GameplayEffect|AssignTagSetbyCallerMagnitude (GameplayEffects|MakeOutgoingSpec _returnvalue _array_element 0.0 _returnvalue_2) "(TagName=\"SetByCaller.Magnitude\")" (/ (Variables|UserOptions|GetBuildingCost) 1e+02)))
        (CallFunction|ApplyEffectsandAssignHandles _returnvalue _returnvalue_4)
        (elif (Utilities|Equal(Class) _array_element (Variables|GameplayEffects|GetGE%20Gravity))
          (bind _returnvalue_5 (Ability|GameplayEffect|AssignTagSetbyCallerMagnitude (GameplayEffects|MakeOutgoingSpec _returnvalue _array_element 0.0 _returnvalue_2) "(TagName=\"SetByCaller.Magnitude\")" (/ (Variables|UserOptions|GetGravityOverride) 1e+02)))
          (CallFunction|ApplyEffectsandAssignHandles _returnvalue _returnvalue_5)
          (elif (Utilities|Equal(Class) _array_element (Variables|GameplayEffects|GetGE%20PawnSpeed))
            (CallFunction|ModifyPawnSpeed (Variables|GetActor))
            (else
              (CallFunction|ApplyEffectsandAssignHandles _returnvalue (GameplayEffects|MakeOutgoingSpec _returnvalue _array_element 0.0 _returnvalue_2)))))))))

(fn BindOrUnbindOnPlayerEmote
  (params (Pawn :type Actor%20Object%20Reference) (Bind :type Boolean))
  (bind _outputdelegate (EventDispatchers|CreateEvent :!event OnPawnStartedEmote))
  (bind _outputdelegate_2 (EventDispatchers|CreateEvent :!event OnPawnStoppedEmote))
  (Networking|SwitchHasAuthority
    (:Authority
      (bind _asfort_pawn (Utilities|Casting|CastToFortPawn Pawn))
      (if (Variables|GetBind)
        (EventDispatchers|BindEventtoOnPawnStartedEmote _asfort_pawn _outputdelegate)
        (EventDispatchers|BindEventtoOnPawnStoppedEmote _asfort_pawn _outputdelegate_2)
        (else
          (EventDispatchers|UnbindEventfromOnPawnStartedEmote _asfort_pawn _outputdelegate)
          (EventDispatchers|UnbindEventfromOnPawnStoppedEmote _asfort_pawn _outputdelegate_2))))))

(fn BindVehicleEvents
  (params (FortAthenaVehicle :type Fort%20Athena%20Vehicle%20Object%20Reference))
  ;[0.505 1 0.2499] Bind events when players enter or exit this vehicle
  (CallFunction|UnbindVehicleEvents FortAthenaVehicle)
  (Vehicle|BindEventtoOnPawnExitedVehicleSeat FortAthenaVehicle (EventDispatchers|CreateEvent :!event Player%20Exited%20Vehicle))
  (Vehicle|BindEventtoOnPawnEnteredVehicleSeat FortAthenaVehicle (EventDispatchers|CreateEvent :!event Player%20Entered%20Vehicle)))

(fn BuildDataRegistryResolverScope
  (Utilities|FlowControl|Sequence
    (:then_0
      (bind _output_get (LocalVariables|SetSpatialInterfaceName (SpatialGameplay|GetSpatialGameplayInterfaceFNameforActor self)))
      (if (!= _output_get "")
        (Utilities|Array|Add (LocalVariables|GetOutNameScopes) (LocalVariables|GetSpatialInterfaceName))))
    (:then_1
      (Utilities|Array|Add (LocalVariables|GetOutNameScopes) (CallFunction|GenerateScopeName)))
    (:then_2
      (return (LocalVariables|GetOutNameScopes) 2 true))))

(fn CanPassInCreateMode
  ;[1 0.9202 0.2854] Check to see if we can pass during Create mode when playing in creative
  (return (or (CallFunction|IsinPlayMode) (and (not (CallFunction|IsinPlayMode)) (Variables|DamageOptions|GetbActiveInEditor)))))

(fn CheckVehicleDamageAbility
  (params (Object :type Object%20Reference))
  ;[1 0.8039 0.2999] Should this vehicle be damaged at all?
  (return (or (and (Variables|DamageOptions|GetbDamageVehicles) (not (CallFunction|ReturnIsVehicleEmpty Object))) (and (Variables|DamageOptions|GetbDamageVehicles) (Variables|DamageOptions|GetbDamageEmptyVehicles)))))

(fn CheckForAIGameplayEffectReplacement
  (params (TargetActor :type Actor%20Object%20Reference) (InGameplayEffect :type Gameplay%20Effect%20Class%20Reference))
  (bind _selectedge (LocalVariables|GetSelectedGE))
  (bind _aigameplayeffectreplacements (Variables|GameplayEffects|GetAIGameplayEffectReplacements))
  (bind (_value _returnvalue) (Utilities|Map|Find _aigameplayeffectreplacements _selectedge))
  (LocalVariables|SetSelectedGE InGameplayEffect)
  (bind _asfort_pawn (Utilities|Casting|CastToFortPawn TargetActor)
    (:then
      (Utilities|Casting|CastToAIController (Pawn|GetController _asfort_pawn)
        (:then
          (if (Pawn|IsBotControlled _asfort_pawn)
            _
            (elif _returnvalue
              (bind _output_get (LocalVariables|SetSelectedGE _value))
              (return _output_get (Utilities|NotEqual(Class) _output_get (Variables|GetInGameplayEffect))))))
        (:CastFailed
          _)))
    (:CastFailed
      _)))

(fn CheckPlayerTeamAndClassValidity
  (params (Actor :type Actor%20Object%20Reference))
  ;[0.4111 1 0.2042] Check whether or not the controller is valid based on its team and class compared to what is being filtered
  (bind (_valid_controller _fortplayerpawn) (CreativeUserRestrictions|ActorHasValidParticipantController (Variables|Default|GetCreativeUseRestrictionComponent) Actor true true true))
  (return _valid_controller _fortplayerpawn))

(fn DoesPawnAlreadyHaveEffect
  (params (Pawn :type Object%20Reference))
  ;[0.955 1 0.2919] Cycle through all registered players to see if a pawn currently entering has already been registered. this is purely to prevent double triggers on single entry.
  (bind _array_element (Utilities|Array|ForEachLoopwithBreak (Variables|Default|GetAffectedActor))
    (:LoopBody
      (if (Utilities|Equal(Object) _array_element Pawn)
        (LocalVariables|SetPlayerHasEffect true)))
    (:Completed
      _))
  (return (LocalVariables|GetPlayerHasEffect)))

(fn EndEffect
  (params (Vehicle :type Actor%20Object%20Reference) (Pawn :type Pawn%20Object%20Reference))
  (bind _returnvalue (Utilities|IsValid Pawn :!class /Script/Engine.KismetSystemLibrary))
  (bind _returnvalue_2 (select _returnvalue Pawn Vehicle))
  ;[1 0.6484 0.217] Remove all relevant GEs upon leaving the volume
  (Utilities|IsValid _returnvalue_2
    (:Is%20Valid
      (CallFunction|RemoveActiveHandlesFromPlayer (Ability|GetAbilitySystemComponent _returnvalue_2))
      (CallFunction|AddOrRemoveAffectedActor true _returnvalue_2)
      (CallFunction|SetPlayerinMutatorZoneScope (Variables|/Script/Engine.Pawn|GetPlayerState (Variables|GetPawn))))))

(fn GenerateScopeName
  (return (Utilities|String|StringToName (Utilities|GetObjectName self))))

(fn GetVolumeGE
  (bind _gameplayeffects (LocalVariables|GetGameplayEffects))
  ;[1 0.5233 0.4534] Lets filter through our options, and set the correct Volume GE in our array of what to use down the line
  (Utilities|Array|Clear (LocalVariables|GetGameplayEffects))
  (Utilities|FlowControl|Sequence
    (:then_0
      (if (not (Variables|UserOptions|GetbAllowEditing))
        (Utilities|Array|Add _gameplayeffects (Variables|GameplayEffects|GetGE_DisableEditing))))
    (:then_1
      (if (not (Variables|UserOptions|GetbAllowWeaponFire))
        (Utilities|Array|Add _gameplayeffects (Variables|GameplayEffects|GetGE_DisableWeapons))))
    (:then_2
      (if (not (Variables|UserOptions|GetbAllowBuilding))
        (Utilities|Array|Add _gameplayeffects (Variables|GameplayEffects|GetGE_DisableBuilding))))
    (:then_3
      (if (not (Variables|UserOptions|GetbAllowJumping))
        (Utilities|Array|Add _gameplayeffects (Variables|GameplayEffects|GetGE_DisableJumping))))
    (:then_4
      (if (not (Variables|UserOptions|GetbAllowMapMarker))
        (Utilities|Array|Add _gameplayeffects (Variables|GameplayEffects|GetGE_DisableMapMarker))))
    (:then_5
      (if (not (Variables|UserOptions|GetbAllowEmoteWheel))
        (Utilities|Array|Add _gameplayeffects (Variables|GameplayEffects|GetGE_DisableEmoteMenu))))
    (:then_6
      (if (!= (/ (Variables|UserOptions|GetBuildingSpeed) 1e+02) 1.0)
        (Utilities|Array|Add _gameplayeffects (Variables|GameplayEffects|GetGE_BuildingSpeed))))
    (:then_7
      (if (!= (/ (Variables|UserOptions|GetBuildingCost) 1e+02) 1.0)
        (Utilities|Array|Add _gameplayeffects (Variables|GameplayEffects|GetGE_BuildingCost))))
    (:then_8
      (if (!= (/ (Variables|UserOptions|GetMovementMultiplier) 1e+02) 1.0)
        (Utilities|Array|Add _gameplayeffects (Variables|GameplayEffects|GetGE%20PawnSpeed))))
    (:then_9
      (if (!= (/ (Variables|UserOptions|GetGravityOverride) 1e+02) 1.0)
        (Utilities|Array|Add _gameplayeffects (Variables|GameplayEffects|GetGE%20Gravity))))
    (:then_10
      (return _gameplayeffects))))

(fn GetZoneShape
  (return (Variables|UserOptions|GetZone%20Shape)))

(fn GetCacheContextOverride
  (return 0))

(fn GetDesiredLifeSpan
  (if (Variables|UserOptions|GetPickupLife-span_Override)
    (return (Math|Float|SelectFloat 0.1 (Variables|UserOptions|GetPickupLife-span) (Math|Float|NearlyEqual(Float) (Variables|UserOptions|GetPickupLife-span))))
    (else
      (return -1.0))))

(fn HandlePickUpBeginOverlap
  (params (Actor :type Actor%20Object%20Reference))
  (if (Math|Boolean|ANDBoolean (Utilities|IsValid Actor :!class /Script/Engine.KismetSystemLibrary) (Utilities|ClassIsChildOf (Utilities|GetClass Actor) "/Script/FortniteGame.FortPickup") (>= (Variables|UserOptions|GetPickupLife-span) 0.0) (Variables|UserOptions|GetPickupLife-span_Override))
    ;[1 1 1] When disabled we reset Pickup Life Span
    (if (Class|CreativeEnabledComponent|IsEnabled (Variables|Default|GetEnabledComponent))
      (if (Math|Float|NearlyEqual(Float) (CallFunction|GetDesiredLifeSpan) -1.0)
        (CallFunction|ResetPickupLifespanandStopDespawnTimer Actor)
        (else
          (CallFunction|ResetPickupLifespanandStopDespawnTimer Actor)
          (bind _asfort_pickup (Utilities|Casting|CastToFortPickup Actor))
          (Pickup|SetDespawnTime _asfort_pickup (CallFunction|GetDesiredLifeSpan))))
      (else
        (CallFunction|ResetPickupLifespanandStopDespawnTimer Actor)))))

(fn InitAIGameplayEffectReplacements
  (Variables|GameplayEffects|SetAIGameplayEffectReplacements (Utilities|Map|MakeMap (Variables|GameplayEffects|GetGE%20PawnSpeed) (Variables|GameplayEffects|GetGE%20AI%20PawnSpeed))))

(fn MigrateLegacyUserOptions
  (bind (_enablevfx _ballowmapmarker _ballowemotewheel _ballowjumping _on_player_starts_jamming _on_player_stops_jamming _allowbuilding) (Class|GetClassDefaults "/CRD_VolumetricRegion/Device_MutatorVolume_V2.Device_MutatorVolume_V2_C"))
  (bind (_enablevfx_2 _ballowmapmarker_2 _ballowemotewheel_2 _ballowjumping_2 _on_player_starts_jamming_2 _on_player_stops_jamming_2 _zone_shape _ge_disablejumping _ge_disablemapmarker _ge_disableemotemenu _on_agent_exiting_zone _on_agent_entering_zone _binvincibility _binvincibility_override _enabledonphase) (Class|GetClassDefaults "/CRD_VolumetricRegion/Device_MutatorVolume_V2.Device_MutatorVolume_V2_C"))
  (bind (_enablevfx_3 _ballowmapmarker_3 _ballowemotewheel_3 _ballowjumping_3 _on_player_starts_jamming_3 _on_player_stops_jamming_3 _pickuplifespan) (Class|GetClassDefaults "/CRD_VolumetricRegion/Device_MutatorVolume_V2.Device_MutatorVolume_V2_C"))
  (bind _pickuplifespan_2 (Variables|Legacy|GetPickupLifeSpan))
  (bind _returnvalue (!= _pickuplifespan_2 _pickuplifespan))
  (bind _movementmultiplier (Variables|UserOptions|GetMovementMultiplier))
  (bind _buildingspeed (Variables|UserOptions|GetBuildingSpeed))
  (bind _gravityoverride (Variables|UserOptions|GetGravityOverride))
  (bind _buildingcost (Variables|UserOptions|GetBuildingCost))
  (|CallParentFunction)
  (Utilities|FlowControl|Sequence
    (:then_0
      (Utilities|FlowControl|Sequence
        (:then_0
          (bind _migratedproperties (DeviceConversion|MigrateLegacyEnabledIndex (Variables|UserOptions|GetEnabledOnPhase) _enabledonphase))
          (LocalVariables|SetOptionsMigrated _migratedproperties))
        (:then_1
          (if (Utilities|Enum|NotEqual(Enum) (Variables|Legacy|GetAllowBuilding) _allowbuilding)
            (Variables|UserOptions|SetbAllowBuilding false)
            (Variables|Legacy|SetAllowBuilding _allowbuilding)))
        (:then_2
          (if _returnvalue
            (Variables|UserOptions|SetPickupLife-span_Override _returnvalue)
            (Variables|UserOptions|SetPickupLife-span _pickuplifespan_2)
            (Variables|Legacy|SetPickupLifeSpan _pickuplifespan)))
        (:then_3
          (Utilities|FlowControl|Sequence
            (:then_0
              (if (<= _buildingcost 2.0)
                (Variables|UserOptions|SetBuildingCost (* _buildingcost 1e+02))))
            (:then_1
              (if (<= _gravityoverride 19.0)
                (Variables|UserOptions|SetGravityOverride (* _gravityoverride 1e+02))))
            (:then_2
              (if (<= _movementmultiplier 2.0)
                (Variables|UserOptions|SetMovementMultiplier (* _movementmultiplier 1e+02))
                (LocalVariables|SetOptionsMigrated true)))
            (:then_3
              (if (<= _buildingspeed 24.0)
                (Variables|UserOptions|SetBuildingSpeed (* _buildingspeed 1e+02))))))))
    (:then_1
      (if (LocalVariables|GetOptionsMigrated)
        (OptionData|RefreshPropertyOverrides (Variables|Default|GetToyOptionsComponent))))))

(fn ModifyPawnSpeed
  (params (TargetActor :type Actor%20Object%20Reference))
  (bind (_outgameplayeffectspec _wasreplaced) (Default|CheckforAIGameplayEffectReplacement TargetActor (select (and (Variables|UserOptions|GetbOverrideAllMovementBonusesAtZero) (== (Variables|UserOptions|GetMovementMultiplier) 0.0)) (Variables|GameplayEffects|GetGEPawnImmobilize) (Variables|GameplayEffects|GetGE%20PawnSpeed))))
  (LocalVariables|SetbUseAIEffect _wasreplaced)
  (bind _returnvalue (Ability|GameplayEffect|AssignTagSetbyCallerMagnitude (GameplayEffects|MakeOutgoingSpec :self (Ability|GetAbilitySystemComponent (Variables|GetTargetActor)) :GameplayEffectClass (select (and _wasreplaced (== (Variables|UserOptions|GetMovementMultiplier) 0.0)) (Variables|GameplayEffects|GetGE%20AI%20Immobilize) _outgameplayeffectspec) :Level 0.0) (select (LocalVariables|GetbUseAIEffect) "(TagName=\"SetByCaller.MoveSpeed\")" "(TagName=\"SetByCaller.Magnitude\")") (/ (Variables|UserOptions|GetMovementMultiplier) 1e+02)))
  (CallFunction|ApplyEffectsandAssignHandles (Ability|GetAbilitySystemComponent (Variables|GetTargetActor)) _returnvalue))

(fn OnOverlap
  (params (Has%20Ended%20Overlap :type Boolean) (Other%20Actor :type Actor%20Object%20Reference))
  ;[0 0.7081 1 0.2] Mutator zone affects more than the parent does, so if the parent overlap function fails to affect the actor, call our 'Non Pawn Entered/Exited' functions
  (bind _actor_affected (|CallParentFunction Has%20Ended%20Overlap Other%20Actor))
  (if (not _actor_affected)
    (if Has%20Ended%20Overlap
      (CallFunction|Non-PawnExitedTriggerVolume Other%20Actor)
      (else
        (CallFunction|Non-PawnEnteredTriggerVolume Other%20Actor)))
    (return true)))

(fn OnPreviewStart
  (return false))

(fn Re-TriggerEffectsonVehicle
  (params (Object :type Fort%20Vehicle%20Interface%20Interface))
  ;[0.2308 0.1369 1] Remove then re-apply the effects on the vehicle in the event that it becomes occupied or vacant
  (bind _asfort_athena_vehicle (Utilities|Casting|CastToFortAthenaVehicle Object))
  (Default|EndEffect _asfort_athena_vehicle)
  (CallFunction|Non-PawnEnteredTriggerVolume _asfort_athena_vehicle))

(fn ReapplyEffectsOnOverlappingPlayers
  (params (Player%20Pawn :type Fort%20Player%20Pawn%20Object%20Reference) (Pawn%20Returned :type Boolean))
  (bind _playercontroller (Default|GetPlayerControllerfromActor Player%20Pawn))
  ;[0.2458 1 0.79 0.5] If we have a player pawn, and are overlaping it, check to see if volume's affect on the player has changed
  (Utilities|IsValid Player%20Pawn
    (:Is%20Valid
      (Utilities|IsValid _playercontroller
        (:Is%20Valid
          (LocalVariables|SetParticipantController _playercontroller)
          (if (Math|Boolean|ANDBoolean (Utilities|IsValid (Variables|Default|GetFNE_VolumeOverlapComponent) :!class /Script/Engine.KismetSystemLibrary) (Collision|IsOverlappingActor (Variables|Default|GetFNE_VolumeOverlapComponent) Player%20Pawn :!class /Script/Engine.PrimitiveComponent) Pawn%20Returned)
            (LocalVariables|SetbOldAffectedByVolume (CallFunction|DoesPawnAlreadyHaveEffect (Variables|GetPlayer%20Pawn)))
            (bind _bwouldaffectplayer (CallFunction|WouldVolumeAffectPlayer (Variables|GetPlayer%20Pawn)))
            (LocalVariables|SetbNewAffectedByVolume _bwouldaffectplayer)
            ;[0.2458 0.5908 0.3671 0.5] If the volume no longer affects the player then trigger the pawn exited event
            (if (and (not (LocalVariables|GetbNewAffectedByVolume)) (LocalVariables|GetbOldAffectedByVolume))
              (CallFunction|PawnExitedTriggerVolume (Variables|GetPlayer%20Pawn) (LocalVariables|GetParticipantController))
              (else
                ;[-0.0543 0.7613 0.9724 0.5] If the pawn is now affected by the volume then trigger the pawn entered event
                (if (and (LocalVariables|GetbNewAffectedByVolume) (not (LocalVariables|GetbOldAffectedByVolume)))
                  (CallFunction|PawnEnteredTriggerVolume (Variables|GetPlayer%20Pawn) (LocalVariables|GetParticipantController)))))))))))

(fn RemoveActiveHandlesfromPlayer
  (params (Ability%20System%20Component :type Ability%20System%20Component%20Object%20Reference))
  ;[0.3118 1 0.9701] Retrieve the Gameplay handle used to identify the gameplay effect.
  (bind (_arrayindex _arrayelement) (Utilities|Array|ReverseforEachLoop (Variables|Default|GetAbility%20Handle%20Array))
    (:LoopBody
      (if (Utilities|Equal(Object) Ability%20System%20Component (.AbilitySystemComponent_8_57E9F25741437CC55361B58F9DB743E0 (Utilities|Struct|BreakAbilityHandleandSystemHold _arrayelement)))
        (GameplayEffects|RemoveActiveGameplayEffect Ability%20System%20Component (Utilities|Struct|BreakAbilityHandleandSystemHold _arrayelement) -1)
        (Utilities|Array|RemoveIndex (Variables|Default|GetAbility%20Handle%20Array) _arrayindex)))))

(fn ResetPickupLifespanAndStopDespawnTimer
  (params (Pickup :type Actor%20Object%20Reference))
  (Actor|SetLifeSpan Pickup 0.0)
  (bind _asfort_pickup (Utilities|Casting|CastToFortPickup Pickup))
  (Pickup|SetDespawnTime _asfort_pickup 0.0))

(fn ReturnIsVehicleEmpty
  (params (Object :type Object%20Reference))
  ;[0.8591 1 0.3345] Return whether or not this vehicle is empty
  (bind _asfort_athena_vehicle (Utilities|Casting|CastToFortAthenaVehicle Object)
    (:then
      (return (<= (Utilities|Array|Length (Vehicle|GetAllPassengers _asfort_athena_vehicle)) 0)))
    (:CastFailed
      (return false))))

(fn SendTriggerMessage
  (params (Trigger%20Message%20Component :type Fort%20Gameplay%20Trigger%20Message%20Component%20Object%20Reference) (Pawn :type Pawn%20Object%20Reference))
  (bind _returnvalue (Pawn|GetController Pawn))
  (Utilities|IsValid _returnvalue
    (:Is%20Valid
      (Class|FortGameplayTriggerMessageComponent|SendTriggerMessage Trigger%20Message%20Component _returnvalue))))

(fn SetPlayerInMutatorZoneScope
  (params (PlayerState :type Player%20State%20Object%20Reference) (bIsWithinScope :type Boolean))
  (bind _currentminigame (Default|GetCurrentMinigame)
    (:Is%20Valid
      (bind _asfort_athena_mutator_mutator_zone_device (Utilities|Casting|CastToFortAthenaMutator_MutatorZoneDevice (Modifier|GetMutatorbyClass (Actor|GetComponentbyClass _currentminigame "/Script/FortniteGame.FortMutatorListComponent") "/Script/FortniteGame.FortAthenaMutator_MutatorZoneDevice")))
      (bind _output_get (LocalVariables|SetMutator _asfort_athena_mutator_mutator_zone_device))
      (Utilities|IsValid _output_get
        (:Is%20Valid
          (bind _asfort_player_state_athena (Utilities|Casting|CastToFortPlayerStateAthena (Variables|GetPlayerState)))
          (LocalVariables|SetFortPlayerState _asfort_player_state_athena)
          (if (Variables|GetbIsWithinScope)
            (MutatorZone|AddPlayertoMutatorZone (LocalVariables|GetMutator) (CallFunction|GenerateScopeName) (LocalVariables|GetFortPlayerState))
            (else
              (MutatorZone|RemovePlayerfromMutatorZone (LocalVariables|GetMutator) (CallFunction|GenerateScopeName) (LocalVariables|GetFortPlayerState)))))))))

(fn ShouldAffectCreatures
  (params (Pawn :type Pawn%20Object%20Reference))
  (return (and (Default|ActorIsACreature Pawn) (Variables|UserOptions|GetbAffectsCreatures))))

(fn ShouldBotBeAffected
  (params (BotPawn :type Pawn%20Object%20Reference))
  (bind _asfort_player_state_athena (Utilities|Casting|CastToFortPlayerStateAthena (Variables|/Script/Engine.Pawn|GetPlayerState BotPawn))
    (:then
      (return (or (and (Variables|UserOptions|GetbAffectsPlayers) (Class|FortPlayerStateAthena|IsAnAthenaGameParticipant _asfort_player_state_athena)) (and (not (Class|FortPlayerStateAthena|IsAnAthenaGameParticipant _asfort_player_state_athena)) (Variables|UserOptions|GetbAffectsGuards)))))
    (:CastFailed
      (return false))))

(fn TransmitOnEmoteEndedifValidActor
  (params (Controller :type Controller%20Object%20Reference) (Actor :type Actor%20Object%20Reference))
  (Networking|SwitchHasAuthority
    (:Authority
      (bind _shouldaffect (Default|AffectsTarget Actor))
      (if _shouldaffect
        (bind _controller (Variables|GetController)
          (:then
            (LocalVariables|SetValidController _controller))
          (:else
            (bind _asfort_player_pawn_athena (Utilities|Casting|CastToFortPlayerPawnAthena (Variables|GetActor)))
            (LocalVariables|SetValidController (Pawn|GetController _asfort_player_pawn_athena))))
        (bind _validcontroller (LocalVariables|GetValidController)
          (:then
            (GameplayEvent|BroadcastEvent (Variables|UserOptions|Events|GetOn%20Player%20Stops%20Emoting) 0 (Utilities|Struct|MakeMutatorEmoteEndedInstigator _validcontroller))
            (if (Utilities|DoesObjectImplementInterface (Ability|GetFortAbilitySystemComponent _validcontroller) "/Script/GameplayTags.GameplayTagAssetInterface")
              (Utilities|IsValid (Ability|GetFortAbilitySystemComponent _validcontroller)
                (:Is%20Valid
                  (if (not (GameplayTags|HasMatchingGameplayTag (Ability|GetFortAbilitySystemComponent _validcontroller) "(TagName=\"Gameplay.Action.Player.Jam.Jamming\")"))
                    (GameplayEvent|BroadcastEvent (Variables|UserOptions|Events|GetOn%20Player%20Stops%20Jamming) 0 (Utilities|Struct|MakeMutatorJamEndedInstigator _validcontroller))
                    (return))))))
          (:else))))))

(fn TransmitOnEmoteStartedifValidActor
  (params (Controller :type Controller%20Object%20Reference) (Actor :type Actor%20Object%20Reference))
  (Networking|SwitchHasAuthority
    (:Authority
      (bind _shouldaffect (Default|AffectsTarget Actor))
      (if _shouldaffect
        (bind _controller (Variables|GetController)
          (:then
            (LocalVariables|SetValidController _controller))
          (:else
            (bind _asfort_player_pawn_athena (Utilities|Casting|CastToFortPlayerPawnAthena (Variables|GetActor)))
            (LocalVariables|SetValidController (Pawn|GetController _asfort_player_pawn_athena))))
        (bind _validcontroller (LocalVariables|GetValidController)
          (:then
            (GameplayEvent|BroadcastEvent (Variables|UserOptions|Events|GetOn%20Player%20Starts%20Emoting) 0 (Utilities|Struct|MakeMutatorEmotestartedInstigator _validcontroller))
            (if (Utilities|DoesObjectImplementInterface (Ability|GetFortAbilitySystemComponent _validcontroller) "/Script/GameplayTags.GameplayTagAssetInterface")
              (Utilities|IsValid (Ability|GetFortAbilitySystemComponent _validcontroller)
                (:Is%20Valid
                  (if (GameplayTags|HasMatchingGameplayTag (Ability|GetFortAbilitySystemComponent _validcontroller) "(TagName=\"Gameplay.Action.Player.Jam.Jamming\")")
                    (GameplayEvent|BroadcastEvent (Variables|UserOptions|Events|GetOn%20Player%20Starts%20Jamming) 0 (Utilities|Struct|MakeMutatorJamStartedInstigator _validcontroller))
                    (return))))))
          (:else))))))

(fn TriggerEffect
  (params (Pawn :type Pawn%20Object%20Reference) (Vehicle :type Actor%20Object%20Reference))
  (bind _returnvalue (Ability|GetFortAbilitySystemComponent Pawn))
  ;[0.1639 0.174 1] Filter out if we're affecting players or vehicles
  (if (CallFunction|IsinPlayMode)
    (if (Utilities|IsValid Pawn :!class /Script/Engine.KismetSystemLibrary)
      ;[0.3017 1 0.2404] Apply the GE and put the Vehicle or Pawn in our Affected Arrays
      (CallFunction|ApplyGE Pawn)
      (CallFunction|SetPlayerinMutatorZoneScope (Variables|/Script/Engine.Pawn|GetPlayerState (Variables|GetPawn)) true)
      ;[0.3779 1 0.8944] Stop attacks if we don't allow firing
      (Networking|SwitchHasAuthority
        (:Authority
          (if (not (Variables|UserOptions|GetbAllowWeaponFire))
            (Utilities|IsValid _returnvalue
              (:Is%20Valid
                (Tags|CancelAbilitiesWithTags _returnvalue (GameplayTags|MakeLiteralGameplayTagContainer (Variables|Default|GetCancelAbilitiesWithTheseTags))))))))
      (else
        (if (CallFunction|CheckVehicleDamageAbility Vehicle)
          (CallFunction|ApplyGE Vehicle))))))

(fn UnbindVehicleEvents
  (params (FortAthenaVehicle :type Fort%20Athena%20Vehicle%20Object%20Reference))
  ;[1 0.6892 0.2661] Unbind all events where players enter or leave the specific vehicle
  (Vehicle|UnbindEventfromOnPawnEnteredVehicleSeat FortAthenaVehicle (EventDispatchers|CreateEvent :!event Player%20Entered%20Vehicle))
  (Vehicle|UnbindEventfromOnPawnExitedVehicleSeat FortAthenaVehicle (EventDispatchers|CreateEvent :!event Player%20Exited%20Vehicle)))

(fn UpdateVisibilityComponentstoUserSettings
  ;[0.4546 1 0.379] Set our visibility components to use later
  (bind _visibleingamecomponent (Variables|Default|GetVisibleInGameComponent)
    (:then
      (Visibility|SetVisibilityComponents _visibleingamecomponent (Visibility|UpdateVisualComponents) true))
    (:else)))

(fn UpdateVisualComponents
  ;[1 0.8606 0.217 0.864] Create an array of all components we need to affect visually
  (return (Utilities|Array|MakeArray 0 (Variables|Default|GetFNE_VolumeOverlapComponent))))

(fn UpdateCoreExperienceSettings
  (ScalableFloat|SetValuewithContext (LocalVariables|GetInvincibilityScalableFloat) (Variables|UserOptions|GetbInvincibility) (Variables|UserOptions|GetbInvincibility_Override)))

(fn ConstructionScript
  (|CallParentFunction)
  (Default|InitAIGameplayEffectReplacements))

(fn VFXParameters
  (bind (_returnvalue_x _returnvalue_y _returnvalue_z) (Math|Vector|BreakVector (CallFunction|GetVolumeScale)))
  (bind (_returnvalue_x_2 _returnvalue_y_2 _returnvalue_z_2) (Math|Vector|BreakVector (* (Math|Vector|MakeVector 512.0 512.0 384.0) (CallFunction|GetVolumeScale))))
  (bind (_zoneedgessize_x _zoneedgessize_y _zoneedgessize_z) (Math|Vector|BreakVector (Variables|GetZoneEdgesSize_X)))
  (return (Utilities|Array|MakeArray (Utilities|Struct|MakeParticleSysParam "BlocksVolume" "PSPT_VectorRand" 0.0 0.0 (Variables|Default|GetBlocksVolumeHigh) (Variables|Default|GetBlocksVolumeLow)) (Utilities|Struct|MakeParticleSysParam "BlocksSpawnRateScale" "PSPT_Scalar" (Utilities|ClampMaxFloat (/ (Math|Float|float*float _returnvalue_x_2 _returnvalue_y_2 _returnvalue_z_2) 100663296.0) 1e+01) 0.0 "0.000000,0.000000,0.000000" (0.0 0.0 0.0) "(B=0,G=0,R=0,A=0)" 0 0) (Utilities|Struct|MakeParticleSysParam "MeshSize" "PSPT_Vector" 0.0 0.0 (Math|Vector|MakeVector _zoneedgessize_x _zoneedgessize_y (Math|Float|SafeDivide _zoneedgessize_z _returnvalue_z)) "0.000000,0.000000,0.000000" "(B=0,G=0,R=0,A=0)" 0 0) (Utilities|Struct|MakeParticleSysParam "BlocksSize" "PSPT_Vector" 0.0 0.0 (Math|Vector|MakeVector (Utilities|Reciprocal _returnvalue_x) (Utilities|Reciprocal _returnvalue_x)) "0.000000,0.000000,0.000000" "(B=0,G=0,R=0,A=0)" 0 0) (Utilities|Struct|MakeParticleSysParam "BlocksVelocity" "PSPT_Vector" 0.0 0.0 (Math|Vector|MakeVector 0.0 0.0 (* -2e+01 (Utilities|Reciprocal _returnvalue_z))) "0.000000,0.000000,0.000000" "(B=0,G=0,R=0,A=0)" 0 0))))

(fn WouldVolumeAffectPlayer
  (params (Pawn :type Actor%20Object%20Reference))
  (bind _valid_controller (CallFunction|CheckPlayerTeamandClassValidity Pawn))
  (return (Math|Boolean|ANDBoolean _valid_controller (Variables|UserOptions|GetbAffectsPlayers) (CallFunction|CanPassinCreateMode))))
