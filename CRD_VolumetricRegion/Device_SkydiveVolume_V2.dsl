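; Fortnite Creative Skydive Volume device (V2) that puts players and participant bots inside it into a vortex skydive state.
; On entry it applies gameplay effects and sets vortex parameters from overlap bounds and user options, including push force and optional player locking.
; Blocking gameplay tags defer the effect through async tag-removal waits, which recheck on completion.
; Exit, disable, minigame end or leaving the island triggers cleanup, while enter, exit, occupied and emptied events broadcast.
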(event OnEnabledStateChanged(EnabledComponent)
  (params (Enabled :type Boolean))
  ;[1 1 1] OnEnabledStateChanged
  (if (not Enabled)
    (CallFunction|CleanAffectedPawns)))

(event EventOnWorldReady
  (|CallParentFunction)
  (CallFunction|UpdateCreativeVolumeFrameonWorldReady))

(event EventBeginPlay
  ;[0.139 0.8021 0.04948] Initialization - Skydive volumes want to register to a central location (for FX Updating)
  (|CallParentFunction)
  (CreativeObjectRegistry|GetOrCreateObjectRegister self "(TagName=\"Creative.Register.Vortex\")" true 0 (Variables|Default|GetVortex%20Register))
  ;[0.3728 1 0.6644] Listen for players leaving islands to ensure we don't allow flight while outside the bounds of the island
  (bind _outvolumemanager (Volume|GetVolumeManager)
    (:Valid
      (Variables|Default|SetCachedIslandVolume (VolumeManager|GetVolumeforActor _outvolumemanager self))
      (EventDispatchers|BindEventtoOnClientExitVolume _outvolumemanager (EventDispatchers|CreateEvent :!event Custom|On%20Player%20Exit%20Any%20Island)))))

(event EventInitializeDevice
  (bind _fortminigameprogress (Variables|Default|GetFortMinigameProgress))
  (bind _currentminigame (Variables|/Script/FortniteGame.FortMinigameLogicComponent|GetCurrentMinigame _fortminigameprogress))
  ;[0.139 0.8021 0.04948] Initialize Device
  (|CallParentFunction)
  (CallFunction|InitializeCreativeIcone)
  (CallFunction|UpdateZoneSize)
  (Utilities|IsValid _currentminigame
    (:Is%20Valid
      (Class|CreativeEnabledComponent|SetEnabledOnPhase (Variables|Default|GetEnabledComponent) (Variables|/Script/FortniteGame.FortMinigame|GetCurrentState _currentminigame) (DeviceConversion|EnabledPhaseEnumtoLegacyInt (Variables|UserOptions|GetEnabledOnPhase)))
      (Registration|RevalidateOverlappingPawns)
      (Networking|SwitchHasAuthority
        (:Authority
          (Variables|Default|SetbLockPlayersInVolume (Variables|UserOptions|GetLockAffectedPlayersInVolume))
          (CallFunction|UpdateVortexParametersforPlayerinVolume))))))

(event EventPawnEnteredTriggerVolume
  (params (Pawn :type Pawn%20Object%20Reference) (PlayerController :type Controller%20Object%20Reference))
  (bind _self self)
  (bind _returnvalue (Ability|GetAbilitySystemComponent _self))
  ;[0.3543 1 0.1729] Pawn Entered Volume
  (if (Class|CreativeEnabledComponent|IsEnabled (Variables|Default|GetEnabledComponent))
    (bind _asfort_player_pawn (Utilities|Casting|CastToFortPlayerPawn Pawn))
    (Utilities|IsPawnValid _asfort_player_pawn
      (:Valid
        (bind _returnvalue_2 (Utilities|Array|AddUnique (Variables|Default|GetAffectedPawns) _asfort_player_pawn))
        (if (and (>= _returnvalue_2 0) (Networking|HasAuthority _asfort_player_pawn))
          (bind _data (Utilities|FlowControl|MergeExecutionwithData _asfort_player_pawn (stale Custom|RecheckDelayTags FortPlayerPawn))
            (:Then
              (bind (_bdelayvotexeffect _foundtag) (CallFunction|CheckforDelayTags _data))
              (if (not _bdelayvotexeffect)
                (Utilities|FlowControl|Sequence
                  (:then_0
                    (Utilities|IsValid (Ability|GetAbilitySystemComponent _data)
                      (:Is%20Valid
                        (Utilities|IsValid _returnvalue
                          (:Is%20Valid
                            (GameplayEffects|ApplyGameplayEffectToTarget :self _returnvalue :GameplayEffectClass (Variables|Default|GetInVortexGameplayEffect) :Target (Ability|GetAbilitySystemComponent _data) :Level 0.0)
                            (if (Variables|UserOptions|GetPlayerSkydiveFX)
                              (bind _returnvalue_3 (GameplayEffects|ApplyGameplayEffectToTarget :self _returnvalue :GameplayEffectClass (Variables|Default|GetPlayerFXGameplayEffect) :Target (Ability|GetAbilitySystemComponent _data) :Level 0.0))
                              (Utilities|Map|Add (Variables|Default|GetPlayerFXGameplayEffectHandle) _data _returnvalue_3)))))))
                  (:then_1
                    (CallFunction|SetPlayerinVortexVolume _data)))
                (else
                  (Utilities|FlowControl|Sequence
                    (:then_0
                      _)
                    (:then_1
                      (CallFunction|DelayVortexEffect _data _foundtag))) ; Fire the broadcasts, then delay applying the effect
                ))
              (if (not (Utilities|Array|ContainsItem (Variables|Default|GetDelayedPawns) _data))
                (Utilities|IsValid (Pawn|GetController _data)
                  (:Is%20Valid
                    (GameplayEvent|BroadcastEvent (Variables|UserOptions|Events|GetOn%20Player%20Enter) 0 (Utilities|Struct|MakeGameplayEventEnterSkyDiveVol (Pawn|GetController _data)))
                    (if (== (Utilities|Array|Length (Variables|Default|GetAffectedPawns)) 1)
                      (GameplayEvent|BroadcastEvent (Variables|UserOptions|Events|GetOn%20Zone%20Occupied) 0 (Utilities|Struct|MakeGameplayEventOccupiedSkyDive (Pawn|GetController _data))))))
                (else
                  (Utilities|Array|RemoveItem (Variables|Default|GetDelayedPawns) _data))))))))))

(event Custom|RecheckDelayTags
  (params (FortPlayerPawn :type Fort%20Player%20Pawn%20Object%20Reference))
  (bind _self self)
  (bind _returnvalue (Ability|GetAbilitySystemComponent _self))
  ;[0.3543 1 0.1729] Pawn Entered Volume
  (bind _data (Utilities|FlowControl|MergeExecutionwithData (stale EventPawnEnteredTriggerVolume _asfort_player_pawn) FortPlayerPawn)
    (:Then
      (bind (_bdelayvotexeffect _foundtag) (CallFunction|CheckforDelayTags _data))
      (if (not _bdelayvotexeffect)
        (Utilities|FlowControl|Sequence
          (:then_0
            (Utilities|IsValid (Ability|GetAbilitySystemComponent _data)
              (:Is%20Valid
                (Utilities|IsValid _returnvalue
                  (:Is%20Valid
                    (GameplayEffects|ApplyGameplayEffectToTarget :self _returnvalue :GameplayEffectClass (Variables|Default|GetInVortexGameplayEffect) :Target (Ability|GetAbilitySystemComponent _data) :Level 0.0)
                    (if (Variables|UserOptions|GetPlayerSkydiveFX)
                      (bind _returnvalue_2 (GameplayEffects|ApplyGameplayEffectToTarget :self _returnvalue :GameplayEffectClass (Variables|Default|GetPlayerFXGameplayEffect) :Target (Ability|GetAbilitySystemComponent _data) :Level 0.0))
                      (Utilities|Map|Add (Variables|Default|GetPlayerFXGameplayEffectHandle) _data _returnvalue_2)))))))
          (:then_1
            (CallFunction|SetPlayerinVortexVolume _data)))
        (else
          (Utilities|FlowControl|Sequence
            (:then_0
              _)
            (:then_1
              (CallFunction|DelayVortexEffect _data _foundtag))) ; Fire the broadcasts, then delay applying the effect
        ))
      (if (not (Utilities|Array|ContainsItem (Variables|Default|GetDelayedPawns) _data))
        (Utilities|IsValid (Pawn|GetController _data)
          (:Is%20Valid
            (GameplayEvent|BroadcastEvent (Variables|UserOptions|Events|GetOn%20Player%20Enter) 0 (Utilities|Struct|MakeGameplayEventEnterSkyDiveVol (Pawn|GetController _data)))
            (if (== (Utilities|Array|Length (Variables|Default|GetAffectedPawns)) 1)
              (GameplayEvent|BroadcastEvent (Variables|UserOptions|Events|GetOn%20Zone%20Occupied) 0 (Utilities|Struct|MakeGameplayEventOccupiedSkyDive (Pawn|GetController _data))))))
        (else
          (Utilities|Array|RemoveItem (Variables|Default|GetDelayedPawns) _data))))))

(event EventPlayerBotEnteredTriggerVolume
  (params (Pawn :type Pawn%20Object%20Reference) (AI%20Bot%20Controller :type Fort%20Athena%20AIBot%20Controller%20Object%20Reference))
  ;[0.3543 1 0.1729] Pawn Entered Volume
  (if (Class|CreativeEnabledComponent|IsEnabled (Variables|Default|GetEnabledComponent))
    (bind _asfort_player_state_athena (Utilities|Casting|CastToFortPlayerStateAthena (Variables|/Script/Engine.Pawn|GetPlayerState Pawn)))
    (if (Class|FortPlayerStateAthena|IsAnAthenaGameParticipant _asfort_player_state_athena)
      (CallFunction|PawnEnteredTriggerVolume Pawn))))

(event EventPawnExitedTriggerVolume
  (params (Pawn :type Pawn%20Object%20Reference) (PlayerController :type Controller%20Object%20Reference))
  ;[1 0.5521 0.0603] Pawn Exited Volume
  (if (Class|CreativeEnabledComponent|IsEnabled (Variables|Default|GetEnabledComponent))
    (CallFunction|CleanAffectedPawn Pawn)))

(event EventPlayerBotExitedTriggerVolume
  (params (Pawn :type Pawn%20Object%20Reference) (AI%20Bot%20Controller :type Fort%20Athena%20AIBot%20Controller%20Object%20Reference))
  ;[1 0.5521 0.0603] Pawn Exited Volume
  (if (Class|CreativeEnabledComponent|IsEnabled (Variables|Default|GetEnabledComponent))
    (bind _asfort_player_state_athena (Utilities|Casting|CastToFortPlayerStateAthena (Variables|/Script/Engine.Pawn|GetPlayerState Pawn)))
    (if (Class|FortPlayerStateAthena|IsAnAthenaGameParticipant _asfort_player_state_athena)
      (CallFunction|CleanAffectedPawn Pawn))))

(event MinigameEndedDelegate(FortMinigameProgress)
  ;[1 0.4479 0.4479 0.75] Cleanup after game
  (CallFunction|CleanAffectedPawns))

(event Custom|Update%20Vortex%20Parameters%20For%20Player%20In%20Volume
  ;[0.1763 0.8701 1] Update players vortex parameters
  (for _array_element (Variables|Default|GetAffectedPawns)
    (CallFunction|SetPlayerinVortexVolume _array_element)))

(event Custom|LockPlayersInVolumeFunction
  (params (Instigator :type Controller%20Object%20Reference))
  ;[0.492 0.2848 1] Receivers
  (Variables|Default|SetbLockPlayersInVolume true)
  (CallFunction|UpdateVortexParametersforPlayerinVolume))

(event Custom|AllowPlayersToLeaveVolumeFunction
  (params (Instigator :type Controller%20Object%20Reference))
  ;[0.492 0.2848 1] Receivers
  (Variables|Default|SetbLockPlayersInVolume false)
  (CallFunction|UpdateVortexParametersforPlayerinVolume))

(event Custom|On%20Player%20Exit%20Any%20Island
  (params (Client :type Player%20State%20Object%20Reference) (Volume :type Fort%20Volume%20Object%20Reference))
  ;[1 0.4875 0.1602] Remove volume effects from player if they leave the bounds of our volume and are currently being affected
  (if (== Volume (Variables|Default|GetCachedIslandVolume))
    (Utilities|IsValid Client
      (:Is%20Valid
        (bind _asfort_player_pawn (Utilities|Casting|CastToFortPlayerPawn (PlayerState|GetPawn Client)))
        (if (Utilities|Array|ContainsItem (Variables|Default|GetAffectedPawns) _asfort_player_pawn)
          (CallFunction|PawnExitedTriggerVolume _asfort_player_pawn) ; Remove the effects of this volume from the player
        )))))

(event EventEndPlay
  (params (EndPlayReason :type EEndPlayReason%20Enum :default "Destroyed"))
  ;[1 0.2283 0.06831] End play
  (|CallParentFunction EndPlayReason)
  (Utilities|FlowControl|Sequence
    (:then_0
      ;;[0.9008 1 0.04388] Unbind events.
      (bind _outvolumemanager (Volume|GetVolumeManager)
        (:Valid
          (EventDispatchers|UnbindEventfromOnClientExitVolume _outvolumemanager (EventDispatchers|CreateEvent :!event Custom|On%20Player%20Exit%20Any%20Island)))))))

(event Custom|DelayVortexEffect
  (params (Player :type Fort%20Player%20Pawn%20Object%20Reference) (TagToCheck :type Gameplay%20Tag%20Structure))
  ;[0.599 0.1435 0.1435] Delay Vortex Effect
  (Utilities|Array|AddUnique (Variables|Default|GetDelayedPawns) Player)
  (bind _asynctaskproxy (Ability|Tasks|WaitGameplayTagRemoveFromActor Player TagToCheck true)
    (:then
      (Utilities|Map|Add (Variables|Default|GetWaitGameplayTagRemovedPlayer) _asynctaskproxy Player))
    (:Removed
      (Utilities|FlowControl|DelayUntilNextTick) ; Wait for the action to be marked complete next tick
      (Default|HandleDelayTagRemoved))))

(event Custom|UpdateZoneSize
  (bind _zoneshape (CallFunction|GetZoneShape))
  (bind _zonemeshscalar (Variables|Default|GetZoneMeshScalar))
  (bind (_value _returnvalue) (Utilities|Map|Find _zonemeshscalar _zoneshape))
  (bind _value_location (Math|Transform|BreakTransform _value))
  (bind _in_game_zone (Variables|Default|GetIn%20Game%20Zone))
  ;[0 0.8423 1 0.25] Update Skydive VFX to match with overlap volume size.
  (Variables|Default|SetZoneSizeVector (Math|Vector|MakeVector (Variables|UserOptions|GetZoneWidth) (Variables|UserOptions|GetZoneDepth) (Variables|UserOptions|GetZoneHeight)))
  (Utilities|IsValid _in_game_zone
    (:Is%20Valid
      (bind _output_get (Variables|Default|SetNewScale (* (Variables|Default|GetZoneSizeVector) (Variables|Default|GetInGameZoneScale))))
      (Transformation|SetRelativeScale3D _in_game_zone _output_get)
      (if _returnvalue
        (Transformation|SetRelativeLocation _in_game_zone (Math|Vector|MakeVector (.x _value_location) (.y _value_location) (+ (.z _value_location) (* (Variables|Default|GetZoneRelativeZ) (Variables|UserOptions|GetZoneHeight)))) false false)
        (else
          (Fortnite|FortShippingLog self "SkydiveVolume: Failed to find Zone Shape."))))
    (:Is%20Not%20Valid
      (Fortnite|FortShippingLog self "SkydiveVolume: In Game Zone is Invalid!" true))))

(fn CancelAllDelayAsync
  (bind _keys (Utilities|Map|Keys (Variables|Default|GetWaitGameplayTagRemovedPlayer)))
  (for _array_element _keys
    (AsyncAction|Cancel _array_element))
  (Utilities|Map|Clear (Variables|Default|GetWaitGameplayTagRemovedPlayer)))

(fn CancelDelayAsyncForTarget
  (params (TargetFortPlayerPawn :type Fort%20Player%20Pawn%20Object%20Reference))
  (bind _waitgameplaytagremovedplayer (Variables|Default|GetWaitGameplayTagRemovedPlayer))
  (bind _values (Utilities|Map|Values _waitgameplaytagremovedplayer))
  (if (>= (Utilities|Array|FindItem _values TargetFortPlayerPawn) 0)
    (bind _keys (Utilities|Map|Keys _waitgameplaytagremovedplayer))
    (AsyncAction|Cancel (Utilities|Array|Get(acopy) _keys (Utilities|Array|FindItem _values TargetFortPlayerPawn)))
    (Utilities|Map|Remove _waitgameplaytagremovedplayer (Utilities|Array|Get(acopy) _keys (Utilities|Array|FindItem _values TargetFortPlayerPawn)))))

(fn CheckForDelayTags
  (params (FortPlayerPawn :type Fort%20Player%20Pawn%20Object%20Reference))
  (for _array_element (Variables|Default|GetGameplayTags_DelayVortexEffect)
    (if (GameplayTags|HasTag (GameplayTags|GetOwnedGameplayTags FortPlayerPawn :!class /Script/GameplayTags.GameplayTagAssetInterface) _array_element true)
      (return true _array_element)))
  (return false 0))

(fn CleanAffectedPawn
  (params (Pawn :type Pawn%20Object%20Reference))
  (bind _invortexgameplayeffect (Variables|Default|GetInVortexGameplayEffect))
  (bind _playerfxgameplayeffecthandle (Variables|Default|GetPlayerFXGameplayEffectHandle))
  (bind _asfort_player_pawn (Utilities|Casting|CastToFortPlayerPawn Pawn))
  (Utilities|Array|RemoveItem (Variables|Default|GetAffectedPawns) _asfort_player_pawn)
  (Utilities|Array|RemoveItem (Variables|Default|GetDelayedPawns) _asfort_player_pawn)
  (CallFunction|CancelDelayAsyncforTarget _asfort_player_pawn)
  (Networking|SwitchHasAuthority
    (:Authority
      (Utilities|FlowControl|Sequence
        (:then_0
          ;[1 0.2446 0.06435] Remove skydive effects from player
          (Utilities|IsValid (Ability|GetAbilitySystemComponent _asfort_player_pawn)
            (:Is%20Valid
              (GameplayEffects|RemoveActiveGameplayEffect (Ability|GetAbilitySystemComponent _asfort_player_pawn) (Utilities|Map|Find _playerfxgameplayeffecthandle _asfort_player_pawn) 1)
              (Utilities|Map|Remove _playerfxgameplayeffecthandle _asfort_player_pawn)
              (GameplayEffects|RemoveActiveGameplayEffectbySourceEffect (Ability|GetAbilitySystemComponent _asfort_player_pawn) _invortexgameplayeffect 0 1)
              (if (<= (GameplayEffects|GetGameplayEffectCount (Ability|GetAbilitySystemComponent _asfort_player_pawn) _invortexgameplayeffect 0 true) 0)
                (Vortex|SetinVortex _asfort_player_pawn false 0.0 0.0 0.0 0.0 0.0 true)))))
        (:then_1
          ;[0.3281 0.6669 1] Trigger Channels
          (Utilities|IsValid (Pawn|GetController _asfort_player_pawn)
            (:Is%20Valid
              (GameplayEvent|BroadcastEvent (Variables|UserOptions|Events|GetOn%20Player%20Exits) 0 (Utilities|Struct|MakeGameplayEventExitSkyDiveVol (Pawn|GetController _asfort_player_pawn)))
              (if (== (Utilities|Array|Length (Variables|Default|GetAffectedPawns)) 0)
                (GameplayEvent|BroadcastEvent (Variables|UserOptions|Events|GetOn%20Zone%20Emptied) 0 (Utilities|Struct|MakeGameplayEventEmptySkyDiveVol (Pawn|GetController _asfort_player_pawn)))))))))))

(fn CleanAffectedPawns
  (bind _affectedpawns (Variables|Default|GetAffectedPawns))
  (Networking|SwitchHasAuthority
    (:Authority
      (if (> (Utilities|Array|Length _affectedpawns) 0)
        (for _array_element _affectedpawns
          (CallFunction|CleanAffectedPawn _array_element))
        (else
          _))
      (Utilities|Array|Clear _affectedpawns)
      (Utilities|Array|Clear (Variables|Default|GetDelayedPawns))
      (CallFunction|CancelAllDelayAsync))))

(fn GetAllVisualComponents
  (return (Utilities|Array|MakeArray (Variables|Default|GetCreativeIcon)) (Utilities|Array|MakeArray (Variables|Default|GetIn%20Game%20Zone) (Variables|Default|GetFNE_VolumeOverlapComponent))))

(fn GetCollisionStaticMesh
  (return "0, 0, 0" (Variables|/Script/Engine.StaticMeshComponent|GetStaticMesh (Variables|Default|GetCreativeIcon))))

(fn GetMeshComponents
  (return (Utilities|Array|MakeArray (Variables|Default|GetCreativeIcon) (Variables|Default|GetCreativeVolumeFrame))))

(fn GetVolumeSizeParameters
  (bind _fne_volumeoverlapcomponent (Variables|Default|GetFNE_VolumeOverlapComponent))
  (bind (_origin _boxextent) (Collision|GetComponentBounds _fne_volumeoverlapcomponent))
  ;[1 0.5402 0.07099] Constructs the transform the vortex volume should use based on the devices trigger volume, and calculates the thickness of the padding at the edge of the volume that pushes the player toward the center
  (Utilities|IsValid _fne_volumeoverlapcomponent
    (:Is%20Valid
      (return (Math|Transform|MakeTransform _origin (Transformation|GetWorldRotation _fne_volumeoverlapcomponent) _boxextent) (Math|Float|Max(Float) (* (Math|Vector|GetMinElement _boxextent) (Variables|VortexParameters|GetInnerBoundsSizeRatio)) (Variables|VortexParameters|GetMinBoundsSize))))))

(fn GetVortexDirection
  (params (For%20Actor :type Actor%20Object%20Reference))
  (bind _asfort_player_pawn (Utilities|Casting|CastToFortPlayerPawn For%20Actor)
    (:then
      (return (Utilities|Array|ContainsItem (Variables|Default|GetAffectedPawns) _asfort_player_pawn) (* (Transformation|GetActorUpVector) (Math|Float|Sign(Float) (Variables|UserOptions|GetPushForce)))))
    (:CastFailed
      (return false "0, 0, 0"))))

(fn HandleDelayTagRemoved
  (bind _waitgameplaytagremovedplayer (Variables|Default|GetWaitGameplayTagRemovedPlayer))
  (bind _keys (Utilities|Map|Keys (Variables|Default|GetWaitGameplayTagRemovedPlayer)))
  ;[0.4728 1 0.2257 0.5] Find completed Wait actions
  (bind (_arrayindex _arrayelement) (Utilities|Array|ReverseforEachLoop _keys)
    (:LoopBody
      (if (not (AsyncAction|IsActive _arrayelement))
        (Utilities|IsValid (Utilities|Map|Find _waitgameplaytagremovedplayer _arrayelement)
          (:Is%20Valid
            (CallFunction|RecheckDelayTags (Utilities|Map|Find _waitgameplaytagremovedplayer _arrayelement)) ; Check again for any tags that might cause delay, then continue the delay or apply the effects.
          )
          (:Is%20Not%20Valid
            _))
        (Utilities|Map|Remove (Variables|Default|GetWaitGameplayTagRemovedPlayer) _arrayelement) ; This Wait action is done. Remove its map entry.
      ))) ; Reverse so we don't mess this up when we remove an entry
)

(fn HandleDeviceScaling
  (bind (_relativescale3d_x _relativescale3d_y _relativescale3d_z) (Math|Vector|BreakVector (Variables|/Script/Engine.SceneComponent|GetRelativeScale3D_X (Variables|Debug|GetStaticMeshComponent))))
  (Transformation|SetWorldScale3D (Variables|Default|GetCreativeIcon) (Math|Vector|MakeVector (* (* (Utilities|Operators|Add _relativescale3d_x _relativescale3d_y _relativescale3d_z) 0.33) (Variables|Default|GetIconScalar)) (* (* (Utilities|Operators|Add _relativescale3d_x _relativescale3d_y _relativescale3d_z) 0.33) (Variables|Default|GetIconScalar)) (* (* (Utilities|Operators|Add _relativescale3d_x _relativescale3d_y _relativescale3d_z) 0.33) (Variables|Default|GetIconScalar)))))

(fn IgnoreWorldGridScaleOverride
  (return false))

(fn InitializeCreativeIcone
  (Resizing|HandleDeviceScaling)
  (if (not (Editor|InEditor))
    (Class|CreativeEditOnlyMeshComponent|ReloadMeshandMaterials (Variables|Default|GetCreativeIcon))))

(fn MigrateLegacyUserOptions
  (bind (_delayedpawns _skydivevolumescalemultiplier _shape_edit_mesh_and_mats _zonematerial _ingamezonescale _zonesizevector _zonerelativez _gameplaytags_delayvortexeffect _enabledindex) (Class|GetClassDefaults "/CRD_VolumetricRegion/Device_SkydiveVolume_V2.Device_SkydiveVolume_V2_C"))
  (bind _enabledindex_2 (Variables|LegacyOptions|GetEnabledIndex))
  (|CallParentFunction)
  (Utilities|FlowControl|Sequence
    (:then_0
      (Utilities|FlowControl|Sequence
        (:then_0
          (if (!= _enabledindex _enabledindex_2)
            (Variables|UserOptions|SetEnabledOnPhase (Utilities|Enum|BytetoEnumEEnabledOnPhase (Math|Conversions|ToByte(Integer) (+ _enabledindex_2 1))))
            (Variables|LegacyOptions|SetEnabledIndex _enabledindex)
            (LocalVariables|SetOptions%20Migrated true)))))
    (:then_1
      (if (LocalVariables|GetOptions%20Migrated)
        (OptionData|RefreshPropertyOverrides (Variables|Default|GetToyOptionsComponent))))))

(fn RemovePawnFromVolume
  (params (Player%20Pawn :type Fort%20Player%20Pawn%20Object%20Reference)))

(fn ResizeFrame
  (Variables|Default|SetMaterialScalar (Variables|Default|GetFrameSizeOverride))
  (|CallParentFunction))

(fn SetPlayerInVortexVolume
  (params (Player :type Fort%20Player%20Pawn%20Object%20Reference))
  (bind (_volume_transform _volume_thickness) (CallFunction|GetVolumeSizeParameters))
  ;[0.2104 1 0.5208] Sets the given player to be in a vortex volume with parameters from user options. This event can be used to either initially set the player as in a vortex or update the parameters of a player already in a vortex
  (Utilities|IsValid Player
    (:Is%20Valid
      (Vortex|SetinVortexVolume Player true (Variables|UserOptions|GetLaunchVelocity) _volume_transform _volume_thickness (* (Variables|UserOptions|GetPushForce) -1.0) (select (CallFunction|ShouldLockPlayerinVolume Player) 0.0 (Variables|VortexParameters|GetLeavingVolumeTerminalVelocity)) (Variables|UserOptions|GetbAllowGliderDeploy)))))

(fn ShouldLockPlayerInVolume
  (params (Player%20Actor :type Actor%20Object%20Reference))
  (bind _creativeuserestrictioncomponent (Variables|Default|GetCreativeUseRestrictionComponent))
  ;[1 0.3582 0.06447] Returns whether player should not be able to leave the volume based on channel reciver, locked team option, and locked class option
  (Utilities|IsValid _creativeuserestrictioncomponent
    (:Is%20Valid
      (bind _valid_controller (CreativeUserRestrictions|ActorHasValidParticipantController _creativeuserestrictioncomponent Player%20Actor true true true))
      (return (and _valid_controller (Variables|Default|GetbLockPlayersInVolume))))
    (:Is%20Not%20Valid
      (return false))))

(fn UpdateCreativeVolumeFrameOnWorldReady
  ;[1 0.8085 0.09327] Plug in our values that can't be directly referenced but can be made transient
  (Variables|/Game/Creative/Devices/Common/Components/CreativeVolumeFrame.CreativeVolumeFrame_C|SetShapeEditMeshAndMats (Variables|Default|GetShape%20Edit%20Mesh%20And%20Mats) (Variables|Default|GetCreativeVolumeFrame)))

(fn ConstructionScript
  (|CallParentFunction))
